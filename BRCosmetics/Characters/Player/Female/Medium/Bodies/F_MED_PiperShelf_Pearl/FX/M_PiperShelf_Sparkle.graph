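# Unlit, two-sided translucent Niagara sprite and mesh particle material that draws a radial sparkle for the Piper Shelf Pearl cosmetic.
# Emissive multiplies a polar-mapped rainbow texture by a color parameter and particle color.
# Panning polar caustics noise, masked by radial falloff, forms rays that combine with a core and glow, scaled by particle alpha.
# Texture samples use derivatives from pre-polar UVs for mip selection.
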Material /BRCosmetics/Characters/Player/Female/Medium/Bodies/F_MED_PiperShelf_Pearl/FX/M_PiperShelf_Sparkle
graph {
 properties {
  bAutomaticallySetUsageInEditor=false
  bUsedWithNiagaraMeshParticles=true
  bUsedWithNiagaraSprites=true
  blendMode=BLEND_Translucent
  shadingModel=MSM_Unlit
  twoSided=true
  usedShadingModels=Unlit
 }
 nodes [
  n1 = ParticleColor()
  n2 = VectorParameter() {group=Noise Properties, parameterName=Color, sortPriority=0, defaultValue=<rgba 0 13.6111 32 0>}
  n3 = Constant() {r=128}
  n4 = MaterialFunctionCall(UVs=TextureCoordinate(), Texture Scale=@n3) {materialFunction=&ScaleUVsByCenter}
  n5 = MaterialFunctionCall!(UV Coord=@n4.UVs, LenghtScale, RadialScale) {materialFunction=/Game/Global/MaterialFunctions/TechArt/MF_PolarCoordinates}
  n6 = DDX(Value=@n4.UVs)
  n7 = DDY(Value=@n4.UVs)
  n8 = TextureSample!(UVs=@n5.Result, Tex, "DDX(UVs)"=@n6, "DDY(UVs)"=@n7, Apply View MipBias) {mipValueMode=TMVM_Derivative, texture=&T_Tile_RainbowColor}
  MaterialFunctionCall(Emissive=Multiply(A=@n1.RGB, B=Multiply(A=@n2.RGB, B=@n8.RGB))) {materialFunction=&MF_FX_EmissiveAdjustments}
  n9 = ScalarParameter() {group=Noise Properties, parameterName=Global Range Minimum, sortPriority=11}
  n10 = ScalarParameter() {defaultValue=0.7963, group=Noise Properties, parameterName=Global Range Maximum, sortPriority=12}
  n11 = VectorParameter() {
   group=Noise Properties
   parameterName="Noise Scale && Offset"
   sortPriority=2
   channelNames {a=Speed V, b=Speed U, g=Scale V, r=Scale U}
   defaultValue=<rgba 1 1 3 8>
  }
  n12 = Panner(
   Coordinate=Multiply(A=@n5.Result, B=AppendVector(A=@n11.Scale U, B=@n11.Scale U))
   Speed=Divide(A=AppendVector(A=@n11.Speed U, B=@n11.Speed U), B=@n3)
  )
  n13 = TextureObjectParameter() {group=Noise Properties, parameterName=Noise Map, sortPriority=1, texture=/Game/Global/Textures/Default/Caustics/T_Caustics_Softer}
  n14 = ScalarParameter() {defaultValue=0.1429, group=Noise Properties, parameterName=Rays Range Minimum, sortPriority=4}
  n15 = ScalarParameter() {defaultValue=0.7778, group=Noise Properties, parameterName=Rays Range Maximum, sortPriority=5}
  n16 = MaterialFunctionCall(Vector 2=Subtract(A=TextureCoordinate(), B=Constant2Vector() {g=0.5, r=0.5})) {materialFunction=/Engine/Functions/Engine_MaterialFunctions02/Utility/VectorLength}
  n17 = OneMinus(None=Power(Base=@n16.V2 Length, Exp=ScalarParameter() {defaultValue=0.2356, group=Noise Properties, parameterName=Rays Power, sortPriority=3}))
  n18 = OneMinus(None=Power(Base=@n16.V2 Length, Exp=ScalarParameter() {defaultValue=0.2261, group=Noise Properties, parameterName=Core Power, sortPriority=6}))
  n19 = Add(
   A=Multiply(
    A=TextureSample(UVs=@n12, Tex=@n13, "DDX(UVs)"=@n6, "DDY(UVs)"=@n7) {mipValueMode=TMVM_Derivative, samplerType=SAMPLERTYPE_LinearColor}.R
    B=SmoothStep(Min=@n14, Max=@n15, Value=@n17) {constMax=0.5, constMin=0.49}
   )
   B=SmoothStep(
    Min=ScalarParameter() {defaultValue=0.419, group=Noise Properties, parameterName=Core Range Minimum, sortPriority=7}
    Max=ScalarParameter() {defaultValue=0.9491, group=Noise Properties, parameterName=Core Range Maximum, sortPriority=8}
    Value=@n18
   ) {constMax=0.5, constMin=0.49}
  )
  n20 = ScalarParameter() {defaultValue=0.219, group=Noise Properties, parameterName=Glow Range Minimum, sortPriority=9}
  n21 = ScalarParameter() {defaultValue=4.6505, group=Noise Properties, parameterName=Glow Range Maximum, sortPriority=10}
  Multiply(
   A=Add(
    A=SmoothStep(Min=@n9, Max=@n10, Value=@n19) {constMax=0.5, constMin=0.49}
    B=SmoothStep(Min=@n20, Max=@n21, Value=@n18) {constMax=0.5, constMin=0.49}
   )
   B=@n1.A
  )
 ]
}
