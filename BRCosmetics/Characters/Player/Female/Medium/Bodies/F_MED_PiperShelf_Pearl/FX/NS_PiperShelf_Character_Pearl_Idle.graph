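# Niagara idle VFX for the Pearl variant of the PiperShelf character body, adding flickering flare sparkles to the upper body and feet.
# Two emitters pin short-lived sprites to bones of the Pearl skeleton at random hand-authored offsets, negated for mirrored partitions.
# User.Variant blends flare colors, and bisFootHidden stops feet spawning.
# Size, color and alpha respond to time of day, with brightness scaled by quality level.
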NiagaraSystem /BRCosmetics/Characters/Player/Female/Medium/Bodies/F_MED_PiperShelf_Pearl/FX/NS_PiperShelf_Character_Pearl_Idle
dependencies {
 usedDataInterfaces [
  &NiagaraDataInterfaceArrayFloat3
  &NiagaraDataInterfaceCurve
  &NiagaraDataInterfaceArrayFloat
  &NiagaraDataInterfaceSkeletalMesh
  &NiagaraDataInterfaceColorCurve
 ]
 usedDynamicInputs [
  &Multiply_Vector2D_ByFloat
  &Lerp_Float
  &MakeVector2D
 ]
 usedModules [
  &self:SystemSpawnScript.SystemScriptSource.SystemScriptGraph.NiagaraNodeAssignment_2.SetVariables_#1
  &SystemState
  &self:SystemSpawnScript.SystemScriptSource.SystemScriptGraph.NiagaraNodeAssignment_1.SetVariables_#2
  &/TimeOfDayGlobal
 ]
}
emitters {
 Sparkles_Feet {
  emitterUpdateScript [
   &EmitterState()
   &SpawnRate(
    Spawn Probability=&MakeCustomFloatFromBool(Bool=$User.bisFootHidden, True Float=0, False Float=$System.SpawnProbability)
    SpawnRate=&Multiply_Float(A=$System.SpawnRate, B=3)
    Use Spawn Probability=true
   )
  ]
  particleSpawnScript [
   &PartitionParticles(Number Of Partitions=5)
   &InitializeParticle(
    AdjustAlpha=false
    AdjustHue=true
    AdjustSaturation=false
    AdjustValue=false
    Color=&LerpLinearColors(StartColor=<rgba 0.843 0.8 1 1>, EndColor=<rgba 0.4678 1 0.9701 1>, LerpFactor=$User.Variant)
    Color Mode=::NewEnumerator3
    Lifetime Max=0.4
    Lifetime Min=0.3
    Lifetime Mode=::NewEnumerator1
    Sprite Size Max=<35 35>
    Sprite Size Min=<25 25>
    Sprite Size Mode=::NewEnumerator2
    UsePositionOffset=false
    Write Material Random=true
   )
   &CameraOffset(Camera Offset Amount=&FN_FrontEndFloatModifier(
     Gameplay=&MergeHelper_FloatSwitch(0=10, 1=5, 2=5, 3=3, 4=3, Index=$Particles.PartitionParticles.Partition, Num=::NewEnumerator3)
     Front End=&MergeHelper_FloatSwitch(0=6, 1=4, 2=4, 3=3, 4=3, Index=$Particles.PartitionParticles.Partition, Num=::NewEnumerator3)
    ))
  ]
  particleUpdateScript [
   &ParticleState()
   &SkeletalMeshLocation(
    Bone / Socket Index=&MergeHelper_IntSwitch(0=1, 1=2, 2=?, 3=?, 4=?, 5=?, Index=$Particles.PartitionParticles.Partition, Num=::NewEnumerator0)
    Bone Sampling Mode=::NewEnumerator4
    MaskSpawnGroup=false
    Sampled Position Offset=&MergeHelper_VectorSwitch(
     0=&SelectVectorFromArray(
      Array Sampling Mode=::NewEnumerator0
      Direct Array Index=?
      Vector Selection Array=DataInterface() {propertyValues {FloatData [
         <4.9145 0.7918 2.7>
         <4.0101 -0.2923 2.9>
         <7.3122 0.1088 2.9>
         <5.1696 1.0008 -2.3>
         <3.3209 2.1483 -2.0254>
         <3.4208 4.3845 -1.7583>
         <2 5.6877 -1>
         <1.6197 5.1645 -2.7597>
         <-7 15 0.2998>
         <3.3755 3 0.5437>
         <6.3986 2 2>
         <0.9291 1.6485 2.9549>
         <-1.9791 2.5622 2.1796>
         <-6.4212 9.8712 2.086>
         <-2.1517 6.0096 2.1948>
         <1.3658 6.5194 -0.2581>
         <0.9901 -2.3065 0.6449>
         <4.5309 -2.3065 0.6449>
        ]}}
      Random Seed=0
     )
     1=&Multiply_Vector(
      A=&SelectVectorFromArray(
       Array Sampling Mode=::NewEnumerator0
       Direct Array Index=?
       Vector Selection Array=DataInterface() {propertyValues {FloatData [
          <4.9145 0.7918 2.7>
          <4.0101 -0.2923 2.9>
          <7.3122 0.1088 2.9>
          <5.1696 1.0008 -2.3>
          <3.3209 2.1483 -2.0254>
          <3.4208 4.3845 -1.7583>
          <2 5.6877 -1>
          <1.6197 5.1645 -2.7597>
          <-7 15 0.2998>
          <3.3755 3 0.5437>
          <6.3986 2 2>
          <0.9291 1.6485 2.9549>
          <-1.9791 2.5622 2.1796>
          <-6.4212 9.8712 2.086>
          <-2.1517 6.0096 2.1948>
          <1.3658 6.5194 -0.2581>
          <0.9901 -2.3065 0.6449>
          <4.5309 -2.3065 0.6449>
         ]}}
       Random Seed=0
      )
      B=<-1 -1 -1>
     )
     2=&Multiply_Vector(
      A=&SelectVectorFromArray(
       Array Sampling Mode=::NewEnumerator0
       Direct Array Index=?
       Vector Selection Array=DataInterface() {propertyValues {FloatData [
          <4.9145 0.7918 2.7>
          <4.0101 -0.2923 2.9>
          <7.3122 0.1088 2.9>
          <5.1696 1.0008 -2.3>
          <3.3209 2.1483 -2.0254>
          <3.4208 4.3845 -1.7583>
          <2 5.6877 -1>
          <1.6197 5.1645 -2.7597>
          <-7 15 0.2998>
          <3.3755 3 0.5437>
          <6.3986 2 2>
          <0.9291 1.6485 2.9549>
          <-1.9791 2.5622 2.1796>
          <-6.4212 9.8712 2.086>
          <-2.1517 6.0096 2.1948>
          <1.3658 6.5194 -0.2581>
          <0.9901 -2.3065 0.6449>
          <4.5309 -2.3065 0.6449>
         ]}}
       Random Seed=0
      )
      B=<-1 -1 -1>
     )
     3=&SelectVectorFromArray(
      Array Sampling Mode=::NewEnumerator0
      Direct Array Index=?
      Vector Selection Array=DataInterface() {propertyValues {FloatData [
         <12.5753 -0.9357 5.3711>
         <11.4768 -3.5801 4.7877>
         <10.4072 -4.661 3.9741>
         <10.3941 -3.6118 4.8616>
         <8.3624 -3.4524 5.067>
         <10.179 -1.2245 5.4514>
         <11.6392 -0.7666 5.3651>
         <9.5792 -1.0686 5.4388>
         <9.3529 -1.4062 5.4848>
         <9.6719 -5.7638 2.371>
         <8.7494 -5.5545 2.8817>
         <9.9246 -5.2157 3.3325>
         <8.4773 -5.1846 3.4957>
         <7.7346 -0.544 5.4084>
         <7.545 -0.1755 5.2551>
         <7.9089 0.5701 4.8986>
         <6.9144 0.9922 4.5897>
         <9.5032 1.2653 4.574>
         <10.9502 3.0557 3.2685>
         <9.2484 2.2857 3.828>
         <9.3455 3.0628 3.0993>
         <5.3827 2.822 3.0659>
         <6.1413 4.6248 1.5062>
         <6.0572 5.847 0.4026>
         <5.4956 6.1144 0.2805>
        ]}}
      Random Seed=0
     )
     4=&Multiply_Vector(
      A=&SelectVectorFromArray(
       Array Sampling Mode=::NewEnumerator0
       Direct Array Index=?
       Vector Selection Array=DataInterface() {propertyValues {FloatData [
          <12.5753 -0.9357 5.3711>
          <11.4768 -3.5801 4.7877>
          <10.4072 -4.661 3.9741>
          <10.3941 -3.6118 4.8616>
          <8.3624 -3.4524 5.067>
          <10.179 -1.2245 5.4514>
          <11.6392 -0.7666 5.3651>
          <9.5792 -1.0686 5.4388>
          <9.3529 -1.4062 5.4848>
          <9.6719 -5.7638 2.371>
          <8.7494 -5.5545 2.8817>
          <9.9246 -5.2157 3.3325>
          <8.4773 -5.1846 3.4957>
          <7.7346 -0.544 5.4084>
          <7.545 -0.1755 5.2551>
          <7.9089 0.5701 4.8986>
          <6.9144 0.9922 4.5897>
          <9.5032 1.2653 4.574>
          <10.9502 3.0557 3.2685>
          <9.2484 2.2857 3.828>
          <9.3455 3.0628 3.0993>
          <5.3827 2.822 3.0659>
          <6.1413 4.6248 1.5062>
          <6.0572 5.847 0.4026>
          <5.4956 6.1144 0.2805>
         ]}}
       Random Seed=0
      )
      B=<-1 -1 -1>
     )
     Index=$Particles.PartitionParticles.Partition
     Num=::NewEnumerator0
    )
    Skeletal Mesh=$System.Pearl_Skeleton
   ) {module=SkeletalMeshLocation002}
   &ScaleSpriteSize(Uniform Curve Scale=&FNTimeOfDayScalar(Night Scale=1.25, Morning Scale=1, Evening Scale=1)) {module=ScaleSpriteSize001}
   &ScaleColor(
    Scale Alpha=&Multiply_Float(
     A=&FloatFromCurve(
      FloatCurve=DataInterface() {propertyValues {Curve {keys [
          {interpMode=RCIM_Cubic, tangentMode=RCTM_User, time=0, value=0}
          {arriveTangent=-2.4058, interpMode=RCIM_Cubic, leaveTangent=-2.4058, tangentMode=RCTM_User, time=0.26, value=0.8421}
          {arriveTangent=-1.1775, interpMode=RCIM_Cubic, leaveTangent=-1.1775, time=0.4613, value=0.0645}
          {arriveTangent=-0.1197, interpMode=RCIM_Cubic, leaveTangent=-0.1197, time=0.5898, value=0.4537}
          <at 1 0 RCIM_Cubic>
         ]}}}
      Scale Curve=&RandomRangeFloat(Minimum=0.1, Maximum=0.5)
     )
     B=&FloatFromCurve(
      FloatCurve=DataInterface() {propertyValues {Curve {keys [{tangentMode=RCTM_User, time=0, value=0}, {tangentMode=RCTM_User, time=1, value=1}]}}}
      Scale Curve=&FloatFromCurve(
       FloatCurve=DataInterface() {propertyValues {Curve {keys [
           {arriveTangentWeight=0.1667, interpMode=RCIM_Cubic, leaveTangentWeight=0.1667, tangentMode=RCTM_User, time=1.4, value=0}
           {tangentMode=RCTM_User, time=2, value=1}
          ]}}}
       CurveIndex=$Emitter.Age
      )
     )
    )
    Scale RGB=&VectorFromFloat(Value=&Multiply_Float(
      A=&RandomRangeFloat(Minimum=0.7, Maximum=0.9, Evaluation Type=::NewEnumerator1)
      B=&SelectFloatFromArray(Direct Array Index=$Engine.QualityLevel, Float Selection Array=DataInterface() {propertyValues {FloatData [10,5,1,1,1]}})
     ))
    ScaleA=true
    ScaleRGB=true
   ) {module=ScaleColor001}
   &ScaleColor(
    Scale Alpha=$System.TODAlpha
    Scale RGB=&MakeVectorFromLinearColorRGB(Color=$System.TODColor)
    ScaleA=true
    ScaleRGB=true
   )
   /SetVariables(
    Particles.CenterFlareColor:LinearColor=&MultiplyLinearColor(
     Linear Color=$Particles.Color
     Scale Factor=&MakeVector4_FromLinearColor(LinearColor=&LerpLinearColors(StartColor=<rgba 8.0099 9.8007 10 1>, EndColor=<rgba 3.4218 8.799 10 1>, LerpFactor=$User.Variant))
    )
    Particles.CenterFlareSize:Vector2f=&Multiply_Vector2(A=$Particles.SpriteSize, B=&Vector2DFromFloat(Value=0.5))
    Particles.CoronaFlareColor:LinearColor=&MultiplyLinearColor(
     Linear Color=$Particles.Color
     Scale Factor=&MakeVector4_FromLinearColor(LinearColor=&LerpLinearColors(StartColor=<rgba 0.15 0.15 0.15 1>, EndColor=<rgba 0.0729 0.0729 0.0729 1>, LerpFactor=$User.Variant))
    )
    Particles.CoronaFlareSize:Vector2f=&Multiply_Vector2(A=$Particles.Initial.SpriteSize, B=&Vector2DFromFloat(Value=0.75))
    Particles.FlickerFlareColor:LinearColor=&MultiplyLinearColor(
     Linear Color=$Particles.Color
     Scale Factor=&MakeVector4_FromLinearColor(LinearColor=&LerpLinearColors(StartColor=<rgba 0.9441 0.8079 1 1>, EndColor=<rgba 0.0393 0.0337 0.0417 50>, LerpFactor=$User.Variant))
    )
   ) {isSetParameters=true, module=SetVariables_#3}
  ]
  rendererClasses [&NiagaraSpriteRendererProperties]
  renderers [
   {rendererClass=&NiagaraSpriteRendererProperties}
   {rendererClass=&NiagaraSpriteRendererProperties, rendererIndex=1}
   {rendererClass=&NiagaraSpriteRendererProperties, rendererIndex=2}
  ]
 }
 Sparkles_UpperBody {
  emitterUpdateScript [
   &EmitterState()
   &SpawnRate(
    Spawn Probability=$System.SpawnProbability
    SpawnRate=&Multiply_Float(A=$System.SpawnRate, B=3)
    Use Spawn Probability=true
   )
  ]
  particleSpawnScript [
   &PartitionParticles(Number Of Partitions=3)
   &InitializeParticle(
    AdjustAlpha=false
    AdjustHue=true
    AdjustSaturation=false
    AdjustValue=false
    Color=&LerpLinearColors(StartColor=<rgba 0.843 0.8 1 1>, EndColor=<rgba 0.4678 1 0.9701 1>, LerpFactor=$User.Variant)
    Color Mode=::NewEnumerator3
    Lifetime Max=0.4
    Lifetime Min=0.3
    Lifetime Mode=::NewEnumerator1
    Sprite Size Max=<35 35>
    Sprite Size Min=<25 25>
    Sprite Size Mode=::NewEnumerator2
    UsePositionOffset=false
    Write Material Random=true
   )
   &CameraOffset(Camera Offset Amount=&FN_FrontEndFloatModifier(
     Gameplay=&MergeHelper_FloatSwitch(0=10, 1=5, 2=5, 3=3, 4=3, Index=$Particles.PartitionParticles.Partition, Num=::NewEnumerator3)
     Front End=&MergeHelper_FloatSwitch(0=6, 1=4, 2=4, 3=3, 4=3, Index=$Particles.PartitionParticles.Partition, Num=::NewEnumerator3)
    ))
  ]
  particleUpdateScript [
   &ParticleState()
   &SkeletalMeshLocation(
    Bone / Socket Index=&MergeHelper_IntSwitch(0=0, 1=3, 2=4, 3=?, 4=?, 5=?, Index=$Particles.PartitionParticles.Partition, Num=::NewEnumerator1)
    Bone Sampling Mode=::NewEnumerator4
    MaskSpawnGroup=false
    Sampled Position Offset=&MergeHelper_VectorSwitch(
     0=&SelectVectorFromArray(
      Array Sampling Mode=::NewEnumerator0
      Direct Array Index=?
      Vector Selection Array=DataInterface() {propertyValues {FloatData [
         <11.3676 -1.6467 10.3557>
         <8.63 -1.7256 11.9304>
         <7.4525 -1.27 12.568>
         <8.8814 -4.1532 10.932>
         <6.7363 -0.1051 12.9812>
         <3.7883 3.1537 13.57>
         <1.0668 -0.4069 15.5023>
         <4.0833 -0.1806 -14.1933>
         <9.632 -0.5534 -11.4039>
         <1.9966 -2.7752 -14.8841>
         <-0.652 -4.6198 -15.4256>
         <6.2513 -6.5928 -11.1406>
         <5.2166 -4.6507 -12.8888>
         <6.8038 -6.3821 -10.9144>
         <7.7384 -7.6115 -8.626>
         <8.2908 -8.0158 -7.1372>
         <7.2391 -5.5022 11.0561>
         <0.6933 -2.8435 15.4547>
         <0.1496 -1.461 15.7709>
        ]}}
      Random Seed=0
     )
     1=&SelectVectorFromArray(
      Array Sampling Mode=::NewEnumerator0
      Direct Array Index=?
      Vector Selection Array=DataInterface() {propertyValues {FloatData [
         <12.5753 -0.9357 5.3711>
         <11.4768 -3.5801 4.7877>
         <10.4072 -4.661 3.9741>
         <10.3941 -3.6118 4.8616>
         <8.3624 -3.4524 5.067>
         <10.179 -1.2245 5.4514>
         <11.6392 -0.7666 5.3651>
         <9.5792 -1.0686 5.4388>
         <9.3529 -1.4062 5.4848>
         <9.6719 -5.7638 2.371>
         <8.7494 -5.5545 2.8817>
         <9.9246 -5.2157 3.3325>
         <8.4773 -5.1846 3.4957>
         <7.7346 -0.544 5.4084>
         <7.545 -0.1755 5.2551>
         <7.9089 0.5701 4.8986>
         <6.9144 0.9922 4.5897>
         <9.5032 1.2653 4.574>
         <10.9502 3.0557 3.2685>
         <9.2484 2.2857 3.828>
         <9.3455 3.0628 3.0993>
         <5.3827 2.822 3.0659>
         <6.1413 4.6248 1.5062>
         <6.0572 5.847 0.4026>
         <5.4956 6.1144 0.2805>
        ]}}
      Random Seed=0
     )
     2=&Multiply_Vector(
      A=&SelectVectorFromArray(
       Array Sampling Mode=::NewEnumerator0
       Direct Array Index=?
       Vector Selection Array=DataInterface() {propertyValues {FloatData [
          <12.5753 -0.9357 5.3711>
          <11.4768 -3.5801 4.7877>
          <10.4072 -4.661 3.9741>
          <10.3941 -3.6118 4.8616>
          <8.3624 -3.4524 5.067>
          <10.179 -1.2245 5.4514>
          <11.6392 -0.7666 5.3651>
          <9.5792 -1.0686 5.4388>
          <9.3529 -1.4062 5.4848>
          <9.6719 -5.7638 2.371>
          <8.7494 -5.5545 2.8817>
          <9.9246 -5.2157 3.3325>
          <8.4773 -5.1846 3.4957>
          <7.7346 -0.544 5.4084>
          <7.545 -0.1755 5.2551>
          <7.9089 0.5701 4.8986>
          <6.9144 0.9922 4.5897>
          <9.5032 1.2653 4.574>
          <10.9502 3.0557 3.2685>
          <9.2484 2.2857 3.828>
          <9.3455 3.0628 3.0993>
          <5.3827 2.822 3.0659>
          <6.1413 4.6248 1.5062>
          <6.0572 5.847 0.4026>
          <5.4956 6.1144 0.2805>
         ]}}
       Random Seed=0
      )
      B=<-1 -1 -1>
     )
     3=&SelectVectorFromArray(
      Array Sampling Mode=::NewEnumerator0
      Direct Array Index=?
      Vector Selection Array=DataInterface() {propertyValues {FloatData [
         <12.5753 -0.9357 5.3711>
         <11.4768 -3.5801 4.7877>
         <10.4072 -4.661 3.9741>
         <10.3941 -3.6118 4.8616>
         <8.3624 -3.4524 5.067>
         <10.179 -1.2245 5.4514>
         <11.6392 -0.7666 5.3651>
         <9.5792 -1.0686 5.4388>
         <9.3529 -1.4062 5.4848>
         <9.6719 -5.7638 2.371>
         <8.7494 -5.5545 2.8817>
         <9.9246 -5.2157 3.3325>
         <8.4773 -5.1846 3.4957>
         <7.7346 -0.544 5.4084>
         <7.545 -0.1755 5.2551>
         <7.9089 0.5701 4.8986>
         <6.9144 0.9922 4.5897>
         <9.5032 1.2653 4.574>
         <10.9502 3.0557 3.2685>
         <9.2484 2.2857 3.828>
         <9.3455 3.0628 3.0993>
         <5.3827 2.822 3.0659>
         <6.1413 4.6248 1.5062>
         <6.0572 5.847 0.4026>
         <5.4956 6.1144 0.2805>
        ]}}
      Random Seed=0
     )
     4=&Multiply_Vector(
      A=&SelectVectorFromArray(
       Array Sampling Mode=::NewEnumerator0
       Direct Array Index=?
       Vector Selection Array=DataInterface() {propertyValues {FloatData [
          <12.5753 -0.9357 5.3711>
          <11.4768 -3.5801 4.7877>
          <10.4072 -4.661 3.9741>
          <10.3941 -3.6118 4.8616>
          <8.3624 -3.4524 5.067>
          <10.179 -1.2245 5.4514>
          <11.6392 -0.7666 5.3651>
          <9.5792 -1.0686 5.4388>
          <9.3529 -1.4062 5.4848>
          <9.6719 -5.7638 2.371>
          <8.7494 -5.5545 2.8817>
          <9.9246 -5.2157 3.3325>
          <8.4773 -5.1846 3.4957>
          <7.7346 -0.544 5.4084>
          <7.545 -0.1755 5.2551>
          <7.9089 0.5701 4.8986>
          <6.9144 0.9922 4.5897>
          <9.5032 1.2653 4.574>
          <10.9502 3.0557 3.2685>
          <9.2484 2.2857 3.828>
          <9.3455 3.0628 3.0993>
          <5.3827 2.822 3.0659>
          <6.1413 4.6248 1.5062>
          <6.0572 5.847 0.4026>
          <5.4956 6.1144 0.2805>
         ]}}
       Random Seed=0
      )
      B=<-1 -1 -1>
     )
     Index=$Particles.PartitionParticles.Partition
     Num=::NewEnumerator1
    )
    Skeletal Mesh=$System.Pearl_Skeleton
   ) {module=SkeletalMeshLocation002}
   &ScaleSpriteSize(Uniform Curve Scale=&FNTimeOfDayScalar(Night Scale=1.25, Morning Scale=1, Evening Scale=1)) {module=ScaleSpriteSize001}
   &ScaleColor(
    Scale Alpha=&Multiply_Float(
     A=&FloatFromCurve(
      FloatCurve=DataInterface() {propertyValues {Curve {keys [
          {interpMode=RCIM_Cubic, tangentMode=RCTM_User, time=0, value=0}
          {arriveTangent=-2.4058, interpMode=RCIM_Cubic, leaveTangent=-2.4058, tangentMode=RCTM_User, time=0.26, value=0.8421}
          {arriveTangent=-1.1775, interpMode=RCIM_Cubic, leaveTangent=-1.1775, time=0.4613, value=0.0645}
          {arriveTangent=-0.1197, interpMode=RCIM_Cubic, leaveTangent=-0.1197, time=0.5898, value=0.4537}
          <at 1 0 RCIM_Cubic>
         ]}}}
      Scale Curve=&RandomRangeFloat(Minimum=0.1, Maximum=0.5)
     )
     B=&FloatFromCurve(
      FloatCurve=DataInterface() {propertyValues {Curve {keys [{tangentMode=RCTM_User, time=0, value=0}, {tangentMode=RCTM_User, time=1, value=1}]}}}
      Scale Curve=&FloatFromCurve(
       FloatCurve=DataInterface() {propertyValues {Curve {keys [
           {arriveTangentWeight=0.1667, interpMode=RCIM_Cubic, leaveTangentWeight=0.1667, tangentMode=RCTM_User, time=1.4, value=0}
           {tangentMode=RCTM_User, time=2, value=1}
          ]}}}
       CurveIndex=$Emitter.Age
      )
     )
    )
    Scale RGB=&VectorFromFloat(Value=&Multiply_Float(
      A=&RandomRangeFloat(Minimum=0.7, Maximum=0.9, Evaluation Type=::NewEnumerator1)
      B=&SelectFloatFromArray(Direct Array Index=$Engine.QualityLevel, Float Selection Array=DataInterface() {propertyValues {FloatData [10,5,1,1,1]}})
     ))
    ScaleA=true
    ScaleRGB=true
   ) {module=ScaleColor001}
   &ScaleColor(
    Scale Alpha=$System.TODAlpha
    Scale RGB=&MakeVectorFromLinearColorRGB(Color=$System.TODColor)
    ScaleA=true
    ScaleRGB=true
   )
   /SetVariables(
    Particles.CenterFlareColor:LinearColor=&MultiplyLinearColor(
     Linear Color=$Particles.Color
     Scale Factor=&MakeVector4_FromLinearColor(LinearColor=&LerpLinearColors(StartColor=<rgba 8.0099 9.8007 10 1>, EndColor=<rgba 3.4218 8.799 10 1>, LerpFactor=$User.Variant))
    )
    Particles.CenterFlareSize:Vector2f=&Multiply_Vector2(A=$Particles.SpriteSize, B=&Vector2DFromFloat(Value=0.5))
    Particles.CoronaFlareColor:LinearColor=&MultiplyLinearColor(
     Linear Color=$Particles.Color
     Scale Factor=&MakeVector4_FromLinearColor(LinearColor=&LerpLinearColors(StartColor=<rgba 0.15 0.15 0.15 1>, EndColor=<rgba 0.0729 0.0729 0.0729 1>, LerpFactor=$User.Variant))
    )
    Particles.CoronaFlareSize:Vector2f=&Multiply_Vector2(A=$Particles.Initial.SpriteSize, B=&Vector2DFromFloat(Value=0.75))
    Particles.FlickerFlareColor:LinearColor=&MultiplyLinearColor(
     Linear Color=$Particles.Color
     Scale Factor=&MakeVector4_FromLinearColor(LinearColor=&LerpLinearColors(StartColor=<rgba 0.9441 0.8079 1 1>, EndColor=<rgba 0.0393 0.0337 0.0417 50>, LerpFactor=$User.Variant))
    )
   ) {isSetParameters=true, module=SetVariables_#3}
  ]
  rendererClasses [&NiagaraSpriteRendererProperties]
  renderers [
   {rendererClass=&NiagaraSpriteRendererProperties}
   {rendererClass=&NiagaraSpriteRendererProperties, rendererIndex=1}
   {rendererClass=&NiagaraSpriteRendererProperties, rendererIndex=2}
  ]
 }
}
summary {
 systemName=NS_PiperShelf_Character_Pearl_Idle
 emitters [
  {emitterName=Sparkles_UpperBody, rendererClasses [&NiagaraSpriteRendererProperties]}
  {emitterName=Sparkles_Feet, rendererClasses [&NiagaraSpriteRendererProperties]}
 ]
 userVariables [
  User.bisFootHidden {type=NiagaraBool}
  User.bIsFrontEnd {type=NiagaraBool}
  User.bIsFrontEndPreview {type=NiagaraBool}
  User.Lifetime_Scale {type=NiagaraFloat, defaultValue=0, description=""}
  User.Variant {type=NiagaraFloat, defaultValue=0, description=""}
 ]
}
