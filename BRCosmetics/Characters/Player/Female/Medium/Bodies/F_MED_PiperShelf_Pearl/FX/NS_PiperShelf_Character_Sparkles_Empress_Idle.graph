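# Niagara system for idle sparkles on the Piper Shelf Pearl (Empress) character's body.
# Emitters partition short-lived sprites across skeleton bones, placing them at hand-authored offset arrays, negated for paired bones.
# User hand and foot hidden flags gate spawning or switch offsets, and camera offset differs between front end and gameplay.
# Flicker curves, a delayed emitter-age fade-in, quality-level brightness and time-of-day tint drive color for layered center, corona and flicker flares.
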NiagaraSystem /BRCosmetics/Characters/Player/Female/Medium/Bodies/F_MED_PiperShelf_Pearl/FX/NS_PiperShelf_Character_Sparkles_Empress_Idle
dependencies {
 usedDataInterfaces [
  &NiagaraDataInterfaceArrayFloat3
  &NiagaraDataInterfaceCurve
  &NiagaraDataInterfaceArrayFloat
  &NiagaraDataInterfaceSkeletalMesh
  &NiagaraDataInterfaceColorCurve
 ]
 usedDynamicInputs [
  &Multiply_Vector2D_ByFloat
  &Lerp_Float
  &MakeVector2D
 ]
 usedModules [
  &self:SystemSpawnScript.SystemScriptSource.SystemScriptGraph.NiagaraNodeAssignment_2.SetVariables_#1
  &SystemState
  &self:SystemSpawnScript.SystemScriptSource.SystemScriptGraph.NiagaraNodeAssignment_1.SetVariables_#2
  &/TimeOfDayGlobal
 ]
}
emitters {
 Hands_LowerArm_Sparkles {
  emitterUpdateScript [
   &EmitterState()
   &SpawnRate(
    Spawn Probability=&MakeCustomFloatFromBool(Bool=$User.bIsHandHidden, True Float=0, False Float=$System.SpawnProbability)
    SpawnRate=&Multiply_Float(A=$System.SpawnRate, B=1.5)
    Use Spawn Probability=true
   )
  ]
  particleSpawnScript [
   &PartitionParticles(Number Of Partitions=4)
   &InitializeParticle(
    AdjustAlpha=false
    AdjustHue=true
    AdjustSaturation=false
    AdjustValue=false
    Color=<rgba 0.843 0.8 1 1>
    Color Mode=::NewEnumerator3
    Lifetime Max=0.4
    Lifetime Min=0.3
    Lifetime Mode=::NewEnumerator1
    Sprite Size Mode=::NewEnumerator3
    Uniform Sprite Size=&Multiply_Float(A=40, B=&RandomRangeFloat(Minimum=0.75, Maximum=1))
    UsePositionOffset=false
    Write Material Random=true
   )
   &CameraOffset(Camera Offset Amount=&FN_FrontEndFloatModifier(
     Gameplay=&MergeHelper_FloatSwitch(0=4, 1=4, 2=4, 3=4, Index=$Particles.PartitionParticles.Partition, Num=::NewEnumerator2)
     Front End=&MergeHelper_FloatSwitch(0=3, 1=3, 2=3, 3=3, Index=$Particles.PartitionParticles.Partition, Num=::NewEnumerator2)
    ))
  ]
  particleUpdateScript [
   &ParticleState()
   &SkeletalMeshLocation(
    Bone / Socket Index=&MergeHelper_IntSwitch(0=6, 1=7, 2=8, 3=9, 4=?, 5=?, Index=$Particles.PartitionParticles.Partition, Num=::NewEnumerator2)
    Bone Sampling Mode=::NewEnumerator4
    MaskSpawnGroup=false
    Sampled Position Offset=&MergeHelper_VectorSwitch(
     0=&SelectVectorFromArray(
      Array Sampling Mode=::NewEnumerator0
      Direct Array Index=?
      Vector Selection Array=DataInterface() {propertyValues {FloatData [
         <17.0065 0.2755 -0.77>
         <3.2355 1.5733 2.9245>
         <6.2383 1.9629 -1.5>
         <-0.7279 3.4165 -2.0094>
         <-3.2654 -1.7219 1.614>
         <-0.5543 -2.2135 2.1404>
         <11.2237 -2.0841 -0.3281>
         <8.1481 -2.7762 0.1657>
         <8.5581 -2.3253 1.7338>
         <8.6436 -2.2336 1.8475>
         <9.85 0.9547 2>
        ]}}
      Random Seed=0
     )
     1=&Multiply_Vector(
      A=&SelectVectorFromArray(
       Array Sampling Mode=::NewEnumerator0
       Direct Array Index=?
       Vector Selection Array=DataInterface() {propertyValues {FloatData [
          <17.0065 0.2755 -0.77>
          <3.2355 1.5733 2.9245>
          <6.2383 1.9629 -1.5>
          <-0.7279 3.4165 -2.0094>
          <-3.2654 -1.7219 1.614>
          <-0.5543 -2.2135 2.1404>
          <11.2237 -2.0841 -0.3281>
          <8.1481 -2.7762 0.1657>
          <8.5581 -2.3253 1.7338>
          <8.6436 -2.2336 1.8475>
          <9.85 0.9547 2>
         ]}}
       Random Seed=0
      )
      B=<-1 -1 -1>
     )
     2=&SelectVectorFromArray(
      Array Sampling Mode=::NewEnumerator0
      Direct Array Index=?
      Vector Selection Array=DataInterface() {propertyValues {FloatData [
         <-1.5173 0.1051 -1.404>
         <0.7405 -0.187 -3.0162>
         <0.8485 -0.8853 0.5135>
         <2.0365 -0.7992 -3.378>
         <2.4306 -0.225 -0.087>
         <5.3848 -0.6851 1.719>
         <4.4317 -2.2879 -2.9193>
         <4.6687 -3.8049 -4.9755>
         <6.8585 -0.6734 3.3238>
         <5.5586 0.2819 -0.4461>
        ]}}
      Random Seed=0
     )
     3=&Multiply_Vector(
      A=&SelectVectorFromArray(
       Array Sampling Mode=::NewEnumerator0
       Direct Array Index=?
       Vector Selection Array=DataInterface() {propertyValues {FloatData [
          <-1.5173 0.1051 -1.404>
          <0.7405 -0.187 -3.0162>
          <0.8485 -0.8853 0.5135>
          <2.0365 -0.7992 -3.378>
          <2.4306 -0.225 -0.087>
          <5.3848 -0.6851 1.719>
          <4.4317 -2.2879 -2.9193>
          <4.6687 -3.8049 -4.9755>
          <6.8585 -0.6734 3.3238>
          <5.5586 0.2819 -0.4461>
         ]}}
       Random Seed=0
      )
      B=<-1 -1 -1>
     )
     4=&SelectVectorFromArray(
      Array Sampling Mode=::NewEnumerator0
      Direct Array Index=?
      Vector Selection Array=DataInterface() {propertyValues {FloatData [
         <2.0538 0.6915 -3.1465>
         <-0.6458 -4.0657 -1.7933>
         <-2.4306 -3.6897 -2.0127>
         <-1.4009 2.9113 -3.04>
         <-2.8843 -2.7666 -2.7626>
         <-6.7583 -2.3574 -2.014>
         <-12.2108 -2.9032 -0.9181>
         <-11.5524 0.0505 -2.4086>
         <-12.2508 0.0325 -2.4065>
         <-12.509 2.662 -2.2432>
         <-6.5784 3.8685 -1.4702>
         <-4.3198 4.2067 -1.2819>
         <-2.1427 4.3437 -1.114>
         <-9.8142 2.0126 4.7172>
         <-9.8039 4.0174 3.4179>
         <-10.0994 4.5378 2.4588>
         <-10.5067 4.7753 1.3395>
         <-20.0626 2.7474 2.9682>
         <-23.6875 3.8548 -1.373>
         <-18.2593 3.9881 -0.8462>
         <-14.6003 4.2671 2.3565>
         <-14.0341 3.5045 -1.716>
         <-14.9187 -0.6381 -2.3052>
         <-9.5597 -3.4796 0.2778>
         <-20.4585 0.5013 -3.0081>
         <-17.3073 -1.875 -1.9054>
         <-17.2355 -3.4648 0.1428>
         <-14.8715 -3.5452 0.1999>
         <-16.1548 -3.1647 2.086>
         <-12.8954 -1.7856 3.6732>
         <-20.7396 -1.3291 2.9485>
         <-21.1413 -3.1136 -0.1032>
         <-23.7276 0.0516 -3.6666>
         <-11.2198 -0.9259 4.1843>
        ]}}
      Random Seed=0
     )
     5=&Multiply_Vector(
      A=&SelectVectorFromArray(
       Array Sampling Mode=::NewEnumerator0
       Direct Array Index=?
       Vector Selection Array=DataInterface() {propertyValues {FloatData [
          <2.0538 0.6915 -3.1465>
          <-0.6458 -4.0657 -1.7933>
          <-2.4306 -3.6897 -2.0127>
          <-1.4009 2.9113 -3.04>
          <-2.8843 -2.7666 -2.7626>
          <-6.7583 -2.3574 -2.014>
          <-12.2108 -2.9032 -0.9181>
          <-11.5524 0.0505 -2.4086>
          <-12.2508 0.0325 -2.4065>
          <-12.509 2.662 -2.2432>
          <-6.5784 3.8685 -1.4702>
          <-4.3198 4.2067 -1.2819>
          <-2.1427 4.3437 -1.114>
          <-9.8142 2.0126 4.7172>
          <-9.8039 4.0174 3.4179>
          <-10.0994 4.5378 2.4588>
          <-10.5067 4.7753 1.3395>
          <-20.0626 2.7474 2.9682>
          <-23.6875 3.8548 -1.373>
          <-18.2593 3.9881 -0.8462>
          <-14.6003 4.2671 2.3565>
          <-14.0341 3.5045 -1.716>
          <-14.9187 -0.6381 -2.3052>
          <-9.5597 -3.4796 0.2778>
          <-20.4585 0.5013 -3.0081>
          <-17.3073 -1.875 -1.9054>
          <-17.2355 -3.4648 0.1428>
          <-14.8715 -3.5452 0.1999>
          <-16.1548 -3.1647 2.086>
          <-12.8954 -1.7856 3.6732>
          <-20.7396 -1.3291 2.9485>
          <-21.1413 -3.1136 -0.1032>
          <-23.7276 0.0516 -3.6666>
          <-11.2198 -0.9259 4.1843>
         ]}}
       Random Seed=0
      )
      B=<-1 -1 -1>
     )
     6=&SelectVectorFromArray(
      Array Sampling Mode=::NewEnumerator0
      Direct Array Index=?
      Vector Selection Array=DataInterface() {propertyValues {FloatData [
         <17.0065 0.2755 -0.77>
         <3.2355 1.5733 2.9245>
         <6.2383 1.9629 -1.5>
         <-0.7279 3.4165 -2.0094>
         <-3.2654 -1.7219 1.614>
         <-0.5543 -2.2135 2.1404>
         <11.2237 -2.0841 -0.3281>
         <8.1481 -2.7762 0.1657>
         <8.5581 -2.3253 1.7338>
         <8.6436 -2.2336 1.8475>
         <9.85 0.9547 2>
        ]}}
      Random Seed=0
     )
     7=&Multiply_Vector(
      A=&SelectVectorFromArray(
       Array Sampling Mode=::NewEnumerator0
       Direct Array Index=?
       Vector Selection Array=DataInterface() {propertyValues {FloatData [
          <17.0065 0.2755 -0.77>
          <3.2355 1.5733 2.9245>
          <6.2383 1.9629 -1.5>
          <-0.7279 3.4165 -2.0094>
          <-3.2654 -1.7219 1.614>
          <-0.5543 -2.2135 2.1404>
          <11.2237 -2.0841 -0.3281>
          <8.1481 -2.7762 0.1657>
          <8.5581 -2.3253 1.7338>
          <8.6436 -2.2336 1.8475>
          <9.85 0.9547 2>
         ]}}
       Random Seed=0
      )
      B=<-1 -1 -1>
     )
     8=&SelectVectorFromArray(
      Array Sampling Mode=::NewEnumerator0
      Direct Array Index=?
      Vector Selection Array=DataInterface() {propertyValues {FloatData [
         <-1.5173 0.1051 -1.404>
         <0.7405 -0.187 -3.0162>
         <0.8485 -0.8853 0.5135>
         <2.0365 -0.7992 -3.378>
         <2.4306 -0.225 -0.087>
         <5.3848 -0.6851 1.719>
         <4.4317 -2.2879 -2.9193>
         <4.6687 -3.8049 -4.9755>
         <6.8585 -0.6734 3.3238>
         <5.5586 0.2819 -0.4461>
        ]}}
      Random Seed=0
     )
     9=&Multiply_Vector(
      A=&SelectVectorFromArray(
       Array Sampling Mode=::NewEnumerator0
       Direct Array Index=?
       Vector Selection Array=DataInterface() {propertyValues {FloatData [
          <-1.5173 0.1051 -1.404>
          <0.7405 -0.187 -3.0162>
          <0.8485 -0.8853 0.5135>
          <2.0365 -0.7992 -3.378>
          <2.4306 -0.225 -0.087>
          <5.3848 -0.6851 1.719>
          <4.4317 -2.2879 -2.9193>
          <4.6687 -3.8049 -4.9755>
          <6.8585 -0.6734 3.3238>
          <5.5586 0.2819 -0.4461>
         ]}}
       Random Seed=0
      )
      B=<-1 -1 -1>
     )
     Index=$Particles.PartitionParticles.Partition
     Num=::NewEnumerator2
    )
    Skeletal Mesh=$System.Pearl_Skeleton
   ) {module=SkeletalMeshLocation002}
   &ScaleSpriteSize() {module=ScaleSpriteSize001}
   &ScaleColor(
    Scale Alpha=&Multiply_Float(
     A=&FloatFromCurve(
      FloatCurve=DataInterface() {propertyValues {Curve {keys [
          {interpMode=RCIM_Cubic, tangentMode=RCTM_User, time=0, value=0}
          {arriveTangent=-2.4058, interpMode=RCIM_Cubic, leaveTangent=-2.4058, tangentMode=RCTM_User, time=0.26, value=0.8421}
          {arriveTangent=-1.1775, interpMode=RCIM_Cubic, leaveTangent=-1.1775, time=0.4613, value=0.0645}
          {arriveTangent=-0.1197, interpMode=RCIM_Cubic, leaveTangent=-0.1197, time=0.5898, value=0.4537}
          <at 1 0 RCIM_Cubic>
         ]}}}
      Scale Curve=&RandomRangeFloat(Minimum=0.1, Maximum=0.5)
     )
     B=&FloatFromCurve(
      FloatCurve=DataInterface() {propertyValues {Curve {keys [{tangentMode=RCTM_User, time=0, value=0}, {tangentMode=RCTM_User, time=1, value=1}]}}}
      Scale Curve=&FloatFromCurve(
       FloatCurve=DataInterface() {propertyValues {Curve {keys [
           {arriveTangentWeight=0.1667, interpMode=RCIM_Cubic, leaveTangentWeight=0.1667, tangentMode=RCTM_User, time=1.4, value=0}
           {tangentMode=RCTM_User, time=2, value=1}
          ]}}}
       CurveIndex=$Emitter.Age
      )
     )
    )
    Scale RGB=&VectorFromFloat(Value=&Multiply_Float(
      A=&RandomRangeFloat(Minimum=0.6, Maximum=1, Evaluation Type=::NewEnumerator1)
      B=&SelectFloatFromArray(Direct Array Index=$Engine.QualityLevel, Float Selection Array=DataInterface() {propertyValues {FloatData [10,5,1,1,1]}})
     ))
    ScaleA=true
    ScaleRGB=true
   ) {module=ScaleColor001}
   &ScaleColor(
    Scale Alpha=$System.TODAlpha
    Scale RGB=&MakeVectorFromLinearColorRGB(Color=$System.TODColor)
    ScaleA=true
    ScaleRGB=true
   )
   /SetVariables(
    Particles.CenterFlareColor:LinearColor=&MultiplyLinearColor(
     Linear Color=$Particles.Color
     Scale Factor=&MakeVector4_FromLinearColor(LinearColor=<rgba 8.0099 9.8007 10 1>)
    )
    Particles.CenterFlareSize:Vector2f=&Multiply_Vector2(A=$Particles.SpriteSize, B=&Vector2DFromFloat(Value=0.5))
    Particles.CoronaFlareColor:LinearColor=&MultiplyLinearColor(
     Linear Color=$Particles.Color
     Scale Factor=&MakeVector4_FromLinearColor(LinearColor=<rgba 0.15 0.15 0.15 1>)
    )
    Particles.CoronaFlareSize:Vector2f=&Multiply_Vector2(A=$Particles.Initial.SpriteSize, B=&Vector2DFromFloat(Value=1))
    Particles.FlickerFlareColor:LinearColor=&MultiplyLinearColor(
     Linear Color=$Particles.Color
     Scale Factor=&MakeVector4_FromLinearColor(LinearColor=<rgba 0.9441 0.8079 1 1>)
    )
   ) {isSetParameters=true, module=SetVariables_#3}
  ]
  rendererClasses [&NiagaraSpriteRendererProperties]
  renderers [
   {rendererClass=&NiagaraSpriteRendererProperties}
   {rendererClass=&NiagaraSpriteRendererProperties, rendererIndex=1}
   {rendererClass=&NiagaraSpriteRendererProperties, rendererIndex=2}
  ]
 }
 LowerBody_Head_Sparkles {
  emitterUpdateScript [
   &EmitterState()
   &SpawnRate(
    Spawn Probability=&MakeCustomFloatFromBool(Bool=$User.bisFootHidden, True Float=0, False Float=$System.SpawnProbability)
    SpawnRate=&Multiply_Float(A=$System.SpawnRate, B=2)
    Use Spawn Probability=true
   )
  ]
  particleSpawnScript [
   &PartitionParticles(Number Of Partitions=6)
   &InitializeParticle(
    AdjustAlpha=false
    AdjustHue=true
    AdjustSaturation=false
    AdjustValue=false
    Color=<rgba 0.843 0.8 1 1>
    Color Mode=::NewEnumerator3
    Lifetime Max=0.4
    Lifetime Min=0.3
    Lifetime Mode=::NewEnumerator1
    Sprite Size Mode=::NewEnumerator3
    Uniform Sprite Size=&Multiply_Float(A=40, B=&RandomRangeFloat(Minimum=0.75, Maximum=1))
    UsePositionOffset=false
    Write Material Random=true
   )
   &CameraOffset(Camera Offset Amount=&FN_FrontEndFloatModifier(
     Gameplay=&MergeHelper_FloatSwitch(0=8, 1=8, 2=8, 3=8, 4=4, 5=4, 6=4, Index=$Particles.PartitionParticles.Partition, Num=::NewEnumerator5)
     Front End=&MergeHelper_FloatSwitch(0=3, 1=3, 2=3, 3=3, 4=2, 5=2, 6=2, Index=$Particles.PartitionParticles.Partition, Num=::NewEnumerator5)
    ))
  ]
  particleUpdateScript [
   &ParticleState()
   &SkeletalMeshLocation(
    Bone / Socket Index=&MergeHelper_IntSwitch(0=10, 1=11, 2=12, 3=13, 4=14, 5=15, Index=$Particles.PartitionParticles.Partition, Num=::NewEnumerator4)
    Bone Sampling Mode=::NewEnumerator4
    MaskSpawnGroup=false
    Sampled Position Offset=&MergeHelper_VectorSwitch(
     0=&SelectVectorFromArray(
      Array Sampling Mode=::NewEnumerator0
      Direct Array Index=?
      Vector Selection Array=DataInterface() {propertyValues {FloatData [
         <-9.5011 -7.101 8.3335>
         <-7.8626 -10.4662 7.0362>
         <-3.5916 -11.9308 8.8412>
         <-1.2405 -11.9006 8.0091>
         <-0.3188 -10.4972 -0.2352>
         <-3.9917 -9.2248 -2.935>
         <-6.2369 -8.6736 -3.4726>
         <-7.7466 -7.6189 -4.6054>
         <-10.4565 -7.747 -3.7401>
         <-8.2793 -10.8096 4.8908>
         <-9.7047 -9.7469 0.7749>
         <-9.6833 -10.025 1.8191>
         <-11.2928 -8.7171 0.0117>
         <-14.3592 -7.9845 -1.8384>
         <-16.0147 -3.8523 7.4032>
         <-18.4519 -7.7997 0.824>
         <-16.5627 -7.0835 4.4511>
         <-22.6199 -6.2812 4.3659>
         <-22.5789 -6.6713 3.419>
         <-19.7382 -6.3003 5.09>
         <-22.9862 -3.7865 6.5175>
         <-23.5423 -3.8138 6.4205>
         <-25.6543 -4.1805 5.6618>
         <-24.1581 -3.8166 6.3049>
         <-26.6486 -2.5894 6.1861>
         <-10.0371 -2.2112 9.0726>
         <-11.387 -0.3688 8.4281>
         <-12.1143 -2.7924 8.2052>
         <-10.1116 2.3955 7.6971>
         <-15.8873 -1.3758 8.1105>
         <-13.1581 1.9832 7.6369>
         <-13.0095 3.5757 6.7901>
         <-19.0459 -0.5956 7.66>
         <-21.6593 -0.1487 7.1639>
         <-27.0163 0.1199 6.4232>
         <-23.5677 2.6485 5.9957>
         <-25.5414 2.3009 6.0048>
         <-28.372 3.7524 4.9132>
         <-29.1917 1.7277 5.8842>
         <-23.9504 5.1588 4.1042>
         <-20.4059 4.4383 5.3449>
         <-19.1482 7.2219 0.7659>
         <-17.6118 7.2524 0.7025>
         <-14.6546 6.0066 3.848>
         <-13.6231 4.4313 5.9866>
         <-12.1285 5.6515 4.2966>
         <-12.6156 6.7773 1.7439>
         <-13.6073 6.8217 -3.6346>
         <-11.9774 7.2497 -0.4189>
         <-9.7416 6.3224 2.6673>
         <-9.2426 5.4938 4.4497>
         <-8.1365 6.1282 -3.5462>
         <-7.0487 5.4343 4.6496>
         <-7.5342 4.9801 6.3086>
         <-6.9896 5.2757 5.318>
         <-4.6719 6.5439 8.0261>
         <-5.2449 5.7596 -3.0092>
         <0.31 5.8941 -0.8948>
         <0.2973 6.6571 1.2886>
         <-37.6541 -3.746 4.339>
         <-33.5972 -2.2992 5.3567>
         <-34.3449 -4.1678 4.1235>
         <-30.169 -4.4425 4.539>
         <-27.4974 -6.4626 2.4659>
         <-25.343 -6.7248 2.4487>
         <-27.2159 -6 -2.6542>
         <-31.1827 -6.162 -0.3374>
         <-23.1859 -5.8842 -3.8446>
         <-20.2074 -6.9963 -2.9875>
         <-0.5713 -6.6333 -4.6668>
         <0.8562 2.8963 -5.2687>
         <0.741 2.4439 -5.5658>
         <-5.8455 4.9146 -5.351>
         <-8.01 2.4905 -7.2539>
         <-5.8469 1.3932 -7.2726>
         <-3.7271 1.7012 -7.0625>
         <-2.3476 1.3966 -6.8909>
         <-6.0539 -3.0343 -7.2102>
         <-1.6814 -4.6839 -5.8932>
         <-7.5425 0.4264 -7.6367>
         <-10.6191 -4.4262 -7.1025>
         <-11.4933 -1.505 -7.8564>
         <-11.7186 0.9154 -7.9041>
         <-12.9362 -1.3969 -7.8977>
         <-16.356 -3.7309 -6.8478>
         <-23.7803 -1.799 -6.547>
         <-18.6008 1.2126 -7.601>
         <-26.6551 -2.6418 -5.7314>
         <-21.4877 -0.7181 -7.0396>
         <-15.9328 2.6537 -7.3416>
         <-22.2543 1.4748 -7.1758>
         <-12.9762 5.8693 -5.6476>
         <-15.5748 4.8427 -6.4844>
         <-22.3456 5.6604 -5.2688>
         <-22.3176 6.925 -3.2316>
         <-24.1257 6.7135 -3.276>
         <-22.4925 7.5349 -0.9795>
         <-24.7182 7.2067 0.6833>
         <-26.5204 7.1185 -0.1577>
         <-33.1891 4.7901 3.1511>
         <-34.0012 1.3814 5.4839>
         <-35.3591 2.0098 4.9184>
         <-37.446 1.8284 4.716>
         <-39.092 2.1744 4.1786>
         <-40.2632 -0.0937 5.0668>
         <-36.8984 4.1468 2.9042>
         <-36.8553 5.4454 -0.5958>
         <-35.2618 5.7679 0.1288>
         <-28.8319 6.6867 -2.0049>
         <-27.0483 6.7153 -2.6679>
         <-33.9958 5.5914 -2.7024>
         <-34.6375 5.8803 -1.4407>
         <-37.0013 4.9533 -2.5268>
         <-44.0636 3.9471 1.2993>
         <-42.1165 4.7353 -1.3271>
         <-44.5605 3.5931 -2.4286>
         <-36.947 1.4267 -4.9996>
         <-27.8495 4.0748 -5.7069>
         <-36.2509 0.0642 -5.1638>
         <-28.3496 3.0448 -5.9781>
         <-33.7488 0.5184 -5.5048>
         <-34.5032 -3.4231 -3.9042>
         <-38.3464 -4.0436 -3.1163>
         <-38.3168 -4.8402 -2.0307>
         <-38.455 -4.9726 -1.724>
         <-41.634 -5.4571 0.444>
         <-41.3907 -3.9536 -3.255>
         <-44.6488 1.6876 -3.8329>
        ]}}
      Random Seed=0
     )
     1=&Multiply_Vector(
      A=&SelectVectorFromArray(
       Array Sampling Mode=::NewEnumerator0
       Direct Array Index=?
       Vector Selection Array=DataInterface() {propertyValues {FloatData [
          <-9.5011 -7.101 8.3335>
          <-7.8626 -10.4662 7.0362>
          <-3.5916 -11.9308 8.8412>
          <-1.2405 -11.9006 8.0091>
          <-0.3188 -10.4972 -0.2352>
          <-3.9917 -9.2248 -2.935>
          <-6.2369 -8.6736 -3.4726>
          <-7.7466 -7.6189 -4.6054>
          <-10.4565 -7.747 -3.7401>
          <-8.2793 -10.8096 4.8908>
          <-9.7047 -9.7469 0.7749>
          <-9.6833 -10.025 1.8191>
          <-11.2928 -8.7171 0.0117>
          <-14.3592 -7.9845 -1.8384>
          <-16.0147 -3.8523 7.4032>
          <-18.4519 -7.7997 0.824>
          <-16.5627 -7.0835 4.4511>
          <-22.6199 -6.2812 4.3659>
          <-22.5789 -6.6713 3.419>
          <-19.7382 -6.3003 5.09>
          <-22.9862 -3.7865 6.5175>
          <-23.5423 -3.8138 6.4205>
          <-25.6543 -4.1805 5.6618>
          <-24.1581 -3.8166 6.3049>
          <-26.6486 -2.5894 6.1861>
          <-10.0371 -2.2112 9.0726>
          <-11.387 -0.3688 8.4281>
          <-12.1143 -2.7924 8.2052>
          <-10.1116 2.3955 7.6971>
          <-15.8873 -1.3758 8.1105>
          <-13.1581 1.9832 7.6369>
          <-13.0095 3.5757 6.7901>
          <-19.0459 -0.5956 7.66>
          <-21.6593 -0.1487 7.1639>
          <-27.0163 0.1199 6.4232>
          <-23.5677 2.6485 5.9957>
          <-25.5414 2.3009 6.0048>
          <-28.372 3.7524 4.9132>
          <-29.1917 1.7277 5.8842>
          <-23.9504 5.1588 4.1042>
          <-20.4059 4.4383 5.3449>
          <-19.1482 7.2219 0.7659>
          <-17.6118 7.2524 0.7025>
          <-14.6546 6.0066 3.848>
          <-13.6231 4.4313 5.9866>
          <-12.1285 5.6515 4.2966>
          <-12.6156 6.7773 1.7439>
          <-13.6073 6.8217 -3.6346>
          <-11.9774 7.2497 -0.4189>
          <-9.7416 6.3224 2.6673>
          <-9.2426 5.4938 4.4497>
          <-8.1365 6.1282 -3.5462>
          <-7.0487 5.4343 4.6496>
          <-7.5342 4.9801 6.3086>
          <-6.9896 5.2757 5.318>
          <-4.6719 6.5439 8.0261>
          <-5.2449 5.7596 -3.0092>
          <0.31 5.8941 -0.8948>
          <0.2973 6.6571 1.2886>
          <-37.6541 -3.746 4.339>
          <-33.5972 -2.2992 5.3567>
          <-34.3449 -4.1678 4.1235>
          <-30.169 -4.4425 4.539>
          <-27.4974 -6.4626 2.4659>
          <-25.343 -6.7248 2.4487>
          <-27.2159 -6 -2.6542>
          <-31.1827 -6.162 -0.3374>
          <-23.1859 -5.8842 -3.8446>
          <-20.2074 -6.9963 -2.9875>
          <-0.5713 -6.6333 -4.6668>
          <0.8562 2.8963 -5.2687>
          <0.741 2.4439 -5.5658>
          <-5.8455 4.9146 -5.351>
          <-8.01 2.4905 -7.2539>
          <-5.8469 1.3932 -7.2726>
          <-3.7271 1.7012 -7.0625>
          <-2.3476 1.3966 -6.8909>
          <-6.0539 -3.0343 -7.2102>
          <-1.6814 -4.6839 -5.8932>
          <-7.5425 0.4264 -7.6367>
          <-10.6191 -4.4262 -7.1025>
          <-11.4933 -1.505 -7.8564>
          <-11.7186 0.9154 -7.9041>
          <-12.9362 -1.3969 -7.8977>
          <-16.356 -3.7309 -6.8478>
          <-23.7803 -1.799 -6.547>
          <-18.6008 1.2126 -7.601>
          <-26.6551 -2.6418 -5.7314>
          <-21.4877 -0.7181 -7.0396>
          <-15.9328 2.6537 -7.3416>
          <-22.2543 1.4748 -7.1758>
          <-12.9762 5.8693 -5.6476>
          <-15.5748 4.8427 -6.4844>
          <-22.3456 5.6604 -5.2688>
          <-22.3176 6.925 -3.2316>
          <-24.1257 6.7135 -3.276>
          <-22.4925 7.5349 -0.9795>
          <-24.7182 7.2067 0.6833>
          <-26.5204 7.1185 -0.1577>
          <-33.1891 4.7901 3.1511>
          <-34.0012 1.3814 5.4839>
          <-35.3591 2.0098 4.9184>
          <-37.446 1.8284 4.716>
          <-39.092 2.1744 4.1786>
          <-40.2632 -0.0937 5.0668>
          <-36.8984 4.1468 2.9042>
          <-36.8553 5.4454 -0.5958>
          <-35.2618 5.7679 0.1288>
          <-28.8319 6.6867 -2.0049>
          <-27.0483 6.7153 -2.6679>
          <-33.9958 5.5914 -2.7024>
          <-34.6375 5.8803 -1.4407>
          <-37.0013 4.9533 -2.5268>
          <-44.0636 3.9471 1.2993>
          <-42.1165 4.7353 -1.3271>
          <-44.5605 3.5931 -2.4286>
          <-36.947 1.4267 -4.9996>
          <-27.8495 4.0748 -5.7069>
          <-36.2509 0.0642 -5.1638>
          <-28.3496 3.0448 -5.9781>
          <-33.7488 0.5184 -5.5048>
          <-34.5032 -3.4231 -3.9042>
          <-38.3464 -4.0436 -3.1163>
          <-38.3168 -4.8402 -2.0307>
          <-38.455 -4.9726 -1.724>
          <-41.634 -5.4571 0.444>
          <-41.3907 -3.9536 -3.255>
          <-44.6488 1.6876 -3.8329>
         ]}}
       Random Seed=0
      )
      B=<-1 -1 -1>
     )
     2=&SelectVectorFromArray(
      Array Sampling Mode=::NewEnumerator0
      Direct Array Index=?
      Vector Selection Array=DataInterface() {propertyValues {FloatData [
         <-12.9425 -6.387 3.6216>
         <-16.5827 -7.4896 1.6339>
         <-8.4828 -4.5327 1.6501>
         <-4.6066 -4.2273 1.6913>
         <-5.584 -5.1835 3.2259>
         <-3.9535 -5.0472 2.7323>
         <-5.2327 2.875 2.4717>
         <-6.7877 -1.4729 -1.63>
         <-6.7199 2.7064 1.6431>
         <-20.5875 -1.7211 2>
         <-16.7039 0.635 3.139>
         <-20.5856 -0.1211 2.463>
         <-22.7547 0.0254 1.9571>
         <-17.4784 1.2769 1.9876>
         <-5.8898 3.3702 0.9041>
         <-2.01 4.5774 -1.0952>
         <-5.3701 2.6413 -2.3868>
         <-4.4609 -5.8196 1.1597>
         <-5.1661 -5.4639 -1.6313>
         <-9.0792 -7.3158 0.0885>
         <-11.9489 -8.0682 0.9757>
         <-11.7825 -7.9909 -0.6143>
         <-9.111 -3.7766 -4.2836>
         <-7.8802 -1.5994 -4.2708>
         <-5.3425 -0.9814 -3.9311>
         <-12.5021 -0.8744 -4.6291>
         <-2.5447 0.9346 -3.6069>
         <-7.7018 0.9929 -3.6505>
         <-12.6316 2.7854 -0.7521>
         <-17.1669 0.8871 -3.5899>
         <-16.36 2.0544 -2.2486>
         <-15.1586 -8.4463 -1.2259>
         <-22.2134 -7.2388 0.3382>
         <-30.7004 -3.7103 2.5996>
         <-29.6611 -6.6044 -0.8972>
         <-25.1304 -5.9487 -3.1851>
         <-20.4869 -3.7431 -4.8596>
         <-21.603 -1.5774 -4.469>
         <-28.3677 -4.8929 -3.1995>
         <-24.546 -1.2731 -3.9224>
         <-25.3217 -0.2932 -3.2908>
         <-20.315 0.8983 -3.1685>
         <-17.6738 1.8422 -2.4468>
         <-21.7461 1.5705 0.8324>
         <-23.712 1.5507 0.5217>
         <-29.8463 0.7515 1.432>
         <-31.4608 1.0341 -0.073>
         <-30.3995 0.9309 -0.971>
         <-35.0209 0.219 -1.6537>
         <-32.6261 -5.3095 -1.81>
         <-33.5537 -5.2005 -1.6885>
         <-37.41 -5.825 0.5207>
         <-40.3806 0.0668 -2.1607>
        ]}}
      Random Seed=0
     )
     3=&Multiply_Vector(
      A=&SelectVectorFromArray(
       Array Sampling Mode=::NewEnumerator0
       Direct Array Index=?
       Vector Selection Array=DataInterface() {propertyValues {FloatData [
          <-12.9425 -6.387 3.6216>
          <-16.5827 -7.4896 1.6339>
          <-8.4828 -4.5327 1.6501>
          <-4.6066 -4.2273 1.6913>
          <-5.584 -5.1835 3.2259>
          <-3.9535 -5.0472 2.7323>
          <-5.2327 2.875 2.4717>
          <-6.7877 -1.4729 -1.63>
          <-6.7199 2.7064 1.6431>
          <-20.5875 -1.7211 2>
          <-16.7039 0.635 3.139>
          <-20.5856 -0.1211 2.463>
          <-22.7547 0.0254 1.9571>
          <-17.4784 1.2769 1.9876>
          <-5.8898 3.3702 0.9041>
          <-2.01 4.5774 -1.0952>
          <-5.3701 2.6413 -2.3868>
          <-4.4609 -5.8196 1.1597>
          <-5.1661 -5.4639 -1.6313>
          <-9.0792 -7.3158 0.0885>
          <-11.9489 -8.0682 0.9757>
          <-11.7825 -7.9909 -0.6143>
          <-9.111 -3.7766 -4.2836>
          <-7.8802 -1.5994 -4.2708>
          <-5.3425 -0.9814 -3.9311>
          <-12.5021 -0.8744 -4.6291>
          <-2.5447 0.9346 -3.6069>
          <-7.7018 0.9929 -3.6505>
          <-12.6316 2.7854 -0.7521>
          <-17.1669 0.8871 -3.5899>
          <-16.36 2.0544 -2.2486>
          <-15.1586 -8.4463 -1.2259>
          <-22.2134 -7.2388 0.3382>
          <-30.7004 -3.7103 2.5996>
          <-29.6611 -6.6044 -0.8972>
          <-25.1304 -5.9487 -3.1851>
          <-20.4869 -3.7431 -4.8596>
          <-21.603 -1.5774 -4.469>
          <-28.3677 -4.8929 -3.1995>
          <-24.546 -1.2731 -3.9224>
          <-25.3217 -0.2932 -3.2908>
          <-20.315 0.8983 -3.1685>
          <-17.6738 1.8422 -2.4468>
          <-21.7461 1.5705 0.8324>
          <-23.712 1.5507 0.5217>
          <-29.8463 0.7515 1.432>
          <-31.4608 1.0341 -0.073>
          <-30.3995 0.9309 -0.971>
          <-35.0209 0.219 -1.6537>
          <-32.6261 -5.3095 -1.81>
          <-33.5537 -5.2005 -1.6885>
          <-37.41 -5.825 0.5207>
          <-40.3806 0.0668 -2.1607>
         ]}}
       Random Seed=0
      )
      B=<-1 -1 -1>
     )
     4=&SelectVectorFromArray(
      Array Sampling Mode=::NewEnumerator0
      Direct Array Index=?
      Vector Selection Array=DataInterface() {propertyValues {FloatData [
         <2.5 2.5 -2.1846>
         <-2.6104 5.8649 -2.8753>
         <-6.7896 13.3872 -0.4497>
         <-5.3842 9.7507 0.2799>
         <5 -0.5 2>
         <-7.1807 13.9011 -3>
         <3.5012 4.5 0>
         <-0.2579 -0.6922 -1.5138>
         <0.5575 -3.8287 1.9089>
        ]}}
      Random Seed=0
     )
     5=&Multiply_Vector(
      A=&SelectVectorFromArray(
       Array Sampling Mode=::NewEnumerator0
       Direct Array Index=?
       Vector Selection Array=DataInterface() {propertyValues {FloatData [
          <2.5 2.5 -2.1846>
          <-2.6104 5.8649 -2.8753>
          <-6.7896 13.3872 -0.4497>
          <-5.3842 9.7507 0.2799>
          <5 -0.5 2>
          <-7.1807 13.9011 -3>
          <3.5012 4.5 0>
          <-0.2579 -0.6922 -1.5138>
          <0.5575 -3.8287 1.9089>
         ]}}
       Random Seed=0
      )
      B=<-1 -1 -1>
     )
     6=&SelectVectorFromArray(
      Array Sampling Mode=::NewEnumerator0
      Direct Array Index=?
      Vector Selection Array=DataInterface() {propertyValues {FloatData [
         <0.0654 -7.447 4.6204>
         <2.6773 -8.1349 0.1643>
         <-0.7284 -2.5857 -8.6556>
         <-1.7696 -1.1788 -9.0138>
         <2.8805 0.3175 -8.0983>
         <1.5724 -2.3328 -8.0311>
         <7.0748 2.3071 -7.5962>
         <5.7617 -3.4985 -7.4417>
         <4.1074 -5.9073 -5.9509>
         <6.8376 -3.533 -7.2802>
         <8.4336 -5.2053 -5.5826>
         <10.038 -0.1295 -6.8191>
         <11.3832 3.2891 -5.5813>
         <10.9559 1.4339 -6.1579>
         <11.4315 -2.3645 -5.4038>
         <10.2643 -5.3111 -3.9816>
         <12.0003 -3.0791 -4.5803>
         <12.4111 -2.2947 -4.5458>
         <5.4286 -7.1914 -3.3185>
         <8.2778 -7.2188 0.8564>
         <13.7362 -0.5855 -2.6207>
         <13.405 1.4267 -3.0383>
         <14.1998 -1.4315 0.5434>
         <6.353 -6.2982 4.2563>
         <0.9345 -6.7313 5.6183>
         <1.0494 -6.0969 6.7656>
         <3.3275 -4.8879 6.7826>
         <6.0867 -5.6376 5.5661>
         <2.5168 -3.2968 7.9413>
         <5.5983 -3.4706 7.2267>
         <12.9886 -1.4267 3.7958>
         <9.9231 -3.2266 6.0088>
         <12.2973 1.9074 4.9377>
         <12.4313 5.1182 -2.1817>
         <13.1529 2.3235 -3.165>
         <12.4008 2.338 -4.8634>
         <12.442 4.9507 -3.5495>
         <11.8923 5.7163 -3.8141>
         <12.27 5.3947 -2.781>
        ]}}
      Random Seed=0
     )
     7=&Multiply_Vector(
      A=&SelectVectorFromArray(
       Array Sampling Mode=::NewEnumerator0
       Direct Array Index=?
       Vector Selection Array=DataInterface() {propertyValues {FloatData [
          <17.0065 0.2755 -0.77>
          <3.2355 1.5733 2.9245>
          <6.2383 1.9629 -1.5>
          <-0.7279 3.4165 -2.0094>
          <-3.2654 -1.7219 1.614>
          <-0.5543 -2.2135 2.1404>
          <11.2237 -2.0841 -0.3281>
          <8.1481 -2.7762 0.1657>
          <8.5581 -2.3253 1.7338>
          <8.6436 -2.2336 1.8475>
          <9.85 0.9547 2>
         ]}}
       Random Seed=0
      )
      B=<-1 -1 -1>
     )
     8=&SelectVectorFromArray(
      Array Sampling Mode=::NewEnumerator0
      Direct Array Index=?
      Vector Selection Array=DataInterface() {propertyValues {FloatData [
         <-1.5173 0.1051 -1.404>
         <0.7405 -0.187 -3.0162>
         <0.8485 -0.8853 0.5135>
         <2.0365 -0.7992 -3.378>
         <2.4306 -0.225 -0.087>
         <5.3848 -0.6851 1.719>
         <4.4317 -2.2879 -2.9193>
         <4.6687 -3.8049 -4.9755>
         <6.8585 -0.6734 3.3238>
         <5.5586 0.2819 -0.4461>
        ]}}
      Random Seed=0
     )
     9=&Multiply_Vector(
      A=&SelectVectorFromArray(
       Array Sampling Mode=::NewEnumerator0
       Direct Array Index=?
       Vector Selection Array=DataInterface() {propertyValues {FloatData [
          <-1.5173 0.1051 -1.404>
          <0.7405 -0.187 -3.0162>
          <0.8485 -0.8853 0.5135>
          <2.0365 -0.7992 -3.378>
          <2.4306 -0.225 -0.087>
          <5.3848 -0.6851 1.719>
          <4.4317 -2.2879 -2.9193>
          <4.6687 -3.8049 -4.9755>
          <6.8585 -0.6734 3.3238>
          <5.5586 0.2819 -0.4461>
         ]}}
       Random Seed=0
      )
      B=<-1 -1 -1>
     )
     Index=$Particles.PartitionParticles.Partition
     Num=::NewEnumerator4
    )
    Skeletal Mesh=$System.Pearl_Skeleton
   ) {module=SkeletalMeshLocation002}
   &ScaleSpriteSize() {module=ScaleSpriteSize001}
   &ScaleColor(
    Scale Alpha=&Multiply_Float(
     A=&FloatFromCurve(
      FloatCurve=DataInterface() {propertyValues {Curve {keys [
          {interpMode=RCIM_Cubic, tangentMode=RCTM_User, time=0, value=0}
          {arriveTangent=-2.4058, interpMode=RCIM_Cubic, leaveTangent=-2.4058, tangentMode=RCTM_User, time=0.26, value=0.8421}
          {arriveTangent=-1.1775, interpMode=RCIM_Cubic, leaveTangent=-1.1775, time=0.4613, value=0.0645}
          {arriveTangent=-0.1197, interpMode=RCIM_Cubic, leaveTangent=-0.1197, time=0.5898, value=0.4537}
          <at 1 0 RCIM_Cubic>
         ]}}}
      Scale Curve=&RandomRangeFloat(Minimum=0.1, Maximum=0.5)
     )
     B=&FloatFromCurve(
      FloatCurve=DataInterface() {propertyValues {Curve {keys [{tangentMode=RCTM_User, time=0, value=0}, {tangentMode=RCTM_User, time=1, value=1}]}}}
      Scale Curve=&FloatFromCurve(
       FloatCurve=DataInterface() {propertyValues {Curve {keys [
           {arriveTangentWeight=0.1667, interpMode=RCIM_Cubic, leaveTangentWeight=0.1667, tangentMode=RCTM_User, time=1.4, value=0}
           {tangentMode=RCTM_User, time=2, value=1}
          ]}}}
       CurveIndex=$Emitter.Age
      )
     )
    )
    Scale RGB=&VectorFromFloat(Value=&Multiply_Float(
      A=&RandomRangeFloat(Minimum=0.6, Maximum=1, Evaluation Type=::NewEnumerator1)
      B=&SelectFloatFromArray(Direct Array Index=$Engine.QualityLevel, Float Selection Array=DataInterface() {propertyValues {FloatData [10,5,1,1,1]}})
     ))
    ScaleA=true
    ScaleRGB=true
   ) {module=ScaleColor001}
   &ScaleColor(
    Scale Alpha=$System.TODAlpha
    Scale RGB=&MakeVectorFromLinearColorRGB(Color=$System.TODColor)
    ScaleA=true
    ScaleRGB=true
   )
   /SetVariables(
    Particles.CenterFlareColor:LinearColor=&MultiplyLinearColor(
     Linear Color=$Particles.Color
     Scale Factor=&MakeVector4_FromLinearColor(LinearColor=<rgba 8.0099 9.8007 10 1>)
    )
    Particles.CenterFlareSize:Vector2f=&Multiply_Vector2(A=$Particles.SpriteSize, B=&Vector2DFromFloat(Value=0.5))
    Particles.CoronaFlareColor:LinearColor=&MultiplyLinearColor(
     Linear Color=$Particles.Color
     Scale Factor=&MakeVector4_FromLinearColor(LinearColor=<rgba 0.15 0.15 0.15 1>)
    )
    Particles.CoronaFlareSize:Vector2f=&Multiply_Vector2(A=$Particles.Initial.SpriteSize, B=&Vector2DFromFloat(Value=1))
    Particles.FlickerFlareColor:LinearColor=&MultiplyLinearColor(
     Linear Color=$Particles.Color
     Scale Factor=&MakeVector4_FromLinearColor(LinearColor=<rgba 0.9441 0.8079 1 1>)
    )
   ) {isSetParameters=true, module=SetVariables_#3}
  ]
  rendererClasses [&NiagaraSpriteRendererProperties]
  renderers [
   {rendererClass=&NiagaraSpriteRendererProperties}
   {rendererClass=&NiagaraSpriteRendererProperties, rendererIndex=1}
   {rendererClass=&NiagaraSpriteRendererProperties, rendererIndex=2}
  ]
 }
 LowerBody_Head_Sparkles001 {
  emitterUpdateScript [
   &EmitterState()
   &SpawnRate(
    Spawn Probability=$System.SpawnProbability
    SpawnRate=&Multiply_Float(A=$System.SpawnRate, B=2)
    Use Spawn Probability=true
   )
  ]
  particleSpawnScript [
   &PartitionParticles(Number Of Partitions=3)
   &InitializeParticle(
    AdjustAlpha=false
    AdjustHue=true
    AdjustSaturation=false
    AdjustValue=false
    Color=<rgba 0.843 0.8 1 1>
    Color Mode=::NewEnumerator3
    Lifetime Max=0.4
    Lifetime Min=0.3
    Lifetime Mode=::NewEnumerator1
    Sprite Size Mode=::NewEnumerator3
    Uniform Sprite Size=&Multiply_Float(A=40, B=&RandomRangeFloat(Minimum=0.75, Maximum=1))
    UsePositionOffset=false
    Write Material Random=true
   )
   &CameraOffset(Camera Offset Amount=&FN_FrontEndFloatModifier(
     Gameplay=&MergeHelper_FloatSwitch(0=8, 1=8, 2=8, 3=8, 4=4, 5=4, 6=4, Index=$Particles.PartitionParticles.Partition, Num=::NewEnumerator5)
     Front End=&MergeHelper_FloatSwitch(0=3, 1=3, 2=3, 3=3, 4=2, 5=2, 6=2, Index=$Particles.PartitionParticles.Partition, Num=::NewEnumerator5)
    ))
  ]
  particleUpdateScript [
   &ParticleState()
   &SkeletalMeshLocation(
    Bone / Socket Index=&MergeHelper_IntSwitch(0=10, 1=11, 2=16, 3=?, 4=?, 5=?, Index=$Particles.PartitionParticles.Partition, Num=::NewEnumerator1)
    Bone Sampling Mode=::NewEnumerator4
    MaskSpawnGroup=false
    Sampled Position Offset=&MergeHelper_VectorSwitch(
     0=&MakeCustomVectorFromBool(
      Bool=$User.bisFootHidden
      True Vector=&SelectVectorFromArray(
       Array Sampling Mode=::NewEnumerator0
       Direct Array Index=?
       Vector Selection Array=DataInterface() {propertyValues {FloatData [
          <-9.5011 -7.101 8.3335>
          <-7.8626 -10.4662 7.0362>
          <-3.5916 -11.9308 8.8412>
          <-1.2405 -11.9006 8.0091>
          <-0.3188 -10.4972 -0.2352>
          <-3.9917 -9.2248 -2.935>
          <-6.2369 -8.6736 -3.4726>
          <-7.7466 -7.6189 -4.6054>
          <-10.4565 -7.747 -3.7401>
          <-8.2793 -10.8096 4.8908>
          <-9.7047 -9.7469 0.7749>
          <-9.6833 -10.025 1.8191>
          <-11.2928 -8.7171 0.0117>
          <-14.3592 -7.9845 -1.8384>
          <-16.0147 -3.8523 7.4032>
          <-18.4519 -7.7997 0.824>
          <-16.5627 -7.0835 4.4511>
          <-22.6199 -6.2812 4.3659>
          <-22.5789 -6.6713 3.419>
          <-19.7382 -6.3003 5.09>
          <-22.9862 -3.7865 6.5175>
          <-23.5423 -3.8138 6.4205>
          <-25.6543 -4.1805 5.6618>
          <-24.1581 -3.8166 6.3049>
          <-26.6486 -2.5894 6.1861>
          <-10.0371 -2.2112 9.0726>
          <-11.387 -0.3688 8.4281>
          <-12.1143 -2.7924 8.2052>
          <-10.1116 2.3955 7.6971>
          <-15.8873 -1.3758 8.1105>
          <-13.1581 1.9832 7.6369>
          <-13.0095 3.5757 6.7901>
          <-19.0459 -0.5956 7.66>
          <-21.6593 -0.1487 7.1639>
          <-27.0163 0.1199 6.4232>
          <-23.5677 2.6485 5.9957>
          <-25.5414 2.3009 6.0048>
          <-28.372 3.7524 4.9132>
          <-29.1917 1.7277 5.8842>
          <-23.9504 5.1588 4.1042>
          <-20.4059 4.4383 5.3449>
          <-19.1482 7.2219 0.7659>
          <-17.6118 7.2524 0.7025>
          <-14.6546 6.0066 3.848>
          <-13.6231 4.4313 5.9866>
          <-12.1285 5.6515 4.2966>
          <-12.6156 6.7773 1.7439>
          <-13.6073 6.8217 -3.6346>
          <-11.9774 7.2497 -0.4189>
          <-9.7416 6.3224 2.6673>
          <-9.2426 5.4938 4.4497>
          <-8.1365 6.1282 -3.5462>
          <-7.0487 5.4343 4.6496>
          <-7.5342 4.9801 6.3086>
          <-6.9896 5.2757 5.318>
          <-4.6719 6.5439 8.0261>
          <-5.2449 5.7596 -3.0092>
          <0.31 5.8941 -0.8948>
          <0.2973 6.6571 1.2886>
          <-37.6541 -3.746 4.339>
          <-33.5972 -2.2992 5.3567>
          <-34.3449 -4.1678 4.1235>
          <-30.169 -4.4425 4.539>
          <-27.4974 -6.4626 2.4659>
          <-25.343 -6.7248 2.4487>
          <-27.2159 -6 -2.6542>
          <-31.1827 -6.162 -0.3374>
          <-23.1859 -5.8842 -3.8446>
          <-20.2074 -6.9963 -2.9875>
          <-0.5713 -6.6333 -4.6668>
          <0.8562 2.8963 -5.2687>
          <0.741 2.4439 -5.5658>
          <-5.8455 4.9146 -5.351>
          <-8.01 2.4905 -7.2539>
          <-5.8469 1.3932 -7.2726>
          <-3.7271 1.7012 -7.0625>
          <-2.3476 1.3966 -6.8909>
          <-6.0539 -3.0343 -7.2102>
          <-1.6814 -4.6839 -5.8932>
          <-7.5425 0.4264 -7.6367>
          <-10.6191 -4.4262 -7.1025>
          <-11.4933 -1.505 -7.8564>
          <-11.7186 0.9154 -7.9041>
          <-12.9362 -1.3969 -7.8977>
          <-16.356 -3.7309 -6.8478>
          <-23.7803 -1.799 -6.547>
          <-18.6008 1.2126 -7.601>
          <-26.6551 -2.6418 -5.7314>
          <-21.4877 -0.7181 -7.0396>
          <-15.9328 2.6537 -7.3416>
          <-22.2543 1.4748 -7.1758>
          <-12.9762 5.8693 -5.6476>
          <-15.5748 4.8427 -6.4844>
          <-22.3456 5.6604 -5.2688>
          <-22.3176 6.925 -3.2316>
          <-24.1257 6.7135 -3.276>
          <-22.4925 7.5349 -0.9795>
          <-24.7182 7.2067 0.6833>
          <-26.5204 7.1185 -0.1577>
          <-33.1891 4.7901 3.1511>
          <-34.0012 1.3814 5.4839>
          <-35.3591 2.0098 4.9184>
          <-37.446 1.8284 4.716>
          <-39.092 2.1744 4.1786>
          <-40.2632 -0.0937 5.0668>
          <-36.8984 4.1468 2.9042>
          <-36.8553 5.4454 -0.5958>
          <-35.2618 5.7679 0.1288>
          <-28.8319 6.6867 -2.0049>
          <-27.0483 6.7153 -2.6679>
         ]}}
       Random Seed=0
      )
      False Vector=&SelectVectorFromArray(
       Array Sampling Mode=::NewEnumerator0
       Direct Array Index=?
       Vector Selection Array=DataInterface() {propertyValues {FloatData [
          <-9.5011 -7.101 8.3335>
          <-7.8626 -10.4662 7.0362>
          <-3.5916 -11.9308 8.8412>
          <-1.2405 -11.9006 8.0091>
          <-0.3188 -10.4972 -0.2352>
          <-3.9917 -9.2248 -2.935>
          <-6.2369 -8.6736 -3.4726>
          <-7.7466 -7.6189 -4.6054>
          <-10.4565 -7.747 -3.7401>
          <-8.2793 -10.8096 4.8908>
          <-9.7047 -9.7469 0.7749>
          <-9.6833 -10.025 1.8191>
          <-11.2928 -8.7171 0.0117>
          <-14.3592 -7.9845 -1.8384>
          <-16.0147 -3.8523 7.4032>
          <-18.4519 -7.7997 0.824>
          <-16.5627 -7.0835 4.4511>
          <-22.6199 -6.2812 4.3659>
          <-22.5789 -6.6713 3.419>
          <-19.7382 -6.3003 5.09>
          <-22.9862 -3.7865 6.5175>
          <-23.5423 -3.8138 6.4205>
          <-25.6543 -4.1805 5.6618>
          <-24.1581 -3.8166 6.3049>
          <-26.6486 -2.5894 6.1861>
          <-10.0371 -2.2112 9.0726>
          <-11.387 -0.3688 8.4281>
          <-12.1143 -2.7924 8.2052>
          <-10.1116 2.3955 7.6971>
          <-15.8873 -1.3758 8.1105>
          <-13.1581 1.9832 7.6369>
          <-13.0095 3.5757 6.7901>
          <-19.0459 -0.5956 7.66>
          <-21.6593 -0.1487 7.1639>
          <-27.0163 0.1199 6.4232>
          <-23.5677 2.6485 5.9957>
          <-25.5414 2.3009 6.0048>
          <-28.372 3.7524 4.9132>
          <-29.1917 1.7277 5.8842>
          <-23.9504 5.1588 4.1042>
          <-20.4059 4.4383 5.3449>
          <-19.1482 7.2219 0.7659>
          <-17.6118 7.2524 0.7025>
          <-14.6546 6.0066 3.848>
          <-13.6231 4.4313 5.9866>
          <-12.1285 5.6515 4.2966>
          <-12.6156 6.7773 1.7439>
          <-13.6073 6.8217 -3.6346>
          <-11.9774 7.2497 -0.4189>
          <-9.7416 6.3224 2.6673>
          <-9.2426 5.4938 4.4497>
          <-8.1365 6.1282 -3.5462>
          <-7.0487 5.4343 4.6496>
          <-7.5342 4.9801 6.3086>
          <-6.9896 5.2757 5.318>
          <-4.6719 6.5439 8.0261>
          <-5.2449 5.7596 -3.0092>
          <0.31 5.8941 -0.8948>
          <0.2973 6.6571 1.2886>
          <-37.6541 -3.746 4.339>
          <-33.5972 -2.2992 5.3567>
          <-34.3449 -4.1678 4.1235>
          <-30.169 -4.4425 4.539>
          <-27.4974 -6.4626 2.4659>
          <-25.343 -6.7248 2.4487>
          <-27.2159 -6 -2.6542>
          <-31.1827 -6.162 -0.3374>
          <-23.1859 -5.8842 -3.8446>
          <-20.2074 -6.9963 -2.9875>
          <-0.5713 -6.6333 -4.6668>
          <0.8562 2.8963 -5.2687>
          <0.741 2.4439 -5.5658>
          <-5.8455 4.9146 -5.351>
          <-8.01 2.4905 -7.2539>
          <-5.8469 1.3932 -7.2726>
          <-3.7271 1.7012 -7.0625>
          <-2.3476 1.3966 -6.8909>
          <-6.0539 -3.0343 -7.2102>
          <-1.6814 -4.6839 -5.8932>
          <-7.5425 0.4264 -7.6367>
          <-10.6191 -4.4262 -7.1025>
          <-11.4933 -1.505 -7.8564>
          <-11.7186 0.9154 -7.9041>
          <-12.9362 -1.3969 -7.8977>
          <-16.356 -3.7309 -6.8478>
          <-23.7803 -1.799 -6.547>
          <-18.6008 1.2126 -7.601>
          <-26.6551 -2.6418 -5.7314>
          <-21.4877 -0.7181 -7.0396>
          <-15.9328 2.6537 -7.3416>
          <-22.2543 1.4748 -7.1758>
          <-12.9762 5.8693 -5.6476>
          <-15.5748 4.8427 -6.4844>
          <-22.3456 5.6604 -5.2688>
          <-22.3176 6.925 -3.2316>
          <-24.1257 6.7135 -3.276>
          <-22.4925 7.5349 -0.9795>
          <-24.7182 7.2067 0.6833>
          <-26.5204 7.1185 -0.1577>
          <-33.1891 4.7901 3.1511>
          <-34.0012 1.3814 5.4839>
          <-35.3591 2.0098 4.9184>
          <-37.446 1.8284 4.716>
          <-39.092 2.1744 4.1786>
          <-40.2632 -0.0937 5.0668>
          <-36.8984 4.1468 2.9042>
          <-36.8553 5.4454 -0.5958>
          <-35.2618 5.7679 0.1288>
          <-28.8319 6.6867 -2.0049>
          <-27.0483 6.7153 -2.6679>
          <-33.9958 5.5914 -2.7024>
          <-34.6375 5.8803 -1.4407>
          <-37.0013 4.9533 -2.5268>
          <-44.0636 3.9471 1.2993>
          <-42.1165 4.7353 -1.3271>
          <-44.5605 3.5931 -2.4286>
          <-36.947 1.4267 -4.9996>
          <-27.8495 4.0748 -5.7069>
          <-36.2509 0.0642 -5.1638>
          <-28.3496 3.0448 -5.9781>
          <-33.7488 0.5184 -5.5048>
          <-34.5032 -3.4231 -3.9042>
          <-38.3464 -4.0436 -3.1163>
          <-38.3168 -4.8402 -2.0307>
          <-38.455 -4.9726 -1.724>
          <-41.634 -5.4571 0.444>
          <-41.3907 -3.9536 -3.255>
          <-44.6488 1.6876 -3.8329>
         ]}}
       Random Seed=0
      )
     )
     1=&Multiply_Vector(
      A=&MakeCustomVectorFromBool(
       Bool=$User.bisFootHidden
       True Vector=&SelectVectorFromArray(
        Array Sampling Mode=::NewEnumerator0
        Direct Array Index=?
        Vector Selection Array=DataInterface() {propertyValues {FloatData [
           <-9.5011 -7.101 8.3335>
           <-7.8626 -10.4662 7.0362>
           <-3.5916 -11.9308 8.8412>
           <-1.2405 -11.9006 8.0091>
           <-0.3188 -10.4972 -0.2352>
           <-3.9917 -9.2248 -2.935>
           <-6.2369 -8.6736 -3.4726>
           <-7.7466 -7.6189 -4.6054>
           <-10.4565 -7.747 -3.7401>
           <-8.2793 -10.8096 4.8908>
           <-9.7047 -9.7469 0.7749>
           <-9.6833 -10.025 1.8191>
           <-11.2928 -8.7171 0.0117>
           <-14.3592 -7.9845 -1.8384>
           <-16.0147 -3.8523 7.4032>
           <-18.4519 -7.7997 0.824>
           <-16.5627 -7.0835 4.4511>
           <-22.6199 -6.2812 4.3659>
           <-22.5789 -6.6713 3.419>
           <-19.7382 -6.3003 5.09>
           <-22.9862 -3.7865 6.5175>
           <-23.5423 -3.8138 6.4205>
           <-25.6543 -4.1805 5.6618>
           <-24.1581 -3.8166 6.3049>
           <-26.6486 -2.5894 6.1861>
           <-10.0371 -2.2112 9.0726>
           <-11.387 -0.3688 8.4281>
           <-12.1143 -2.7924 8.2052>
           <-10.1116 2.3955 7.6971>
           <-15.8873 -1.3758 8.1105>
           <-13.1581 1.9832 7.6369>
           <-13.0095 3.5757 6.7901>
           <-19.0459 -0.5956 7.66>
           <-21.6593 -0.1487 7.1639>
           <-27.0163 0.1199 6.4232>
           <-23.5677 2.6485 5.9957>
           <-25.5414 2.3009 6.0048>
           <-28.372 3.7524 4.9132>
           <-29.1917 1.7277 5.8842>
           <-23.9504 5.1588 4.1042>
           <-20.4059 4.4383 5.3449>
           <-19.1482 7.2219 0.7659>
           <-17.6118 7.2524 0.7025>
           <-14.6546 6.0066 3.848>
           <-13.6231 4.4313 5.9866>
           <-12.1285 5.6515 4.2966>
           <-12.6156 6.7773 1.7439>
           <-13.6073 6.8217 -3.6346>
           <-11.9774 7.2497 -0.4189>
           <-9.7416 6.3224 2.6673>
           <-9.2426 5.4938 4.4497>
           <-8.1365 6.1282 -3.5462>
           <-7.0487 5.4343 4.6496>
           <-7.5342 4.9801 6.3086>
           <-6.9896 5.2757 5.318>
           <-4.6719 6.5439 8.0261>
           <-5.2449 5.7596 -3.0092>
           <0.31 5.8941 -0.8948>
           <0.2973 6.6571 1.2886>
           <-37.6541 -3.746 4.339>
           <-33.5972 -2.2992 5.3567>
           <-34.3449 -4.1678 4.1235>
           <-30.169 -4.4425 4.539>
           <-27.4974 -6.4626 2.4659>
           <-25.343 -6.7248 2.4487>
           <-27.2159 -6 -2.6542>
           <-31.1827 -6.162 -0.3374>
           <-23.1859 -5.8842 -3.8446>
           <-20.2074 -6.9963 -2.9875>
           <-0.5713 -6.6333 -4.6668>
           <0.8562 2.8963 -5.2687>
           <0.741 2.4439 -5.5658>
           <-5.8455 4.9146 -5.351>
           <-8.01 2.4905 -7.2539>
           <-5.8469 1.3932 -7.2726>
           <-3.7271 1.7012 -7.0625>
           <-2.3476 1.3966 -6.8909>
           <-6.0539 -3.0343 -7.2102>
           <-1.6814 -4.6839 -5.8932>
           <-7.5425 0.4264 -7.6367>
           <-10.6191 -4.4262 -7.1025>
           <-11.4933 -1.505 -7.8564>
           <-11.7186 0.9154 -7.9041>
           <-12.9362 -1.3969 -7.8977>
           <-16.356 -3.7309 -6.8478>
           <-23.7803 -1.799 -6.547>
           <-18.6008 1.2126 -7.601>
           <-26.6551 -2.6418 -5.7314>
           <-21.4877 -0.7181 -7.0396>
           <-15.9328 2.6537 -7.3416>
           <-22.2543 1.4748 -7.1758>
           <-12.9762 5.8693 -5.6476>
           <-15.5748 4.8427 -6.4844>
           <-22.3456 5.6604 -5.2688>
           <-22.3176 6.925 -3.2316>
           <-24.1257 6.7135 -3.276>
           <-22.4925 7.5349 -0.9795>
           <-24.7182 7.2067 0.6833>
           <-26.5204 7.1185 -0.1577>
           <-33.1891 4.7901 3.1511>
           <-34.0012 1.3814 5.4839>
           <-35.3591 2.0098 4.9184>
           <-37.446 1.8284 4.716>
           <-39.092 2.1744 4.1786>
           <-40.2632 -0.0937 5.0668>
           <-36.8984 4.1468 2.9042>
           <-36.8553 5.4454 -0.5958>
           <-35.2618 5.7679 0.1288>
           <-28.8319 6.6867 -2.0049>
           <-27.0483 6.7153 -2.6679>
           <-33.9958 5.5914 -2.7024>
           <-34.6375 5.8803 -1.4407>
          ]}}
        Random Seed=0
       )
       False Vector=&SelectVectorFromArray(
        Array Sampling Mode=::NewEnumerator0
        Direct Array Index=?
        Vector Selection Array=DataInterface() {propertyValues {FloatData [
           <-9.5011 -7.101 8.3335>
           <-7.8626 -10.4662 7.0362>
           <-3.5916 -11.9308 8.8412>
           <-1.2405 -11.9006 8.0091>
           <-0.3188 -10.4972 -0.2352>
           <-3.9917 -9.2248 -2.935>
           <-6.2369 -8.6736 -3.4726>
           <-7.7466 -7.6189 -4.6054>
           <-10.4565 -7.747 -3.7401>
           <-8.2793 -10.8096 4.8908>
           <-9.7047 -9.7469 0.7749>
           <-9.6833 -10.025 1.8191>
           <-11.2928 -8.7171 0.0117>
           <-14.3592 -7.9845 -1.8384>
           <-16.0147 -3.8523 7.4032>
           <-18.4519 -7.7997 0.824>
           <-16.5627 -7.0835 4.4511>
           <-22.6199 -6.2812 4.3659>
           <-22.5789 -6.6713 3.419>
           <-19.7382 -6.3003 5.09>
           <-22.9862 -3.7865 6.5175>
           <-23.5423 -3.8138 6.4205>
           <-25.6543 -4.1805 5.6618>
           <-24.1581 -3.8166 6.3049>
           <-26.6486 -2.5894 6.1861>
           <-10.0371 -2.2112 9.0726>
           <-11.387 -0.3688 8.4281>
           <-12.1143 -2.7924 8.2052>
           <-10.1116 2.3955 7.6971>
           <-15.8873 -1.3758 8.1105>
           <-13.1581 1.9832 7.6369>
           <-13.0095 3.5757 6.7901>
           <-19.0459 -0.5956 7.66>
           <-21.6593 -0.1487 7.1639>
           <-27.0163 0.1199 6.4232>
           <-23.5677 2.6485 5.9957>
           <-25.5414 2.3009 6.0048>
           <-28.372 3.7524 4.9132>
           <-29.1917 1.7277 5.8842>
           <-23.9504 5.1588 4.1042>
           <-20.4059 4.4383 5.3449>
           <-19.1482 7.2219 0.7659>
           <-17.6118 7.2524 0.7025>
           <-14.6546 6.0066 3.848>
           <-13.6231 4.4313 5.9866>
           <-12.1285 5.6515 4.2966>
           <-12.6156 6.7773 1.7439>
           <-13.6073 6.8217 -3.6346>
           <-11.9774 7.2497 -0.4189>
           <-9.7416 6.3224 2.6673>
           <-9.2426 5.4938 4.4497>
           <-8.1365 6.1282 -3.5462>
           <-7.0487 5.4343 4.6496>
           <-7.5342 4.9801 6.3086>
           <-6.9896 5.2757 5.318>
           <-4.6719 6.5439 8.0261>
           <-5.2449 5.7596 -3.0092>
           <0.31 5.8941 -0.8948>
           <0.2973 6.6571 1.2886>
           <-37.6541 -3.746 4.339>
           <-33.5972 -2.2992 5.3567>
           <-34.3449 -4.1678 4.1235>
           <-30.169 -4.4425 4.539>
           <-27.4974 -6.4626 2.4659>
           <-25.343 -6.7248 2.4487>
           <-27.2159 -6 -2.6542>
           <-31.1827 -6.162 -0.3374>
           <-23.1859 -5.8842 -3.8446>
           <-20.2074 -6.9963 -2.9875>
           <-0.5713 -6.6333 -4.6668>
           <0.8562 2.8963 -5.2687>
           <0.741 2.4439 -5.5658>
           <-5.8455 4.9146 -5.351>
           <-8.01 2.4905 -7.2539>
           <-5.8469 1.3932 -7.2726>
           <-3.7271 1.7012 -7.0625>
           <-2.3476 1.3966 -6.8909>
           <-6.0539 -3.0343 -7.2102>
           <-1.6814 -4.6839 -5.8932>
           <-7.5425 0.4264 -7.6367>
           <-10.6191 -4.4262 -7.1025>
           <-11.4933 -1.505 -7.8564>
           <-11.7186 0.9154 -7.9041>
           <-12.9362 -1.3969 -7.8977>
           <-16.356 -3.7309 -6.8478>
           <-23.7803 -1.799 -6.547>
           <-18.6008 1.2126 -7.601>
           <-26.6551 -2.6418 -5.7314>
           <-21.4877 -0.7181 -7.0396>
           <-15.9328 2.6537 -7.3416>
           <-22.2543 1.4748 -7.1758>
           <-12.9762 5.8693 -5.6476>
           <-15.5748 4.8427 -6.4844>
           <-22.3456 5.6604 -5.2688>
           <-22.3176 6.925 -3.2316>
           <-24.1257 6.7135 -3.276>
           <-22.4925 7.5349 -0.9795>
           <-24.7182 7.2067 0.6833>
           <-26.5204 7.1185 -0.1577>
           <-33.1891 4.7901 3.1511>
           <-34.0012 1.3814 5.4839>
           <-35.3591 2.0098 4.9184>
           <-37.446 1.8284 4.716>
           <-39.092 2.1744 4.1786>
           <-40.2632 -0.0937 5.0668>
           <-36.8984 4.1468 2.9042>
           <-36.8553 5.4454 -0.5958>
           <-35.2618 5.7679 0.1288>
           <-28.8319 6.6867 -2.0049>
           <-27.0483 6.7153 -2.6679>
           <-33.9958 5.5914 -2.7024>
           <-34.6375 5.8803 -1.4407>
           <-37.0013 4.9533 -2.5268>
           <-44.0636 3.9471 1.2993>
           <-42.1165 4.7353 -1.3271>
           <-44.5605 3.5931 -2.4286>
           <-36.947 1.4267 -4.9996>
           <-27.8495 4.0748 -5.7069>
           <-36.2509 0.0642 -5.1638>
           <-28.3496 3.0448 -5.9781>
           <-33.7488 0.5184 -5.5048>
           <-34.5032 -3.4231 -3.9042>
           <-38.3464 -4.0436 -3.1163>
           <-38.3168 -4.8402 -2.0307>
           <-38.455 -4.9726 -1.724>
           <-41.634 -5.4571 0.444>
           <-41.3907 -3.9536 -3.255>
           <-44.6488 1.6876 -3.8329>
          ]}}
        Random Seed=0
       )
      )
      B=<-1 -1 -1>
     )
     2=&SelectVectorFromArray(
      Array Sampling Mode=::NewEnumerator0
      Direct Array Index=?
      Vector Selection Array=DataInterface() {propertyValues {FloatData [
         <0.0654 -7.447 4.6204>
         <2.6773 -8.1349 0.1643>
         <-0.7284 -2.5857 -8.6556>
         <-1.7696 -1.1788 -9.0138>
         <2.8805 0.3175 -8.0983>
         <1.5724 -2.3328 -8.0311>
         <7.0748 2.3071 -7.5962>
         <5.7617 -3.4985 -7.4417>
         <4.1074 -5.9073 -5.9509>
         <6.8376 -3.533 -7.2802>
         <8.4336 -5.2053 -5.5826>
         <10.038 -0.1295 -6.8191>
         <11.3832 3.2891 -5.5813>
         <10.9559 1.4339 -6.1579>
         <11.4315 -2.3645 -5.4038>
         <10.2643 -5.3111 -3.9816>
         <12.0003 -3.0791 -4.5803>
         <12.4111 -2.2947 -4.5458>
         <5.4286 -7.1914 -3.3185>
         <8.2778 -7.2188 0.8564>
         <13.7362 -0.5855 -2.6207>
         <13.405 1.4267 -3.0383>
         <14.1998 -1.4315 0.5434>
         <6.353 -6.2982 4.2563>
         <0.9345 -6.7313 5.6183>
         <1.0494 -6.0969 6.7656>
         <3.3275 -4.8879 6.7826>
         <6.0867 -5.6376 5.5661>
         <2.5168 -3.2968 7.9413>
         <5.5983 -3.4706 7.2267>
         <12.9886 -1.4267 3.7958>
         <9.9231 -3.2266 6.0088>
         <12.2973 1.9074 4.9377>
         <12.4313 5.1182 -2.1817>
         <13.1529 2.3235 -3.165>
         <12.4008 2.338 -4.8634>
         <12.442 4.9507 -3.5495>
         <11.8923 5.7163 -3.8141>
         <12.27 5.3947 -2.781>
        ]}}
      Random Seed=0
     )
     3=&Multiply_Vector(
      A=&SelectVectorFromArray(
       Array Sampling Mode=::NewEnumerator0
       Direct Array Index=?
       Vector Selection Array=DataInterface() {propertyValues {FloatData [
          <-12.9425 -6.387 3.6216>
          <-16.5827 -7.4896 1.6339>
          <-8.4828 -4.5327 1.6501>
          <-4.6066 -4.2273 1.6913>
          <-5.584 -5.1835 3.2259>
          <-3.9535 -5.0472 2.7323>
          <-5.2327 2.875 2.4717>
          <-6.7877 -1.4729 -1.63>
          <-6.7199 2.7064 1.6431>
          <-20.5875 -1.7211 2>
          <-16.7039 0.635 3.139>
          <-20.5856 -0.1211 2.463>
          <-22.7547 0.0254 1.9571>
          <-17.4784 1.2769 1.9876>
          <-5.8898 3.3702 0.9041>
          <-2.01 4.5774 -1.0952>
          <-5.3701 2.6413 -2.3868>
          <-4.4609 -5.8196 1.1597>
          <-5.1661 -5.4639 -1.6313>
          <-9.0792 -7.3158 0.0885>
          <-11.9489 -8.0682 0.9757>
          <-11.7825 -7.9909 -0.6143>
          <-9.111 -3.7766 -4.2836>
          <-7.8802 -1.5994 -4.2708>
          <-5.3425 -0.9814 -3.9311>
          <-12.5021 -0.8744 -4.6291>
          <-2.5447 0.9346 -3.6069>
          <-7.7018 0.9929 -3.6505>
          <-12.6316 2.7854 -0.7521>
          <-17.1669 0.8871 -3.5899>
          <-16.36 2.0544 -2.2486>
          <-15.1586 -8.4463 -1.2259>
          <-22.2134 -7.2388 0.3382>
          <-30.7004 -3.7103 2.5996>
          <-29.6611 -6.6044 -0.8972>
          <-25.1304 -5.9487 -3.1851>
          <-20.4869 -3.7431 -4.8596>
          <-21.603 -1.5774 -4.469>
          <-28.3677 -4.8929 -3.1995>
          <-24.546 -1.2731 -3.9224>
          <-25.3217 -0.2932 -3.2908>
          <-20.315 0.8983 -3.1685>
          <-17.6738 1.8422 -2.4468>
          <-21.7461 1.5705 0.8324>
          <-23.712 1.5507 0.5217>
          <-29.8463 0.7515 1.432>
          <-31.4608 1.0341 -0.073>
          <-30.3995 0.9309 -0.971>
          <-35.0209 0.219 -1.6537>
          <-32.6261 -5.3095 -1.81>
          <-33.5537 -5.2005 -1.6885>
          <-37.41 -5.825 0.5207>
          <-40.3806 0.0668 -2.1607>
         ]}}
       Random Seed=0
      )
      B=<-1 -1 -1>
     )
     4=&SelectVectorFromArray(
      Array Sampling Mode=::NewEnumerator0
      Direct Array Index=?
      Vector Selection Array=DataInterface() {propertyValues {FloatData [
         <2.5 2.5 -2.1846>
         <-2.6104 5.8649 -2.8753>
         <-6.7896 13.3872 -0.4497>
         <-5.3842 9.7507 0.2799>
         <5 -0.5 2>
         <-7.1807 13.9011 -3>
         <3.5012 4.5 0>
         <-0.2579 -0.6922 -1.5138>
         <0.5575 -3.8287 1.9089>
        ]}}
      Random Seed=0
     )
     5=&Multiply_Vector(
      A=&SelectVectorFromArray(
       Array Sampling Mode=::NewEnumerator0
       Direct Array Index=?
       Vector Selection Array=DataInterface() {propertyValues {FloatData [
          <2.5 2.5 -2.1846>
          <-2.6104 5.8649 -2.8753>
          <-6.7896 13.3872 -0.4497>
          <-5.3842 9.7507 0.2799>
          <5 -0.5 2>
          <-7.1807 13.9011 -3>
          <3.5012 4.5 0>
          <-0.2579 -0.6922 -1.5138>
          <0.5575 -3.8287 1.9089>
         ]}}
       Random Seed=0
      )
      B=<-1 -1 -1>
     )
     6=&SelectVectorFromArray(
      Array Sampling Mode=::NewEnumerator0
      Direct Array Index=?
      Vector Selection Array=DataInterface() {propertyValues {FloatData [
         <0.0654 -7.447 4.6204>
         <2.6773 -8.1349 0.1643>
         <-0.7284 -2.5857 -8.6556>
         <-1.7696 -1.1788 -9.0138>
         <2.8805 0.3175 -8.0983>
         <1.5724 -2.3328 -8.0311>
         <7.0748 2.3071 -7.5962>
         <5.7617 -3.4985 -7.4417>
         <4.1074 -5.9073 -5.9509>
         <6.8376 -3.533 -7.2802>
         <8.4336 -5.2053 -5.5826>
         <10.038 -0.1295 -6.8191>
         <11.3832 3.2891 -5.5813>
         <10.9559 1.4339 -6.1579>
         <11.4315 -2.3645 -5.4038>
         <10.2643 -5.3111 -3.9816>
         <12.0003 -3.0791 -4.5803>
         <12.4111 -2.2947 -4.5458>
         <5.4286 -7.1914 -3.3185>
         <8.2778 -7.2188 0.8564>
         <13.7362 -0.5855 -2.6207>
         <13.405 1.4267 -3.0383>
         <14.1998 -1.4315 0.5434>
         <6.353 -6.2982 4.2563>
         <0.9345 -6.7313 5.6183>
         <1.0494 -6.0969 6.7656>
         <3.3275 -4.8879 6.7826>
         <6.0867 -5.6376 5.5661>
         <2.5168 -3.2968 7.9413>
         <5.5983 -3.4706 7.2267>
         <12.9886 -1.4267 3.7958>
         <9.9231 -3.2266 6.0088>
         <12.2973 1.9074 4.9377>
         <12.4313 5.1182 -2.1817>
         <13.1529 2.3235 -3.165>
         <12.4008 2.338 -4.8634>
         <12.442 4.9507 -3.5495>
         <11.8923 5.7163 -3.8141>
         <12.27 5.3947 -2.781>
        ]}}
      Random Seed=0
     )
     7=&Multiply_Vector(
      A=&SelectVectorFromArray(
       Array Sampling Mode=::NewEnumerator0
       Direct Array Index=?
       Vector Selection Array=DataInterface() {propertyValues {FloatData [
          <17.0065 0.2755 -0.77>
          <3.2355 1.5733 2.9245>
          <6.2383 1.9629 -1.5>
          <-0.7279 3.4165 -2.0094>
          <-3.2654 -1.7219 1.614>
          <-0.5543 -2.2135 2.1404>
          <11.2237 -2.0841 -0.3281>
          <8.1481 -2.7762 0.1657>
          <8.5581 -2.3253 1.7338>
          <8.6436 -2.2336 1.8475>
          <9.85 0.9547 2>
         ]}}
       Random Seed=0
      )
      B=<-1 -1 -1>
     )
     8=&SelectVectorFromArray(
      Array Sampling Mode=::NewEnumerator0
      Direct Array Index=?
      Vector Selection Array=DataInterface() {propertyValues {FloatData [
         <-1.5173 0.1051 -1.404>
         <0.7405 -0.187 -3.0162>
         <0.8485 -0.8853 0.5135>
         <2.0365 -0.7992 -3.378>
         <2.4306 -0.225 -0.087>
         <5.3848 -0.6851 1.719>
         <4.4317 -2.2879 -2.9193>
         <4.6687 -3.8049 -4.9755>
         <6.8585 -0.6734 3.3238>
         <5.5586 0.2819 -0.4461>
        ]}}
      Random Seed=0
     )
     9=&Multiply_Vector(
      A=&SelectVectorFromArray(
       Array Sampling Mode=::NewEnumerator0
       Direct Array Index=?
       Vector Selection Array=DataInterface() {propertyValues {FloatData [
          <-1.5173 0.1051 -1.404>
          <0.7405 -0.187 -3.0162>
          <0.8485 -0.8853 0.5135>
          <2.0365 -0.7992 -3.378>
          <2.4306 -0.225 -0.087>
          <5.3848 -0.6851 1.719>
          <4.4317 -2.2879 -2.9193>
          <4.6687 -3.8049 -4.9755>
          <6.8585 -0.6734 3.3238>
          <5.5586 0.2819 -0.4461>
         ]}}
       Random Seed=0
      )
      B=<-1 -1 -1>
     )
     Index=$Particles.PartitionParticles.Partition
     Num=::NewEnumerator1
    )
    Skeletal Mesh=$System.Pearl_Skeleton
   ) {module=SkeletalMeshLocation002}
   &ScaleSpriteSize() {module=ScaleSpriteSize001}
   &ScaleColor(
    Scale Alpha=&Multiply_Float(
     A=&FloatFromCurve(
      FloatCurve=DataInterface() {propertyValues {Curve {keys [
          {interpMode=RCIM_Cubic, tangentMode=RCTM_User, time=0, value=0}
          {arriveTangent=-2.4058, interpMode=RCIM_Cubic, leaveTangent=-2.4058, tangentMode=RCTM_User, time=0.26, value=0.8421}
          {arriveTangent=-1.1775, interpMode=RCIM_Cubic, leaveTangent=-1.1775, time=0.4613, value=0.0645}
          {arriveTangent=-0.1197, interpMode=RCIM_Cubic, leaveTangent=-0.1197, time=0.5898, value=0.4537}
          <at 1 0 RCIM_Cubic>
         ]}}}
      Scale Curve=&RandomRangeFloat(Minimum=0.1, Maximum=0.5)
     )
     B=&FloatFromCurve(
      FloatCurve=DataInterface() {propertyValues {Curve {keys [{tangentMode=RCTM_User, time=0, value=0}, {tangentMode=RCTM_User, time=1, value=1}]}}}
      Scale Curve=&FloatFromCurve(
       FloatCurve=DataInterface() {propertyValues {Curve {keys [
           {arriveTangentWeight=0.1667, interpMode=RCIM_Cubic, leaveTangentWeight=0.1667, tangentMode=RCTM_User, time=1.4, value=0}
           {tangentMode=RCTM_User, time=2, value=1}
          ]}}}
       CurveIndex=$Emitter.Age
      )
     )
    )
    Scale RGB=&VectorFromFloat(Value=&Multiply_Float(
      A=&RandomRangeFloat(Minimum=0.6, Maximum=1, Evaluation Type=::NewEnumerator1)
      B=&SelectFloatFromArray(Direct Array Index=$Engine.QualityLevel, Float Selection Array=DataInterface() {propertyValues {FloatData [10,5,1,1,1]}})
     ))
    ScaleA=true
    ScaleRGB=true
   ) {module=ScaleColor001}
   &ScaleColor(
    Scale Alpha=$System.TODAlpha
    Scale RGB=&MakeVectorFromLinearColorRGB(Color=$System.TODColor)
    ScaleA=true
    ScaleRGB=true
   )
   /SetVariables(
    Particles.CenterFlareColor:LinearColor=&MultiplyLinearColor(
     Linear Color=$Particles.Color
     Scale Factor=&MakeVector4_FromLinearColor(LinearColor=<rgba 8.0099 9.8007 10 1>)
    )
    Particles.CenterFlareSize:Vector2f=&Multiply_Vector2(A=$Particles.SpriteSize, B=&Vector2DFromFloat(Value=0.5))
    Particles.CoronaFlareColor:LinearColor=&MultiplyLinearColor(
     Linear Color=$Particles.Color
     Scale Factor=&MakeVector4_FromLinearColor(LinearColor=<rgba 0.15 0.15 0.15 1>)
    )
    Particles.CoronaFlareSize:Vector2f=&Multiply_Vector2(A=$Particles.Initial.SpriteSize, B=&Vector2DFromFloat(Value=1))
    Particles.FlickerFlareColor:LinearColor=&MultiplyLinearColor(
     Linear Color=$Particles.Color
     Scale Factor=&MakeVector4_FromLinearColor(LinearColor=<rgba 0.9441 0.8079 1 1>)
    )
   ) {isSetParameters=true, module=SetVariables_#3}
   /SetVariables(Particles.VisibilityTag=&MakeCustomIntFromBool(
     Bool=&SetBoolByIntComparison(A=$Particles.PartitionParticles.Partition, B=2, Comparison Type=::NewEnumerator4)
     True INT=&MakeCustomIntFromBool(Bool=$User.bisHeadHidden, True INT=1, False INT=0)
     False INT=0
    )) {isSetParameters=true}
  ]
  rendererClasses [&NiagaraSpriteRendererProperties]
  renderers [
   {rendererClass=&NiagaraSpriteRendererProperties}
   {rendererClass=&NiagaraSpriteRendererProperties, rendererIndex=1}
   {rendererClass=&NiagaraSpriteRendererProperties, rendererIndex=2}
  ]
 }
 Pelvis_Spine_Arms_Sparkles002 {
  emitterUpdateScript [
   &EmitterState()
   &SpawnRate(
    Spawn Probability=$System.SpawnProbability
    SpawnRate=&Multiply_Float(A=$System.SpawnRate, B=1.5)
    Use Spawn Probability=true
   )
  ]
  particleSpawnScript [
   &PartitionParticles(Number Of Partitions=6)
   &InitializeParticle(
    AdjustAlpha=false
    AdjustHue=true
    AdjustSaturation=false
    AdjustValue=false
    Color=<rgba 0.843 0.8 1 1>
    Color Mode=::NewEnumerator3
    Lifetime Max=0.4
    Lifetime Min=0.3
    Lifetime Mode=::NewEnumerator1
    Sprite Size Mode=::NewEnumerator3
    Uniform Sprite Size=&Multiply_Float(A=40, B=&RandomRangeFloat(Minimum=0.75, Maximum=1))
    UsePositionOffset=false
    Write Material Random=true
   )
   &CameraOffset(Camera Offset Amount=&FN_FrontEndFloatModifier(
     Gameplay=&MergeHelper_FloatSwitch(0=10, 1=10, 2=7, 3=7, 4=7, 5=7, Index=$Particles.PartitionParticles.Partition, Num=::NewEnumerator4)
     Front End=&MergeHelper_FloatSwitch(0=10, 1=10, 2=3, 3=3, 4=3, 5=3, Index=$Particles.PartitionParticles.Partition, Num=::NewEnumerator4)
    ))
  ]
  particleUpdateScript [
   &ParticleState()
   &SkeletalMeshLocation(
    Bone / Socket Index=&MergeHelper_IntSwitch(0=0, 1=1, 2=2, 3=3, 4=4, 5=5, Index=$Particles.PartitionParticles.Partition, Num=::NewEnumerator4)
    Bone Sampling Mode=::NewEnumerator4
    MaskSpawnGroup=false
    Sampled Position Offset=&MergeHelper_VectorSwitch(
     0=&SelectVectorFromArray(
      Array Sampling Mode=::NewEnumerator0
      Direct Array Index=?
      Vector Selection Array=DataInterface() {propertyValues {FloatData [
         <-1.0432 -8.1814 11.9835>
         <5.7753 -7.4425 7.5019>
         <0.1783 -10.8446 6.8762>
         <3.9267 -9.3738 5.243>
         <5.1458 -8.8863 3.5019>
         <0.3697 -11.4202 -0.1293>
         <3.2341 -10.1487 -1.7886>
         <5.3648 -8.9215 -3.8351>
         <4.2182 -8.3607 -8.2573>
         <0.9369 -10.6705 -7.2707>
         <2.1056 -9.3961 -8.7582>
         <6.5299 -8.4609 -0.9764>
         <7.2932 -8.2386 -2.2795>
         <6.6079 -8.3575 -3.303>
         <8.8786 -5.3028 -8.6621>
         <6.9685 -5.9795 -9.2346>
         <6.6408 0.3555 -11.4518>
         <9.1029 2.0435 -9.9805>
         <6.0295 7.8044 -4.3157>
         <4.3042 8.5143 -1.1929>
         <4.533 8.7003 0.5605>
         <4.1847 8.8228 -1.0578>
         <2.763 8.398 -2.4184>
         <0.8333 8.1659 -1.5161>
         <-0.7598 7.945 0.4404>
         <0.6077 8.135 1.7662>
         <5.2845 6.2701 8.4359>
         <8.7002 4.5074 8.7873>
         <6.7782 3.6979 10.4062>
         <7.116 3.4037 10.4356>
         <6.2253 0.8558 11.6276>
         <8.0119 -3.0434 10.4313>
         <10.0593 -2.9737 9.3049>
         <7.267 -5.5404 9.192>
         <9.1342 -7.6129 1.3762>
         <-0.2159 5.143 -10.0389>
         <1.7853 -6.5471 12.1143>
         <0.308 -5.7135 13.2451>
         <3.7597 -3.0289 12.7688>
         <2.3659 -2.7923 13.5333>
         <4.847 -1.1793 12.3499>
         <-0.5241 -4.1961 14.3355>
         <3.5471 3.7594 11.7152>
         <2.0484 2.6664 13.1641>
         <1.2978 3.3629 12.9459>
         <8.7681 7.018 -6.0589>
         <9.6664 6.112 -7.0413>
         <11.7634 8.0162 -4.4771>
         <12.0044 8.1321 -4.2288>
         <12.8005 2.5946 -8.8244>
         <14.9047 1.7097 -9.0869>
         <10.6999 -5.9216 -6.264>
         <12.8068 -5.6268 -6.6769>
         <14.7475 -6.654 -4.8788>
         <15.5456 -7.152 -3.1369>
         <14.7266 -5.5282 -7.0437>
         <17.1883 -6.7565 -5.6691>
         <15.8404 3.9614 -8.7135>
         <15.4557 5.7362 -7.964>
         <15.1569 6.7834 -6.8844>
         <15.0362 8.9347 2.7038>
         <12.4513 7.084 5.8746>
         <13.6549 6.316 6.8805>
         <12.4221 4.5425 8.0744>
         <12.7772 2.2273 8.9985>
         <14.8082 2.5569 9.0395>
         <11.9586 -2.9632 8.6843>
         <13.7615 -3.6705 8.0321>
         <13.1195 -5.2124 6.8659>
         <13.8362 -6.6833 3.8994>
         <13.9544 -6.5126 0.2724>
         <15.5739 -6.6282 0.3632>
         <17.1717 -1.0488 9.6691>
         <15.9714 0.1941 9.3927>
         <5.2517 -6.7316 -9.7139>
         <3.7903 -3.5461 -12.6481>
         <4.5494 -3.0868 -12.2756>
         <2.9989 -1.4503 -13.365>
         <4.8427 -2.1865 -12.2211>
         <-0.226 -3.3732 -14.5241>
         <4.1786 1.6342 -12.3267>
         <2.0842 -0.0171 -13.7197>
         <1.4842 1.3893 -13.6955>
         <1.4846 4.4529 -11.2397>
         <4.5873 8.3035 -1.9855>
         <8.3852 8.2161 -1.6967>
         <6.9713 7.5174 -4.6364>
         <8.8774 8.245 -2.0212>
         <9.678 8.1071 -0.038>
         <13.6353 8.5278 -2.1863>
        ]}}
      Random Seed=0
     )
     1=&SelectVectorFromArray(
      Array Sampling Mode=::NewEnumerator0
      Direct Array Index=?
      Vector Selection Array=DataInterface() {propertyValues {FloatData [
         <15.7744 2.7046 -2.3057>
         <17.0666 0.0749 5.0294>
         <13.468 4.3273 1.4134>
         <13.9121 4.4305 2.5616>
         <13.7607 4.12 -2.7143>
         <17.419 3.7052 -0.3881>
         <5.5239 -7.7752 1.7566>
         <5.7895 -7.8995 2.31>
         <6.979 -7.9063 2.0866>
         <9.8965 4.5857 10.0092>
         <9.648 7.3172 5.3922>
         <6.6219 10.0181 1.5907>
         <5.015 -8.7048 -6.9859>
         <6.2738 -8.5328 -8.7316>
         <4.6996 -7.4402 -10.1095>
         <4.4107 -7.2485 -10.1066>
         <4.0101 -5.2335 -11.5876>
         <5.5345 1.0834 -12.1728>
         <9.2118 2.1319 -13.3477>
         <10.4048 4.6268 -8.7647>
         <10.2789 6.2945 -5.1987>
         <11.1126 5.6015 -5.1899>
         <6.3051 10.047 -1.6656>
         <-5.6455 -5.6265 -5.1844>
         <-2.0754 -6.7414 -1.5955>
         <-3.0702 -6.8107 -3.2743>
         <0.1897 -7.098 -0.7552>
         <4.2744 -8.4388 -2.7843>
         <-0.1787 -7.567 -6.2384>
         <3.0964 -7.1449 -9.9403>
         <1.8212 -7.2496 -9.2594>
         <3.3622 -2.8325 -12.4141>
         <4.5104 0.5782 -12.4942>
         <4.5982 0.7411 -12.4866>
         <3.3682 0.3323 -12.4195>
         <4.4187 2.223 -12.306>
         <9.0484 7.5816 -6.7616>
         <8.866 6.9627 -8.2107>
         <7.1258 6.3565 -10.5567>
         <6.5242 8.1349 -9.7667>
         <2.0216 4.2186 -11.6678>
         <0.7886 -0.1751 -12.0323>
         <-0.295 2.1638 -11.52>
         <-1.5749 -4.2126 -10.7189>
         <-2.0016 -4.5952 -10.1623>
         <-2.744 -2.5374 -11.04>
         <-1.7667 2.0714 -11.3884>
         <-1.8365 3.8584 -10.8124>
         <-5.5947 -0.8871 -10.4066>
         <-2.4133 -6.849 -5.4772>
         <-4.4953 -6.0855 -5.2091>
         <-6.5461 -0.3267 -10.2299>
         <-5.3894 3.4063 -10.3363>
         <-7.2898 11.7945 -3.7316>
         <-5.3662 9.1705 -7.625>
         <-5.4153 11.8898 -3.7117>
         <-3.7876 10.4479 -6.5152>
         <0.8151 5.7413 -11.3684>
         <-2.3083 10.3953 -8.6616>
         <-2.3413 11.6054 -6.2751>
         <-0.8225 12.8407 -6.1369>
         <0.3045 13.1529 -4.937>
         <1.7562 13.2574 -6.0279>
         <4.3226 11.3036 -8.9838>
         <5.5666 11.5805 -3.6251>
         <8.2999 9.1714 4.3175>
         <8.4993 8.835 4.5023>
         <8.6495 8.0647 7.2849>
         <6.5205 10.7793 2.3843>
         <6.3742 11.3092 3.9302>
         <7.5236 8.8931 7.9359>
         <10.3428 5.5589 8.5502>
         <0.7199 -4.0396 11.7641>
         <1.4381 -4.5524 11.7109>
         <1.196 -2.4415 12.3157>
         <1.0616 -0.5058 12.3281>
         <1.9991 0.0724 12.4252>
         <0.4199 0.0678 12.1536>
         <4.3827 11.8309 8.2213>
         <1.1316 12.0531 9.0682>
         <1.7877 13.4796 5.4368>
         <1.4671 11.5018 -0.1359>
         <-4.5112 12.2914 0.9283>
         <-3.8356 12.257 0.783>
         <-5.8016 12.1977 0.2943>
         <-1.9573 11.9375 2.8811>
         <-0.8697 12.9897 6.0762>
         <-2.9201 10.9447 6.6696>
         <-5.2127 11.6179 4.6235>
         <-7.0009 10.9519 5.8382>
         <-9.1319 10.7897 5.5384>
         <-7.8604 6.6947 9.1006>
         <-8.5011 -3.9711 6.2347>
         <-6.6321 -5.5346 2.25>
         <-4.2506 -5.9806 1.0283>
         <3.9394 -8.1476 3.0791>
         <3.7412 -8.4393 4.8097>
         <4.7247 -8.3616 7.9863>
         <1.1972 -5.6789 10.7617>
         <0.0592 -6.6937 8.4586>
         <0.0326 -7.5749 5.2202>
         <1.6973 -7.9389 4.1774>
         <2.3879 -7.8886 3.0918>
         <-3.7151 -5.1929 7.8835>
         <-4.4649 -3.215 9.9096>
         <-5.3796 -3.0182 9.5488>
         <-7.5065 -3.1584 8.0395>
         <-4.719 -0.449 10.8764>
         <-3.8055 0.3858 11.1594>
         <-4.871 1.1612 10.7999>
         <-2.8353 5.2905 10.3099>
         <1.0475 6.8506 11.2462>
         <2.0869 7.1889 11.4879>
         <7.6231 3.1113 12.6913>
         <9.6847 3.1882 12.6717>
         <5.8402 -6.1741 12.2939>
         <5.0688 -5.4411 13.5577>
        ]}}
      Random Seed=0
     )
     2=&SelectVectorFromArray(
      Array Sampling Mode=::NewEnumerator0
      Direct Array Index=?
      Vector Selection Array=DataInterface() {propertyValues {FloatData [
         <-6.8321 4.1233 -2.2005>
         <-6.6805 -3.1865 0.3192>
         <-5.6186 -3.6823 0.1478>
         <-10.2295 -1.4583 -0.425>
         <-11.9531 2.2259 -1.0822>
         <-11.027 6.9241 3.5971>
         <-10.2309 5.9215 0.0772>
         <-10.4104 5.0783 -0.9074>
         <-8.065 6.1829 -0.01>
         <-6.7516 6.5731 0.8095>
        ]}}
      Random Seed=0
     )
     3=&Multiply_Vector(
      A=&SelectVectorFromArray(
       Array Sampling Mode=::NewEnumerator0
       Direct Array Index=?
       Vector Selection Array=DataInterface() {propertyValues {FloatData [
          <-6.8321 4.1233 -2.2005>
          <-6.6805 -3.1865 0.3192>
          <-5.6186 -3.6823 0.1478>
          <-10.2295 -1.4583 -0.425>
          <-11.9531 2.2259 -1.0822>
          <-11.027 6.9241 3.5971>
          <-10.2309 5.9215 0.0772>
          <-10.4104 5.0783 -0.9074>
          <-8.065 6.1829 -0.01>
          <-6.7516 6.5731 0.8095>
         ]}}
       Random Seed=0
      )
      B=<-1 -1 -1>
     )
     4=&SelectVectorFromArray(
      Array Sampling Mode=::NewEnumerator0
      Direct Array Index=?
      Vector Selection Array=DataInterface() {propertyValues {FloatData [
         <2.0538 0.6915 -3.1465>
         <-0.6458 -4.0657 -1.7933>
         <-2.4306 -3.6897 -2.0127>
         <-1.4009 2.9113 -3.04>
         <-2.8843 -2.7666 -2.7626>
         <-6.7583 -2.3574 -2.014>
         <-12.2108 -2.9032 -0.9181>
         <-11.5524 0.0505 -2.4086>
         <-12.2508 0.0325 -2.4065>
         <-12.509 2.662 -2.2432>
         <-6.5784 3.8685 -1.4702>
         <-4.3198 4.2067 -1.2819>
         <-2.1427 4.3437 -1.114>
         <-9.8142 2.0126 4.7172>
         <-9.8039 4.0174 3.4179>
         <-10.0994 4.5378 2.4588>
         <-10.5067 4.7753 1.3395>
         <-20.0626 2.7474 2.9682>
         <-23.6875 3.8548 -1.373>
         <-18.2593 3.9881 -0.8462>
         <-14.6003 4.2671 2.3565>
         <-14.0341 3.5045 -1.716>
         <-14.9187 -0.6381 -2.3052>
         <-9.5597 -3.4796 0.2778>
         <-20.4585 0.5013 -3.0081>
         <-17.3073 -1.875 -1.9054>
         <-17.2355 -3.4648 0.1428>
         <-14.8715 -3.5452 0.1999>
         <-16.1548 -3.1647 2.086>
         <-12.8954 -1.7856 3.6732>
         <-20.7396 -1.3291 2.9485>
         <-21.1413 -3.1136 -0.1032>
         <-23.7276 0.0516 -3.6666>
         <-11.2198 -0.9259 4.1843>
        ]}}
      Random Seed=0
     )
     5=&Multiply_Vector(
      A=&SelectVectorFromArray(
       Array Sampling Mode=::NewEnumerator0
       Direct Array Index=?
       Vector Selection Array=DataInterface() {propertyValues {FloatData [
          <2.0538 0.6915 -3.1465>
          <-0.6458 -4.0657 -1.7933>
          <-2.4306 -3.6897 -2.0127>
          <-1.4009 2.9113 -3.04>
          <-2.8843 -2.7666 -2.7626>
          <-6.7583 -2.3574 -2.014>
          <-12.2108 -2.9032 -0.9181>
          <-11.5524 0.0505 -2.4086>
          <-12.2508 0.0325 -2.4065>
          <-12.509 2.662 -2.2432>
          <-6.5784 3.8685 -1.4702>
          <-4.3198 4.2067 -1.2819>
          <-2.1427 4.3437 -1.114>
          <-9.8142 2.0126 4.7172>
          <-9.8039 4.0174 3.4179>
          <-10.0994 4.5378 2.4588>
          <-10.5067 4.7753 1.3395>
          <-20.0626 2.7474 2.9682>
          <-23.6875 3.8548 -1.373>
          <-18.2593 3.9881 -0.8462>
          <-14.6003 4.2671 2.3565>
          <-14.0341 3.5045 -1.716>
          <-14.9187 -0.6381 -2.3052>
          <-9.5597 -3.4796 0.2778>
          <-20.4585 0.5013 -3.0081>
          <-17.3073 -1.875 -1.9054>
          <-17.2355 -3.4648 0.1428>
          <-14.8715 -3.5452 0.1999>
          <-16.1548 -3.1647 2.086>
          <-12.8954 -1.7856 3.6732>
          <-20.7396 -1.3291 2.9485>
          <-21.1413 -3.1136 -0.1032>
          <-23.7276 0.0516 -3.6666>
          <-11.2198 -0.9259 4.1843>
         ]}}
       Random Seed=0
      )
      B=<-1 -1 -1>
     )
     6=&SelectVectorFromArray(
      Array Sampling Mode=::NewEnumerator0
      Direct Array Index=?
      Vector Selection Array=DataInterface() {propertyValues {FloatData [
         <17.0065 0.2755 -0.77>
         <3.2355 1.5733 2.9245>
         <6.2383 1.9629 -1.5>
         <-0.7279 3.4165 -2.0094>
         <-3.2654 -1.7219 1.614>
         <-0.5543 -2.2135 2.1404>
         <11.2237 -2.0841 -0.3281>
         <8.1481 -2.7762 0.1657>
         <8.5581 -2.3253 1.7338>
         <8.6436 -2.2336 1.8475>
         <9.85 0.9547 2>
        ]}}
      Random Seed=0
     )
     7=&Multiply_Vector(
      A=&SelectVectorFromArray(
       Array Sampling Mode=::NewEnumerator0
       Direct Array Index=?
       Vector Selection Array=DataInterface() {propertyValues {FloatData [
          <17.0065 0.2755 -0.77>
          <3.2355 1.5733 2.9245>
          <6.2383 1.9629 -1.5>
          <-0.7279 3.4165 -2.0094>
          <-3.2654 -1.7219 1.614>
          <-0.5543 -2.2135 2.1404>
          <11.2237 -2.0841 -0.3281>
          <8.1481 -2.7762 0.1657>
          <8.5581 -2.3253 1.7338>
          <8.6436 -2.2336 1.8475>
          <9.85 0.9547 2>
         ]}}
       Random Seed=0
      )
      B=<-1 -1 -1>
     )
     8=&SelectVectorFromArray(
      Array Sampling Mode=::NewEnumerator0
      Direct Array Index=?
      Vector Selection Array=DataInterface() {propertyValues {FloatData [
         <-1.5173 0.1051 -1.404>
         <0.7405 -0.187 -3.0162>
         <0.8485 -0.8853 0.5135>
         <2.0365 -0.7992 -3.378>
         <2.4306 -0.225 -0.087>
         <5.3848 -0.6851 1.719>
         <4.4317 -2.2879 -2.9193>
         <4.6687 -3.8049 -4.9755>
         <6.8585 -0.6734 3.3238>
         <5.5586 0.2819 -0.4461>
        ]}}
      Random Seed=0
     )
     9=&Multiply_Vector(
      A=&SelectVectorFromArray(
       Array Sampling Mode=::NewEnumerator0
       Direct Array Index=?
       Vector Selection Array=DataInterface() {propertyValues {FloatData [
          <-1.5173 0.1051 -1.404>
          <0.7405 -0.187 -3.0162>
          <0.8485 -0.8853 0.5135>
          <2.0365 -0.7992 -3.378>
          <2.4306 -0.225 -0.087>
          <5.3848 -0.6851 1.719>
          <4.4317 -2.2879 -2.9193>
          <4.6687 -3.8049 -4.9755>
          <6.8585 -0.6734 3.3238>
          <5.5586 0.2819 -0.4461>
         ]}}
       Random Seed=0
      )
      B=<-1 -1 -1>
     )
     Index=$Particles.PartitionParticles.Partition
     Num=::NewEnumerator4
    )
    Skeletal Mesh=$System.Pearl_Skeleton
   ) {module=SkeletalMeshLocation002}
   &ScaleSpriteSize() {module=ScaleSpriteSize001}
   &ScaleColor(
    Scale Alpha=&Multiply_Float(
     A=&FloatFromCurve(
      FloatCurve=DataInterface() {propertyValues {Curve {keys [
          {interpMode=RCIM_Cubic, tangentMode=RCTM_User, time=0, value=0}
          {arriveTangent=-2.4058, interpMode=RCIM_Cubic, leaveTangent=-2.4058, tangentMode=RCTM_User, time=0.26, value=0.8421}
          {arriveTangent=-1.1775, interpMode=RCIM_Cubic, leaveTangent=-1.1775, time=0.4613, value=0.0645}
          {arriveTangent=-0.1197, interpMode=RCIM_Cubic, leaveTangent=-0.1197, time=0.5898, value=0.4537}
          <at 1 0 RCIM_Cubic>
         ]}}}
      Scale Curve=&RandomRangeFloat(Minimum=0.1, Maximum=0.5)
     )
     B=&FloatFromCurve(
      FloatCurve=DataInterface() {propertyValues {Curve {keys [{tangentMode=RCTM_User, time=0, value=0}, {tangentMode=RCTM_User, time=1, value=1}]}}}
      Scale Curve=&FloatFromCurve(
       FloatCurve=DataInterface() {propertyValues {Curve {keys [
           {arriveTangentWeight=0.1667, interpMode=RCIM_Cubic, leaveTangentWeight=0.1667, tangentMode=RCTM_User, time=1.4, value=0}
           {tangentMode=RCTM_User, time=2, value=1}
          ]}}}
       CurveIndex=$Emitter.Age
      )
     )
    )
    Scale RGB=&VectorFromFloat(Value=&Multiply_Float(
      A=&RandomRangeFloat(Minimum=0.6, Maximum=1, Evaluation Type=::NewEnumerator1)
      B=&SelectFloatFromArray(Direct Array Index=$Engine.QualityLevel, Float Selection Array=DataInterface() {propertyValues {FloatData [10,5,1,1,1]}})
     ))
    ScaleA=true
    ScaleRGB=true
   ) {module=ScaleColor001}
   &ScaleColor(
    Scale Alpha=$System.TODAlpha
    Scale RGB=&MakeVectorFromLinearColorRGB(Color=$System.TODColor)
    ScaleA=true
    ScaleRGB=true
   )
   /SetVariables(
    Particles.CenterFlareColor:LinearColor=&MultiplyLinearColor(
     Linear Color=$Particles.Color
     Scale Factor=&MakeVector4_FromLinearColor(LinearColor=<rgba 8.0099 9.8007 10 1>)
    )
    Particles.CenterFlareSize:Vector2f=&Multiply_Vector2(A=$Particles.SpriteSize, B=&Vector2DFromFloat(Value=0.5))
    Particles.CoronaFlareColor:LinearColor=&MultiplyLinearColor(
     Linear Color=$Particles.Color
     Scale Factor=&MakeVector4_FromLinearColor(LinearColor=<rgba 0.15 0.15 0.15 1>)
    )
    Particles.CoronaFlareSize:Vector2f=&Multiply_Vector2(A=$Particles.Initial.SpriteSize, B=&Vector2DFromFloat(Value=1))
    Particles.FlickerFlareColor:LinearColor=&MultiplyLinearColor(
     Linear Color=$Particles.Color
     Scale Factor=&MakeVector4_FromLinearColor(LinearColor=<rgba 0.9441 0.8079 1 1>)
    )
   ) {isSetParameters=true, module=SetVariables_#3}
  ]
  rendererClasses [&NiagaraSpriteRendererProperties]
  renderers [
   {rendererClass=&NiagaraSpriteRendererProperties}
   {rendererClass=&NiagaraSpriteRendererProperties, rendererIndex=1}
   {rendererClass=&NiagaraSpriteRendererProperties, rendererIndex=2}
  ]
 }
}
summary {
 systemName=NS_PiperShelf_Character_Sparkles_Empress_Idle
 emitters [
  {emitterName=LowerBody_Head_Sparkles, rendererClasses [&NiagaraSpriteRendererProperties]}
  {emitterName=Pelvis_Spine_Arms_Sparkles002, rendererClasses [&NiagaraSpriteRendererProperties]}
  {emitterName=Hands_LowerArm_Sparkles, rendererClasses [&NiagaraSpriteRendererProperties]}
  {emitterName=LowerBody_Head_Sparkles001, rendererClasses [&NiagaraSpriteRendererProperties]}
 ]
 userVariables [
  User.bisFootHidden {type=NiagaraBool}
  User.bIsFrontEnd {type=NiagaraBool}
  User.bIsFrontEndPreview {type=NiagaraBool}
  User.bIsHandHidden {type=NiagaraBool}
  User.bisHeadHidden {type=NiagaraBool}
  User.Lifetime_Scale {type=NiagaraFloat, defaultValue=0, description=""}
  User.Variant {type=NiagaraFloat, defaultValue=0, description=""}
 ]
}
