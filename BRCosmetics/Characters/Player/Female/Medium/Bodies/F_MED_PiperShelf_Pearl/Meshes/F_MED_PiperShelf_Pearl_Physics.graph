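# Physics asset for the female medium PiperShelf Pearl player body, built around a skirt with leg, foot, glute and floor capsules as apparent colliders.
# Each side has a four-bone A chain and a three-bone B chain with limited swing and twist; B hangs from a zero-size base body through a single-axis hinge.
# Angularly free constraints tie each B tip to the A tip on the same side, while leg joints are angularly locked.
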PhysicsAsset /BRCosmetics/Characters/Player/Female/Medium/Bodies/F_MED_PiperShelf_Pearl/Meshes/F_MED_PiperShelf_Pearl_Physics
bodies {
 base_skirtB_l [{shapeType=Capsule, length=0, radius=0, shapeName=base_skirtB_l_capsule}]
 base_skirtB_r [{shapeType=Capsule, length=0, radius=0, shapeName=base_skirtB_l_capsule, rotation=<rot 0 0 -180>}]
 calf_l [{shapeType=Capsule, length=39.1924, radius=8.3639, center=<-29.5533 -0.8129 -0.8883>, rotation=<rot -0.1577 85.173 -88.0088>}]
 calf_r [{shapeType=Capsule, length=39.1924, radius=8.3639, center=<29.5533 0.8129 0.8883>, rotation=<rot -0.1577 85.173 91.9912>}]
 deform_glute_l [{shapeType=Capsule, length=7.1, radius=9.6053, shapeName=deform_glute_l_capsule, center=<6.9218 4.8944 0>, rotation=<rot 90 180 -130.0288>}]
 deform_glute_r [{
   shapeType=Capsule
   length=7.1
   radius=9.6053
   shapeName=deform_glute_l_capsule
   center=<-6.9218 -4.8944 0>
   rotation=<rot 90 -14.0362 -144.065>
  }]
 dyn_skirtA_01_l [{shapeType=Capsule, length=4, radius=2, shapeName=dyn_skirtA_01_l_capsule, center=<11.8184 -1.0917 0>, rotation=<rot 0 0 -7.6393>}]
 dyn_skirtA_01_r [{shapeType=Capsule, length=7.9, radius=2, shapeName=dyn_skirtA_01_l_capsule, center=<-11.8184 2.6518 0.2093>, rotation=<rot 0 0 172.3607>}]
 dyn_skirtA_02_l [{shapeType=Capsule, length=10.6, radius=6.4, shapeName=dyn_skirtA_02_l_capsule, center=<12.8675 -5 -7.2>, rotation=<rot 0 0 -27.5671>}]
 dyn_skirtA_02_r [{
   shapeType=Capsule
   length=11.3
   radius=5.7
   shapeName=dyn_skirtA_02_l_capsule
   center=<-12.8675 5.044 7.2174>
   rotation=<rot 0.0001 0 153.8134>
  }]
 dyn_skirtA_03_l [{
   shapeType=Capsule
   length=10.3
   radius=6.3
   shapeName=dyn_skirtA_03_l_capsule
   center=<16.937 -2.0698 -4.035>
   rotation=<rot 0.0001 0 -30.23>
  }]
 dyn_skirtA_03_r [{shapeType=Capsule, length=10.3, radius=6.3, shapeName=dyn_skirtA_03_l_capsule, center=<-16.937 3 0>, rotation=<rot 0.0001 0 167.0187>}]
 dyn_skirtA_04_l [{
   shapeType=Capsule
   length=24.3
   radius=7.5
   shapeName=dyn_skirtA_04_l_capsule
   center=<10.8464 2 -6>
   rotation=<rot -39.9316 38.4291 -50.979>
  }]
 dyn_skirtA_04_r [{
   shapeType=Capsule
   length=24.3
   radius=7.5
   shapeName=dyn_skirtA_04_l_capsule
   center=<-10.8464 -2 6>
   rotation=<rot -39.9316 38.4291 129.021>
  }]
 dyn_skirtB_01_l [{shapeType=Capsule, length=8.4, radius=3.4, shapeName=dyn_skirtB_01_l_capsule, center=<12.4014 0 0>, rotation=<rot 0 0.0001 -17.3162>}]
 dyn_skirtB_01_r [{shapeType=Capsule, length=8.4, radius=3.4, shapeName=dyn_skirtB_01_l_capsule, center=<-12.4014 0 0>, rotation=<rot 0 0.0001 162.6838>}]
 dyn_skirtB_02_l [{shapeType=Capsule, length=8.8, radius=7.1, shapeName=dyn_skirtB_02_l_capsule, center=<11.6136 -3 0>}]
 dyn_skirtB_02_r [{shapeType=Capsule, length=8.8, radius=7.1, shapeName=dyn_skirtB_02_l_capsule, center=<-11.6136 3 0>, rotation=<rot 0 0 -180>}]
 dyn_skirtB_03_l [{
   shapeType=Capsule
   length=18.4
   radius=9.8
   shapeName=dyn_skirtB_03_l_capsule
   center=<12.4478 -5.5605 2.2678>
   rotation=<rot -30.4087 0.0001 5.6252>
  }]
 dyn_skirtB_03_r [{
   shapeType=Capsule
   length=18.4
   radius=9.8
   shapeName=dyn_skirtB_03_l_capsule
   center=<-12.4478 5.5605 -2.2678>
   rotation=<rot -30.4087 0.0001 -174.3748>
  }]
 floor [{shapeType=Capsule, length=10, radius=200, shapeName=root_capsule, center=<0 8.2355 -186.1124>, rotation=<rot 0 -90 0>}]
 foot_l [{
   shapeType=Capsule
   length=5.7453
   radius=12.3786
   shapeName=foot_l_capsule
   center=<-2.5694 2.7842 -0.1627>
   rotation=<rot 0 18.5627 -86.602>
  }]
 foot_r [{
   shapeType=Capsule
   length=2.9475
   radius=12.6665
   shapeName=foot_r_capsule
   center=<2.5948 -2.8006 0.1628>
   rotation=<rot 0 21.1006 -86.6576>
  }]
 thigh_l [{shapeType=Capsule, length=35.269, radius=8.7159, center=<-22.3205 0.0008 0.6653>, rotation=<rot -89.5153 179.9985 179.9997>}]
 thigh_r [{shapeType=Capsule, length=35.969, radius=8.7159, center=<24.0227 -0.0008 -0.6797>, rotation=<rot -89.5153 179.9985 -0.0003>}]
}
constraints [
 {bone1Name=dyn_skirtA_03_l, bone2Name=dyn_skirtA_02_l, swing1LimitDegrees=15, swing2LimitDegrees=20, twistLimitDegrees=15}
 {bone1Name=dyn_skirtB_02_l, bone2Name=dyn_skirtB_01_l, swing1LimitDegrees=20, swing2LimitDegrees=20, twistLimitDegrees=5}
 {
  bone1Name=dyn_skirtB_01_l
  bone2Name=base_skirtB_l
  swing1LimitDegrees=20
  swing2LimitDegrees=1
  swing2Motion=Locked
  twistLimitDegrees=1
  twistMotion=Locked
 }
 {
  bone1Name=dyn_skirtB_01_r
  bone2Name=base_skirtB_r
  swing1LimitDegrees=20
  swing2LimitDegrees=1
  swing2Motion=Locked
  twistLimitDegrees=1
  twistMotion=Locked
 }
 {bone1Name=dyn_skirtA_02_l, bone2Name=dyn_skirtA_01_l, swing1LimitDegrees=20, swing2LimitDegrees=20, twistLimitDegrees=15}
 {bone1Name=dyn_skirtA_03_r, bone2Name=dyn_skirtA_02_r, swing1LimitDegrees=15, swing2LimitDegrees=16, twistLimitDegrees=15}
 {
  bone1Name=calf_r
  bone2Name=thigh_r
  swing1LimitDegrees=20
  swing1Motion=Locked
  swing2LimitDegrees=20
  swing2Motion=Locked
  twistLimitDegrees=5
  twistMotion=Locked
 }
 {
  bone1Name=foot_r
  bone2Name=calf_r
  swing1LimitDegrees=45
  swing1Motion=Locked
  swing2LimitDegrees=45
  swing2Motion=Locked
  twistLimitDegrees=45
  twistMotion=Locked
 }
 {
  bone1Name=foot_l
  bone2Name=calf_l
  swing1LimitDegrees=45
  swing1Motion=Locked
  swing2LimitDegrees=45
  swing2Motion=Locked
  twistLimitDegrees=45
  twistMotion=Locked
 }
 {bone1Name=dyn_skirtB_03_r, bone2Name=dyn_skirtB_02_r, swing1LimitDegrees=20, swing2LimitDegrees=25, twistLimitDegrees=5}
 {bone1Name=dyn_skirtA_04_r, bone2Name=dyn_skirtA_03_r, swing1LimitDegrees=15, swing2LimitDegrees=20, twistLimitDegrees=15}
 {
  bone1Name=dyn_skirtB_03_r
  bone2Name=dyn_skirtA_04_r
  swing1LimitDegrees=45
  swing1Motion=Free
  swing2LimitDegrees=45
  swing2Motion=Free
  twistLimitDegrees=45
  twistMotion=Free
 }
 {bone1Name=dyn_skirtA_02_r, bone2Name=dyn_skirtA_01_r, swing1LimitDegrees=12, swing2LimitDegrees=10, twistLimitDegrees=15}
 {bone1Name=dyn_skirtB_02_r, bone2Name=dyn_skirtB_01_r, swing1LimitDegrees=20, swing2LimitDegrees=20, twistLimitDegrees=5}
 {
  bone1Name=calf_l
  bone2Name=thigh_l
  swing1LimitDegrees=45
  swing1Motion=Locked
  swing2LimitDegrees=45
  swing2Motion=Locked
  twistLimitDegrees=45
  twistMotion=Locked
 }
 {bone1Name=dyn_skirtA_04_l, bone2Name=dyn_skirtA_03_l, swing1LimitDegrees=20, swing2LimitDegrees=20, twistLimitDegrees=15}
 {bone1Name=dyn_skirtB_03_l, bone2Name=dyn_skirtB_02_l, swing1LimitDegrees=20, swing2LimitDegrees=20, twistLimitDegrees=15}
 {
  bone1Name=dyn_skirtB_03_l
  bone2Name=dyn_skirtA_04_l
  swing1LimitDegrees=45
  swing1Motion=Free
  swing2LimitDegrees=45
  swing2Motion=Free
  twistLimitDegrees=45
  twistMotion=Free
 }
]
