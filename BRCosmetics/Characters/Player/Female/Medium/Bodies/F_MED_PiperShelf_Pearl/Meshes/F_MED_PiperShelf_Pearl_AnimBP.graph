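# Character-part AnimBlueprint for the female PiperShelf Pearl body, adding secondary motion to skirt, jacket and neck.
# The parent mesh's pose feeds a DBNO-aware Control Rig, then a CustomAnim-curve-weighted slot blend on skirt bones.
# RigidBody physics and spring-based AnimDynamics on six jacket bones follow, then sprint-blended skirt rotations.
# A final neck-bone offset scales with HeadUp, computed each update from the base mesh's head-to-spine transform delta.
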AnimBlueprint /BRCosmetics/Characters/Player/Female/Medium/Bodies/F_MED_PiperShelf_Pearl/Meshes/F_MED_PiperShelf_Pearl_AnimBP
target_skeleton=&/F_MED_PiperShelf_Pearl_Skeleton
graphs [
 AnimGraph {nodes [
   n1 = Get bIsSprinting()
   n2 = Control Rig!(Source=Copy Pose From Mesh() {bCopyCurves=true, bUseAttachedParent=true}, Alpha, isDBNO:bool=Get bIsDBNO()) -> (Pose) {
    alphaBoolBlend {blendOption=Linear}
    controlRigAssetReference {blueprintRigClass=&/F_MED_PiperShelf_Pearl_CtrlRig.F_MED_PiperShelf_Pearl_CtrlRig_C}
   }
   n3 = Layered blend per bone(BasePose=@n2, BlendPoses_0=Slot 'DefaultSlot'(), BlendWeights_0=Get CustomAnim() -> (CustomAnim:float)) {layerSetup [{branchFilters [{boneName=base_skirtA_l}, {boneName=base_skirtA_r}, {boneName=base_skirtB_l}, {boneName=base_skirtB_r}]}]}
   n4 = RigidBody!(ComponentPose=Local To Component(LocalPose=@n3), Alpha, OverrideWorldGravity:Vector="0,0,-1000") -> (Pose) {
    bClampLinearTranslationLimitToRefPose=true
    bOverrideWorldGravity=true
    overridePhysicsAsset=&/F_MED_PiperShelf_Pearl_Physics
    simulationSpace=BaseBoneSpace
    alphaBoolBlend {blendOption=Linear}
    baseBoneRef {boneName=pelvis}
    componentAppliedLinearAccClamp=<3000 3000 1000>
    componentLinearAccScale=<0.1 0.1 0.2>
    componentLinearVelScale=<0.1 0.1 0.2>
    overrideWorldGravity {z=-1000}
    simSpaceSettings {worldAlpha=0.4}
   }
   n5 = Anim Dynamics - Bone: dyn_jacket_03_r(ComponentPose=@n4, Alpha=1, ExternalForce="0,0,0") {
    angularDampingOverride=1
    angularSpringConstant=20000
    bAngularSpring=true
    bOverrideAngularDamping=true
    bUseGravityOverride=true
    gravityScale=0.35
    simulationSpace=RootRelative
    alphaBoolBlend {blendOption=Linear}
    boundBone {boneName=dyn_jacket_03_r}
    componentLinearAccScale=<1 1 1>
    componentLinearVelScale=<1 1 1>
    gravityOverride {z=-980}
    physicsBodyDefinitions [{
      boundBone {boneName=dyn_jacket_03_r}
      boxExtents=<20 30 30>
      constraintSetup {angularLimitsMax=<15 15 15>, angularLimitsMin=<-15 -15 -15>, angularTarget=<1 0 0>}
      localJointOffset=<2 0 0>
     }]
   }
   n6 = Anim Dynamics - Bone: dyn_jacket_02_r(ComponentPose=@n5, Alpha=1, ExternalForce="0,0,0") {
    angularDampingOverride=1
    angularSpringConstant=20000
    bAngularSpring=true
    bOverrideAngularDamping=true
    bUseGravityOverride=true
    gravityScale=0.35
    simulationSpace=RootRelative
    alphaBoolBlend {blendOption=Linear}
    boundBone {boneName=dyn_jacket_02_r}
    componentLinearAccScale=<1 1 1>
    componentLinearVelScale=<1 1 1>
    gravityOverride {z=-980}
    physicsBodyDefinitions [{
      boundBone {boneName=dyn_jacket_02_r}
      boxExtents=<20 30 30>
      constraintSetup {angularLimitsMax=<15 15 15>, angularLimitsMin=<-15 -15 -15>, angularTarget=<1 0 0>}
      localJointOffset=<3 0 0>
     }]
   }
   n7 = Anim Dynamics - Bone: dyn_jacket_01_r(ComponentPose=@n6, Alpha=1, ExternalForce="0,0,0") {
    angularDampingOverride=1
    angularSpringConstant=20000
    bAngularSpring=true
    bOverrideAngularDamping=true
    bUseGravityOverride=true
    gravityScale=0.35
    simulationSpace=RootRelative
    alphaBoolBlend {blendOption=Linear}
    boundBone {boneName=dyn_jacket_01_r}
    componentLinearAccScale=<1 1 1>
    componentLinearVelScale=<1 1 1>
    gravityOverride {z=-980}
    physicsBodyDefinitions [{
      boundBone {boneName=dyn_jacket_01_r}
      boxExtents=<20 30 30>
      constraintSetup {angularLimitsMax=<15 0 15>, angularLimitsMin=<-15 -25 -15>, angularTarget=<1 0 0>}
      localJointOffset=<4 0 0>
     }]
   }
   n8 = Anim Dynamics - Bone: dyn_jacket_03_l(ComponentPose=@n7, Alpha=1, ExternalForce="0,0,0") {
    angularDampingOverride=1
    angularSpringConstant=20000
    bAngularSpring=true
    bOverrideAngularDamping=true
    bUseGravityOverride=true
    gravityScale=0.35
    simulationSpace=RootRelative
    alphaBoolBlend {blendOption=Linear}
    boundBone {boneName=dyn_jacket_03_l}
    componentLinearAccScale=<1 1 1>
    componentLinearVelScale=<1 1 1>
    gravityOverride {z=-980}
    physicsBodyDefinitions [{
      boundBone {boneName=dyn_jacket_03_l}
      boxExtents=<20 30 30>
      constraintSetup {angularLimitsMax=<15 15 15>, angularLimitsMin=<-15 -15 -15>, angularTarget=<1 0 0>}
      localJointOffset=<4 0 0>
     }]
   }
   n9 = Anim Dynamics - Bone: dyn_jacket_02_l(ComponentPose=@n8, Alpha=1, ExternalForce="0,0,0") {
    angularDampingOverride=1
    angularSpringConstant=20000
    bAngularSpring=true
    bOverrideAngularDamping=true
    bUseGravityOverride=true
    gravityScale=0.35
    simulationSpace=RootRelative
    alphaBoolBlend {blendOption=Linear}
    boundBone {boneName=dyn_jacket_02_l}
    componentLinearAccScale=<1 1 1>
    componentLinearVelScale=<1 1 1>
    gravityOverride {z=-980}
    physicsBodyDefinitions [{
      boundBone {boneName=dyn_jacket_02_l}
      boxExtents=<20 30 30>
      constraintSetup {angularLimitsMax=<15 15 15>, angularLimitsMin=<-15 -15 -15>, angularTarget=<1 0 0>}
      localJointOffset=<4 0 0>
     }]
   }
   n10 = Anim Dynamics - Bone: dyn_jacket_01_l(ComponentPose=@n9, Alpha=1, ExternalForce="0,0,0") {
    angularDampingOverride=1
    angularSpringConstant=20000
    bAngularSpring=true
    bOverrideAngularDamping=true
    bUseGravityOverride=true
    gravityScale=0.35
    simulationSpace=RootRelative
    alphaBoolBlend {blendOption=Linear}
    boundBone {boneName=dyn_jacket_01_l}
    componentLinearAccScale=<1 1 1>
    componentLinearVelScale=<1 1 1>
    gravityOverride {z=-980}
    physicsBodyDefinitions [{
      boundBone {boneName=dyn_jacket_01_l}
      boxExtents=<20 30 30>
      constraintSetup {angularLimitsMax=<15 15 5>, angularLimitsMin=<-15 -15 -5>, angularTarget=<1 0 0>}
      localJointOffset=<4 0 0>
     }]
   }
   n11 = "Transform (Modify) Bone - Bone: base_skirtB_l"!(ComponentPose=@n10, bAlphaBoolEnabled:bool=@n1, Rotation="0,-5,0") -> (Pose) {
    alphaInputType=Bool
    rotationMode=BMM_Additive
    rotationSpace=BCS_BoneSpace
    alphaBoolBlend {blendInTime=1, blendOption=Cubic, blendOutTime=1}
    boneToModify {boneName=base_skirtB_l}
    rotation {yaw=-5}
   }
   n12 = "Transform (Modify) Bone - Bone: dyn_skirtA_03_l"!(ComponentPose=@n11, bAlphaBoolEnabled:bool=@n1, Rotation="5,0,0") -> (Pose) {
    alphaInputType=Bool
    rotationMode=BMM_Additive
    rotationSpace=BCS_BoneSpace
    alphaBoolBlend {blendInTime=1, blendOption=Cubic, blendOutTime=1}
    boneToModify {boneName=dyn_skirtA_03_l}
    rotation {pitch=5}
   }
   n13 = "Transform (Modify) Bone - Bone: base_skirtB_r"!(ComponentPose=@n12, bAlphaBoolEnabled:bool=@n1, Rotation="0,-5,0") -> (Pose) {
    alphaInputType=Bool
    rotationMode=BMM_Additive
    rotationSpace=BCS_BoneSpace
    alphaBoolBlend {blendInTime=1, blendOption=Cubic, blendOutTime=1}
    boneToModify {boneName=base_skirtB_r}
    rotation {yaw=-5}
   }
   n14 = "Transform (Modify) Bone - Bone: dyn_skirtA_03_r"!(ComponentPose=@n13, bAlphaBoolEnabled:bool=@n1, Rotation="5,0,0") -> (Pose) {
    alphaInputType=Bool
    rotationMode=BMM_Additive
    rotationSpace=BCS_BoneSpace
    alphaBoolBlend {blendInTime=1, blendOption=Cubic, blendOutTime=1}
    boneToModify {boneName=dyn_skirtA_03_r}
    rotation {pitch=5}
   }
   n15 = "Transform (Modify) Bone - Bone: dyn_neck_01"(ComponentPose=@n14, Alpha=Get HeadUp() -> (HeadUp:double), Translation="0,2.5,-1", Rotation="0,0,0", Scale="1,1,1") {translationMode=BMM_Additive, alphaBoolBlend {blendOption=Linear}, boneToModify {boneName=dyn_neck_01}, translation {y=2.5, z=-1}}
   Output Pose(Result=Component To Local(ComponentPose=@n15)) {layerGroup=None}
  ]}
 EventGraph {nodes [
   Event BlueprintUpdateAnimation()
   > n1 = Sequence!(execute) -> (then_0, then_1, then_2:Exec, then_3:Exec)
   > n2 = Gravity Override State Function DEPRECATED(CharacterPartAnimInstance=Self-Reference(), GameTimeInSeconds=GetGameTimeInSeconds(), GravityOverrideParams=Get Gravity Override Params())
   > Set Gravity Override(Gravity Override=@n2)
   Set CustomAnim(execute=@n1.then_2, CustomAnim:float=GetCurveValue(CurveName=CustomAnim)) -> (Output_Get:float)
   n3 = GetCurveValue(CurveName=CustomCloak)
   > Set CustomCloak(CustomCloak:float="Absolute (Float)"(A=float - float(A=@n3, B=1))) -> (Output_Get:float)
   n4 = Event OnCharacterCustomizationLoaded()
   > n5 = Set Skel Mesh(Skel Mesh=@n4.CustomBaseMesh)
   Set HeadUp(
    execute=@n1.then_3
    HeadUp:double=GetAlphaFromReferenceTransformDelta(SkeletalMeshComponent=@n5, BoneName=spine_05, BaseName=head, InputRange=(X=0,Y=70), TwistAxis="0,0,1")
   ) -> (Output_Get:double)
  ]}
]
variables [
 CustomAnim {type=number}
 CustomCloak {type=number, default=1}
 Forward {type=number}
 Gravity Override {type=Vector}
 Gravity Override Params {type=GravityOverrideParamsStruct}
 HeadUp {type=number}
 is in frontend alpha {type=number}
 isMoving2dFloat {type=number}
 isSliding {type=boolean}
 RotHead_L {type=number}
 Skel Mesh {type=SkeletalMeshComponent}
]
