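# Control Rig for the F_MED_PiperShelf_Pearl face accessory that poses back-hair base bones from head and shoulder motion.
# Four spherical pose readers on the head store look up, down, left and right alphas, which weight additive transform offsets on the hairB base bones.
# A right-clavicle reader, summed with the look-right alpha, adds an offset to base_hairB_R.
# The debug Controls function is unconnected, and the Delta Rotation helper is unused.
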ControlRigBlueprint /BRCosmetics/Characters/Player/Female/Medium/Bodies/F_MED_PiperShelf_Pearl/Meshes/Parts/F_MED_PiperShelf_Pearl_FaceAcc_CtrlRig
graphs [
 New Function_ContainedGraph {nodes [
   n1 = RigVMFunctionEntryNode()
   Clamp = &RigVMFunction_MathIntClamp(Value=@n1.primAxis)
   RigVMFunction_MathBoolToInteger = &RigVMFunction_MathBoolToInteger(Value=@n1.onlyTwist)
   RigVMFunction_MathQuaternionSwingTwist = &RigVMFunction_MathQuaternionSwingTwist(
    TwistAxis=@n1.twistAxis
    Input=&RigUnit_GetRelativeTransformForItem(
     Child=(Type=Bone,Name="None")
     Parent=(Type=Bone,Name="None")
     bChildInitial=False
     bParentInitial=False
     Parent.Type=@n1.ParentType
     Child.Name=@n1.child
     Parent.Name=@n1.parent
    ).RelativeTransform.Rotation
   )
   RigVMFunction_MathQuaternionToEuler = &RigVMFunction_MathQuaternionToEuler(
    Value=RigVMDispatchNode(
     Values=((),())
     Index=@RigVMFunction_MathBoolToInteger.Result
     Values.0=@RigVMFunction_MathQuaternionSwingTwist.Swing
     Values.1=@RigVMFunction_MathQuaternionSwingTwist.Twist
    ) {template="DISPATCH_RigVMDispatch_SelectInt32(in Index,in Values,out Result)"}.Result
    RotationOrder=@n1.rotationOrder
   )
   DISPATCH_RigVMDispatch_SelectInt32 = RigVMDispatchNode(
    Values=(0,0,0)
    Index=@Clamp.Result
    Values.0=@RigVMFunction_MathQuaternionToEuler.Result.X
    Values.1=@RigVMFunction_MathQuaternionToEuler.Result.Y
    Values.2=@RigVMFunction_MathQuaternionToEuler.Result.Z
   ) {template="DISPATCH_RigVMDispatch_SelectInt32(in Index,in Values,out Result)"}
   Remap = &RigVMFunction_MathDoubleRemap(
    TargetMaximum=1
    TargetMinimum=0
    bClamp=False
    Value=@DISPATCH_RigVMDispatch_SelectInt32.Result
    SourceMaximum=@n1.max
    SourceMinimum=@n1.min
   )
   > RigVMFunctionReturnNode(ouputWeight=&RigVMFunction_MathDoubleClamp(Value=@Remap.Result).Result)
  ]}
 New Function_ContainedGraph {nodes [
   n1 = RigVMFunctionEntryNode()
   RigVMFunction_ControlFlowBranch = &RigVMFunction_ControlFlowBranch(ExecuteContext=@n1.ExecuteContext, Condition=@n1.debugControls)
   RigVMFunctionReturnNode(ExecuteContext=@RigVMFunction_ControlFlowBranch.Completed)
   &RigUnit_SetTransform(
    Item=(Type=Bone,Name="neck_01")
    Weight=1
    bInitial=False
    Value=&RigUnit_GetTransform(Item=(Type=Control,Name="neck_01_ctrl"), bInitial=False).Transform
    ExecutePin=@RigVMFunction_ControlFlowBranch.True
   )
   RigUnit_SetTransform = &RigUnit_SetTransform(
    ExecutePin=()
    Item=(Type=Bone,Name="dyn_hairB_R_01")
    Weight=1
    bInitial=False
    Value=&RigUnit_GetTransform(Item=(Type=Control,Name="dyn_hairB_R_01_ctrl"), bInitial=False).Transform
   )
   > &RigUnit_SetTransform(
    Item=(Type=Bone,Name="dyn_hairB_R_02")
    Weight=1
    bInitial=False
    Value=&RigUnit_GetTransform(Item=(Type=Control,Name="dyn_hairB_R_02_ctrl"), bInitial=False).Transform
   )
   > &RigUnit_SetTransform(
    Item=(Type=Bone,Name="dyn_hairB_R_03")
    Weight=1
    bInitial=False
    Value=&RigUnit_GetTransform(Item=(Type=Control,Name="dyn_hairB_R_03_ctrl"), bInitial=False).Transform
   )
   > &RigUnit_SetTransform(
    Item=(Type=Bone,Name="dyn_hairB_R_04")
    Weight=1
    bInitial=False
    Value=&RigUnit_GetTransform(Item=(Type=Control,Name="dyn_hairB_R_04_ctrl"), bInitial=False).Transform
   )
   > &RigUnit_SetTransform(
    Item=(Type=Bone,Name="dyn_hairB_R_05")
    Weight=1
    bInitial=False
    Value=&RigUnit_GetTransform(Item=(Type=Control,Name="dyn_hairB_R_05_ctrl"), bInitial=False).Transform
   )
   RigUnit_GetTransform = &RigUnit_GetTransform(Item=(Type=Control,Name="base_hairB_R_ctrl"), bInitial=False)
   &RigUnit_SetTransform(ExecutePin=(), Item=(Type=Bone,Name="base_hairB_R"), Weight=1, bInitial=False, Value=@RigUnit_GetTransform.Transform)
  ]}
 RigVMFunctionLibrary {nodes [
   Controls = RigVMCollapseNode(ExecuteContext=(), debugControls=False)
   Delta Rotation = RigVMCollapseNode(ExecuteContext=(), debugRotation=False, onlyTwist=False, twistAxis=(X=0,Y=0,Z=0))
  ]}
 RigVMFunction_Sequence_1_1_ContainedGraph {nodes [
   n1 = RigVMFunctionEntryNode()
   n2 = &RigVMFunction_Sequence(ExecuteContext=@n1.ExecuteContext)
   n3 = &RigVMFunction_Sequence(ExecuteContext=@n2.B)
   RigVMFunctionReturnNode(A=@n2.A, B=@n3.A, C=@n3.B)
  ]}
 RigVMModel {nodes [
   RigUnit_BeginExecution = &RigUnit_BeginExecution()
   RigVMFunction_Sequence_1 = RigVMAggregateNode(ExecuteContext=@RigUnit_BeginExecution.ExecutePin)
   n1 = &RigUnit_SphericalPoseReader(
    ActiveRegionScaleFactors=(PositiveWidth=1,NegativeWidth=0.5,PositiveHeight=1.5,NegativeHeight=1.5)
    ActiveRegionSize=0.1
    Debug=(bDrawDebug=False,DebugScale=25,DebugSegments=20,DebugThickness=0.25)
    DriverAxis=(X=0,Y=1,Z=0)
    DriverItem=(Type=Bone,Name="head")
    FalloffRegionScaleFactors=(PositiveWidth=1,NegativeWidth=1.8,PositiveHeight=1.5,NegativeHeight=1.5)
    FalloffSize=0.2
    FlipWidthScaling=false
    OptionalParentItem=(Type=Bone)
    RotationOffset=(X=90,Y=0,Z=-60)
    ExecutePin=@RigVMFunction_Sequence_1.A
   )
   > RigVMVariableNode(Value=@n1.OutputParam) {variable=head_lookup_alpha}
   > n2 = &RigUnit_SphericalPoseReader(
    ActiveRegionScaleFactors=(PositiveWidth=2,NegativeWidth=0.5,PositiveHeight=2,NegativeHeight=2)
    ActiveRegionSize=0.075
    Debug=(bDrawDebug=False,DebugScale=25,DebugSegments=20,DebugThickness=0.25)
    DriverAxis=(X=0,Y=1,Z=0)
    DriverItem=(Type=Bone,Name="head")
    FalloffRegionScaleFactors=(PositiveWidth=1,NegativeWidth=1.1,PositiveHeight=1,NegativeHeight=1)
    FalloffSize=0.2
    FlipHeightScaling=false
    FlipWidthScaling=true
    OptionalParentItem=(Type=Bone)
    RotationOffset=(X=90,Y=0,Z=42.5)
   )
   > RigVMVariableNode(Value=@n2.OutputParam) {variable=head_lookdown_alpha}
   > n3 = &RigUnit_SphericalPoseReader(
    ActiveRegionScaleFactors=(PositiveWidth=0.75,NegativeWidth=0.75,PositiveHeight=1,NegativeHeight=1)
    ActiveRegionSize=0.1
    Debug=(bDrawDebug=False,DebugScale=25,DebugSegments=20,DebugThickness=0.25)
    DriverAxis=(X=0,Y=1,Z=0)
    DriverItem=(Type=Bone,Name="head")
    FalloffRegionScaleFactors=(PositiveWidth=1,NegativeWidth=1,PositiveHeight=1,NegativeHeight=2.2)
    FalloffSize=0.2
    OptionalParentItem=(Type=Bone)
    RotationOffset=(X=150,Y=0,Z=0)
   )
   > RigVMVariableNode(Value=@n3.OutputParam) {variable=head_lookleft_alpha}
   > n4 = &RigUnit_SphericalPoseReader(
    ActiveRegionScaleFactors=(PositiveWidth=0.75,NegativeWidth=0.75,PositiveHeight=1,NegativeHeight=1)
    ActiveRegionSize=0.1
    Debug=(bDrawDebug=False,DebugScale=25,DebugSegments=20,DebugThickness=0.25)
    DriverAxis=(X=0,Y=1,Z=0)
    DriverItem=(Type=Bone,Name="head")
    FalloffRegionScaleFactors=(PositiveWidth=1,NegativeWidth=1,PositiveHeight=1,NegativeHeight=2.2)
    FalloffSize=0.2
    FlipHeightScaling=true
    FlipWidthScaling=false
    OptionalParentItem=(Type=Bone)
    RotationOffset=(X=30,Y=0,Z=0)
   )
   > RigVMVariableNode(Value=@n4.OutputParam) {variable=head_lookright_alpha}
   n5 = &RigUnit_SphericalPoseReader(
    ActiveRegionScaleFactors=(PositiveWidth=1,NegativeWidth=1,PositiveHeight=1,NegativeHeight=1)
    ActiveRegionSize=0.06
    Debug=(bDrawDebug=False,DebugScale=25,DebugSegments=20,DebugThickness=0.25)
    DriverAxis=(X=-1,Y=0,Z=0)
    DriverItem=(Type=Bone,Name="clavicle_r")
    FalloffRegionScaleFactors=(PositiveWidth=1,NegativeWidth=1,PositiveHeight=1.2,NegativeHeight=1)
    FalloffSize=0.15
    OptionalParentItem=(Type=Bone)
    RotationOffset=(X=-120,Y=0,Z=-100)
    ExecutePin=@RigVMFunction_Sequence_1.B
   )
   Add = &RigVMFunction_MathDoubleAdd(A=@n5.OutputParam, B=RigVMVariableNode() {variable=head_lookright_alpha}.Value)
   > &RigUnit_ModifyTransforms(
    ItemToModify=((Item=(Type=Bone,Name="base_hairB_R"),Transform=(Rotation=(X=0,Y=0,Z=-0.1647,W=0.9863),Translation=(X=-1.9468,Y=-0.5493,Z=0),Scale3D=(X=1,Y=1,Z=1))))
    WeightMinimum=0
    Weight=@Add.Result
   )
   &RigUnit_ModifyTransforms(
    ItemToModify="((Item=(Type=Bone,Name=\"base_hairB_C\"),Transform=(Rotation=(X=0,Y=0,Z=-0.5373,W=0.8434),Translation=(X=0,Y=0,Z=0),Scale3D=(X=1,Y=1,Z=1))),(Item=(Type=Bone,Name=\"base_hairB_L\"),Transform=(Rotation=(X=-0.2183,Y=0.1391,Z=-0.519,W=0.8147),Translation=(X=0,Y=0, ...(truncated)"
    WeightMinimum=0
    ExecutePin=@RigVMFunction_Sequence_1.C
    Weight=RigVMVariableNode() {variable=head_lookup_alpha}.Value
   )
   > &RigUnit_ModifyTransforms(
    ItemToModify="((Item=(Type=Bone,Name=\"base_hairB_C\"),Transform=(Rotation=(X=0,Y=0,Z=0.1736,W=0.9848),Translation=(X=0,Y=0,Z=0),Scale3D=(X=1,Y=1,Z=1))),(Item=(Type=Bone,Name=\"base_hairB_L\"),Transform=(Rotation=(X=0,Y=0,Z=0.1736,W=0.9848),Translation=(X=0,Y=0,Z=0),Scale3D ...(truncated)"
    WeightMinimum=0
    Weight=RigVMVariableNode() {variable=head_lookdown_alpha}.Value
   )
   > &RigUnit_ModifyTransforms(
    ItemToModify="((Item=(Type=Bone,Name=\"base_hairB_C\"),Transform=(Rotation=(X=0.3925,Y=0.1279,Z=0.0003,W=-0.9108),Translation=(X=0,Y=0,Z=0),Scale3D=(X=1,Y=1,Z=1))),(Item=(Type=Bone,Name=\"base_hairB_L\"),Transform=(Rotation=(X=0.4912,Y=0.1294,Z=-0.0784,W=-0.8578),Translatio ...(truncated)"
    WeightMinimum=0
    Weight=RigVMVariableNode() {variable=head_lookleft_alpha}.Value
   )
   > &RigUnit_ModifyTransforms(
    ItemToModify="((Item=(Type=Bone,Name=\"base_hairB_C\"),Transform=(Rotation=(X=-0.5315,Y=-0.1333,Z=0.0289,W=-0.836),Translation=(X=-0.1332,Y=-0.2687,Z=-0.4256),Scale3D=(X=1,Y=1,Z=1))),(Item=(Type=Bone,Name=\"base_hairB_L\"),Transform=(Rotation=(X=-0.4115,Y=-0.1951,Z=0.1032,W ...(truncated)"
    WeightMinimum=0
    Weight=RigVMVariableNode() {variable=head_lookright_alpha}.Value
   )
   Controls = RigVMFunctionReferenceNode(ExecuteContext=(), debugControls=false)
  ]}
]
hierarchy [
 root {loc=<0 0 0>, rot=<0 0 0 1>}
 pelvis {parent=root, loc=<0 -0.5301 98.0571>, rot=<0.0127 0.707 -0.0127 -0.707>}
 spine_01 {parent=pelvis, loc=<6.1208 -0.9328 0>, rot=<0 0 0.034 -0.9994>}
 spine_02 {parent=spine_01, loc=<6.5996 -0.2095 0>, rot=<0 0 0.0238 -0.9997>}
 spine_03 {parent=spine_02, loc=<6.4914 -0.1237 0>, rot=<0 0 0.0149 -0.9999>}
 spine_04 {parent=spine_03, loc=<6.9598 0.0794 0>, rot=<0 0 -0.0132 -0.9999>}
 spine_05 {parent=spine_04, loc=<6.9987 0 0>, rot=<0 0 -0.0015 -1>}
 clavicle_l {parent=spine_05, loc=<5.4831 -0.1053 -3.2946>, rot=<0.0465 -0.7289 -0.009 -0.6829>}
 upperarm_l {parent=clavicle_l, loc=<9.8658 -1.2137 -0.4105>, rot=<0.0164 -0.3245 -0.0269 -0.9454>}
 upperarm_twist_01_l {parent=upperarm_l, loc=<7.2135 -0.1134 0>, rot=<0 0 0 1>}
 deform_upperarm_bk_base_l {parent=upperarm_twist_01_l, loc=<-6.4053 -4.3745 0.314>, rot=<0 0 0 -1>}
 deform_upperarm_bk_l {parent=deform_upperarm_bk_base_l, loc=<0 0 0>, rot=<0 0 0 1>}
 deform_upperarm_fr_base_l {parent=upperarm_twist_01_l, loc=<-5.6312 5.092 0.3354>, rot=<0 0 0 -1>}
 deform_upperarm_fr_l {parent=deform_upperarm_fr_base_l, loc=<0 0 0>, rot=<0 0 0 1>}
 deform_upperarm_up_base_l {parent=upperarm_twist_01_l, loc=<-8.891 0.4825 3.9517>, rot=<0 0 0 -1>}
 deform_upperarm_up_l {parent=deform_upperarm_up_base_l, loc=<0 0 0>, rot=<0 0 0 1>}
 base_jacketTop_02_l {parent=clavicle_l, loc=<3.7504 0.9241 4.4386>, rot=<0.2248 0.6678 -0.2229 -0.6736>}
 dyn_jacketTop_02_l {parent=base_jacketTop_02_l, loc=<0 0 0>, rot=<0 0 0 1>}
 base_jacketTop_01_l {parent=clavicle_l, loc=<4.2766 5.6583 1.0891>, rot=<0.2588 0.5462 -0.1823 -0.7755>}
 dyn_jacketTop_01_l {parent=base_jacketTop_01_l, loc=<0 0 0>, rot=<0 0 0 1>}
 dyn_clavicle_l_help {parent=clavicle_l, loc=<6.1642 0 0>, rot=<0 0 0 1>}
 neck_01 {parent=spine_05, loc=<11.3138 0.1529 0>, rot=<0 0 -0.2003 -0.9797>}
 neck_02 {parent=neck_01, loc=<4.5627 0 0>, rot=<0 0 0.0504 -0.9987>}
 head {parent=neck_02, loc=<5.3914 0 0>, rot=<0 0 0.115 -0.9934>}
 base_hairB_C {parent=head, loc=<7.45 -7.3993 0>, rot=<0.0655 -0.9978 0 0>}
 dyn_hairB_C_01 {parent=base_hairB_C, loc=<0 0 0>, rot=<0 0 0 1>}
 dyn_hairB_C_02 {parent=dyn_hairB_C_01, loc=<7.1108 0 0>, rot=<0 0 0.0914 -0.9958>}
 dyn_hairB_C_03 {parent=dyn_hairB_C_02, loc=<7.1108 0 0>, rot=<0 0 -0.0052 -1>}
 dyn_hairB_C_04 {parent=dyn_hairB_C_03, loc=<7.1108 0 0>, rot=<0 0 -0.1244 -0.9922>}
 dyn_hairB_C_05 {parent=dyn_hairB_C_04, loc=<7.1108 0 0>, rot=<0 0 0 1>}
 base_hairB_R {parent=head, loc=<7.3816 -3.637 6.4245>, rot=<0.1412 -0.99 0 0>}
 dyn_hairB_R_01 {parent=base_hairB_R, loc=<0 0 0>, rot=<0 0 0 1>}
 dyn_hairB_R_02 {parent=dyn_hairB_R_01, loc=<7.1108 0 0>, rot=<0 0 0.0675 -0.9977>}
 dyn_hairB_R_03 {parent=dyn_hairB_R_02, loc=<7.1108 0 0>, rot=<0 0 -0.0006 -1>}
 dyn_hairB_R_04 {parent=dyn_hairB_R_03, loc=<7.1108 0 0>, rot=<0 0 -0.1331 -0.9911>}
 dyn_hairB_R_05 {parent=dyn_hairB_R_04, loc=<7.1108 0 0>, rot=<0 0 0 1>}
 dyn_hairB_R_08 {parent=dyn_hairB_R_03, loc=<4.9391 1.7269 -5.4361>, rot=<-0.0657 -0.1342 0.0411 -0.9879>}
 dyn_hairB_R_06 {parent=dyn_hairB_R_01, loc=<4.6697 5.3627 -1.222>, rot=<-0.6695 0.0017 0.2436 -0.7017>}
 dyn_hairB_R_07 {parent=dyn_hairB_R_06, loc=<7.9248 -0.0001 0>, rot=<-0.0028 0.0924 0.1644 -0.9821>}
 base_hairF_R_a {parent=head, loc=<4.6863 5.6721 5.8845>, rot=<0.0376 -0.6823 0.7293 -0.0338>}
 dyn_hairF_R_a01 {parent=base_hairF_R_a, loc=<0 0 0>, rot=<0 0 0 1>}
 dyn_hairF_R_a02 {parent=dyn_hairF_R_a01, loc=<10.789 -0.0066 0.0976>, rot=<-0.0055 -0.0896 0.0141 -0.9959>}
 dyn_hairF_R_a03 {parent=dyn_hairF_R_a02, loc=<7.1108 0 0>, rot=<0.0048 -0.1317 0.0352 -0.9907>}
 dyn_hairF_R_a04 {parent=dyn_hairF_R_a03, loc=<5.0301 0 0>, rot=<0 -0.1603 0 -0.9871>}
 base_hairB_L {parent=head, loc=<7.3811 -3.637 -6.4246>, rot=<0.1412 -0.99 0 0>}
 dyn_hairB_L_01 {parent=base_hairB_L, loc=<0 0 0>, rot=<0 0 0 1>}
 dyn_hairB_L_02 {parent=dyn_hairB_L_01, loc=<7.1102 -0.0002 0>, rot=<0 0 0.0675 -0.9977>}
 dyn_hairB_L_03 {parent=dyn_hairB_L_02, loc=<7.1114 0.0002 0>, rot=<0 0 -0.0006 -1>}
 dyn_hairB_L_04 {parent=dyn_hairB_L_03, loc=<7.11 -0.0003 0>, rot=<0 0 -0.1331 -0.9911>}
 dyn_hairB_L_05 {parent=dyn_hairB_L_04, loc=<7.1113 0 0>, rot=<0 0 0 -1>}
 dyn_hairB_L_06 {parent=dyn_hairB_L_02, loc=<7.8254 1.764 3.8386>, rot=<0.2182 0.1409 0.0876 -0.9617>}
 dyn_hairB_L_07 {parent=dyn_hairB_L_06, loc=<6.6366 -0.0002 0>, rot=<0 0 0.3722 -0.9282>}
 dyn_hairF_L_b01 {parent=dyn_hairB_L_01, loc=<4.669 5.3625 1.222>, rot=<0.6695 -0.0017 0.2436 -0.7017>}
 dyn_hairF_L_b02 {parent=dyn_hairF_L_b01, loc=<7.9248 0 0>, rot=<0.0028 -0.0924 0.1644 -0.9821>}
 base_hairF_L_a {parent=head, loc=<2.1329 5.3163 -5.9202>, rot=<0.6823 0.0376 -0.0338 -0.7293>}
 dyn_hairF_L_a01 {parent=base_hairF_L_a, loc=<0 0 0>, rot=<0 0 0 1>}
 dyn_hairF_L_a02 {parent=dyn_hairF_L_a01, loc=<-8.2125 0 0>, rot=<0.0896 -0.0055 0.9959 0.0141>}
 dyn_hairF_L_a03 {parent=dyn_hairF_L_a02, loc=<7.1108 0 0>, rot=<-0.0048 0.1317 0.0352 -0.9907>}
 dyn_hairF_L_a04 {parent=dyn_hairF_L_a03, loc=<5.0301 0 0>, rot=<0 0.1603 0 -0.9871>}
 base_hairF_R_b {parent=head, loc=<5.6125 4.8232 6.0532>, rot=<0.2553 -0.9419 0.1285 0.1767>}
 dyn_hairF_R_b {parent=base_hairF_R_b, loc=<0 0 0>, rot=<0 0 0 1>}
 dyn_hairF_R_b01 {parent=dyn_hairF_R_b, loc=<5.1721 0 0>, rot=<-0.6261 0.0628 0.0883 -0.7722>}
 dyn_hairF_R_b02 {parent=dyn_hairF_R_b01, loc=<8.1922 0 0>, rot=<0.0026 0.0219 0.0147 -0.9996>}
 clavicle_r {parent=spine_05, loc=<5.4831 -0.1053 3.2946>, rot=<0.7289 0.0465 -0.6829 0.009>}
 upperarm_r {parent=clavicle_r, loc=<-9.8658 1.2137 0.4104>, rot=<0.0164 -0.3245 -0.0269 -0.9454>}
 upperarm_twist_01_r {parent=upperarm_r, loc=<-7.2135 0.1134 0>, rot=<0 0 0 -1>}
 deform_upperarm_bk_base_r {parent=upperarm_twist_01_r, loc=<6.4052 4.3745 -0.3139>, rot=<0 0 0 -1>}
 deform_upperarm_bk_r {parent=deform_upperarm_bk_base_r, loc=<0 0 0>, rot=<0 0 0 1>}
 deform_upperarm_fr_base_r {parent=upperarm_twist_01_r, loc=<5.6311 -5.092 -0.3353>, rot=<0 0 0 -1>}
 deform_upperarm_fr_r {parent=deform_upperarm_fr_base_r, loc=<0 0 0>, rot=<0 0 0 1>}
 deform_upperarm_up_base_r {parent=upperarm_twist_01_r, loc=<8.891 -0.4825 -3.9516>, rot=<0 0 0 -1>}
 deform_upperarm_up_r {parent=deform_upperarm_up_base_r, loc=<0 0 0>, rot=<0 0 0 1>}
 base_jacketTop_02_r {parent=clavicle_r, loc=<-3.7504 -0.9241 -4.4386>, rot=<0.2229 -0.6736 0.2248 -0.6678>}
 dyn_jacketTop_02_r {parent=base_jacketTop_02_r, loc=<0 0 0>, rot=<0 0 0 1>}
 base_jacketTop_01_r {parent=clavicle_r, loc=<-4.2766 -5.6583 -1.0894>, rot=<0.2106 -0.7083 0.2364 -0.6309>}
 dyn_jacketTop_01_r {parent=base_jacketTop_01_r, loc=<0 0 0>, rot=<0 0 0 1>}
 dyn_backpack_help {parent=spine_05, loc=<0 0 0>, rot=<0 0 0 1>}
 neck_01_null {parent=spine_05, type=Null, loc=<11.3138 0.1529 0>, rot=<0 0 -0.2003 -0.9797>}
]
variables [
 armUp_l_alpha {type=double, default=0}
 armUp_r_alpha {type=double, default=0}
 head_lookdown_alpha {type=double, default=0}
 head_lookleft_alpha {type=double, default=0}
 head_lookright_alpha {type=double, default=0}
 head_lookup_alpha {type=double, default=0}
]
