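# Physics asset for the front-hair part of the PiperShelf Pearl female cosmetic character, built from capsule bodies.
# Left and right hair strands form jointed chains off the head, some via base bones, with swing unlocked and twist mostly locked.
# Head, upper spine, clavicles, upper arms and jacket-top bones carry capsules, apparently as colliders, with angularly locked joints where constrained.
# An extra locked constraint cross-links the two right-side chains.
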PhysicsAsset /BRCosmetics/Characters/Player/Female/Medium/Bodies/F_MED_PiperShelf_Pearl/Meshes/Parts/F_MED_PiperShelf_Pearl_FaceAcc_FrontHair_Physics
bodies {
 base_hairF_L_a [{shapeType=Capsule, length=1, radius=1, shapeName=base_hairF_L_a_capsule}]
 base_hairF_R_a [{shapeType=Capsule, length=0.505, radius=0.505, shapeName=base_hairF_R_a_capsule}]
 base_hairF_R_b [{shapeType=Capsule, length=0.505, radius=0.505, shapeName=base_hairF_R_b_capsule, rotation=<rot 0 -90 0>}]
 clavicle_l [{shapeType=Capsule, length=4.2, radius=8.2, shapeName=clavicle_l_capsule, center=<6.9912 -0.2024 -1.8332>, rotation=<rot 0 -90 90>}]
 clavicle_r [{shapeType=Capsule, length=4.2, radius=8.2, shapeName=clavicle_l_capsule, center=<-6.9912 0.2024 1.8332>, rotation=<rot 0 -90 -90>}]
 dyn_clavicle_l_help [{shapeType=Capsule, length=13.2, radius=6.205, shapeName=dyn_clavicle_l_help_capsule, rotation=<rot 90 0 0>}]
 dyn_hairF_L_a01 [{shapeType=Capsule, length=3.7237, radius=1.1, shapeName=dyn_hairF_R_a01_capsule, center=<-7.5617 0 0>}]
 dyn_hairF_L_a02 [
  {
   shapeType=Capsule
   length=4.177
   radius=0.9865
   shapeName=dyn_hairF_R_a02_capsule
   center=<3.4585 0.2283 0.1899>
   rotation=<rot 83.9729 2.6456 -179.9999>
  }
  {
   shapeType=Capsule
   length=3.7254
   radius=0.8058
   shapeName=dyn_hairF_L_a02_capsule
   center=<2.7116 0.2244 -1.7675>
   rotation=<rot 84.3496 -177.592 -177.1171>
  }
 ]
 dyn_hairF_L_a03 [{
   shapeType=Capsule
   length=4.1792
   radius=1.0088
   shapeName=dyn_hairF_R_a03_capsule
   center=<2.2661 0.1318 -0.0161>
   rotation=<rot 88.498 -179.9999 0.0001>
  }]
 dyn_hairF_L_a04 [{
   shapeType=Capsule
   length=3.1544
   radius=0.6
   shapeName=dyn_hairF_R_a04_capsule
   center=<2.147 -0.2498 0.346>
   rotation=<rot 79.9998 -180 5.3089>
  }]
 dyn_hairF_L_b01 [{shapeType=Capsule, length=4.0894, radius=1.3958, shapeName=dyn_hairF_L_b01_capsule, center=<3.9158 0 0>, rotation=<rot 0 93.3354 -84.9458>}]
 dyn_hairF_L_b02 [{
   shapeType=Capsule
   length=6.2927
   radius=1.0136
   shapeName=dyn_hairF_L_b02_capsule
   center=<3.9455 0.3 -1.3>
   rotation=<rot 1.6316 -91.9801 -61.2654>
  }]
 dyn_hairF_R_a01 [{shapeType=Capsule, length=4.2237, radius=0.6877, shapeName=dyn_hairF_R_a01_capsule, center=<5.0748 0 -1.5>, rotation=<rot 90 0 0>}]
 dyn_hairF_R_a02 [{shapeType=Capsule, length=3.4304, radius=1.4617, shapeName=dyn_hairF_R_a02_capsule, center=<3.5045 0 0>, rotation=<rot 90 0 0>}]
 dyn_hairF_R_a03 [{shapeType=Capsule, length=3.2, radius=1, shapeName=dyn_hairF_R_a03_capsule, center=<2.4618 0 0>, rotation=<rot 90 0 0>}]
 dyn_hairF_R_a04 [{shapeType=Capsule, length=3.4, radius=1.1, shapeName=dyn_hairF_R_a04_capsule, center=<3.0352 0 -0.8755>, rotation=<rot 90 0 0>}]
 dyn_hairF_R_b [{shapeType=Capsule, length=1.4279, radius=0.8, shapeName=dyn_hairF_R_b_capsule, center=<3.6324 0 0>, rotation=<rot 0 -93.4885 -89.3493>}]
 dyn_hairF_R_b01 [{
   shapeType=Capsule
   length=4.6061
   radius=1.1
   shapeName=dyn_hairF_R_b01_capsule
   center=<4.2353 0.0034 -0.0073>
   rotation=<rot 0 -89.6459 -89.2303>
  }]
 dyn_hairF_R_b02 [{
   shapeType=Capsule
   length=4.3127
   radius=1.5
   shapeName=dyn_hairF_R_b02_capsule
   center=<3.4275 0.0565 -0.432>
   rotation=<rot 0 -89.0553 -82.8174>
  }]
 dyn_jacketTop_01_l [
  {
   shapeType=Capsule
   length=4.5
   radius=2.5
   shapeName=dyn_jacketTop_01_l_capsule
   center=<-0.8108 -1.4065 -0.3219>
   rotation=<rot 10.3257 -15.2523 90.5098>
  }
  {
   shapeType=Capsule
   length=4.5
   radius=2.5
   shapeName=dyn_jacketTop_01_l_capsule
   center=<2.9357 -1.959 0.7556>
   rotation=<rot 20.5923 -14.036 86.8033>
  }
 ]
 dyn_jacketTop_01_r [
  {shapeType=Capsule, length=6.7, radius=1, shapeName=dyn_jacketTop_01_l_capsule, center=<4.1632 2.911 0.3853>, rotation=<rot 0 13.8765 90>}
  {shapeType=Capsule, length=6.7, radius=1, shapeName=dyn_jacketTop_01_l_capsule, center=<2.1848 2.4222 0.3853>, rotation=<rot 0 13.8765 90>}
 ]
 dyn_jacketTop_02_l [{
   shapeType=Capsule
   length=5.8398
   radius=2.4213
   shapeName=dyn_jacketTop_02_l_capsule
   center=<1.1145 -3.4945 1.2808>
   rotation=<rot -0.8475 -16.1947 118.5425>
  }]
 dyn_jacketTop_02_r [{shapeType=Capsule, length=5, radius=2.3, shapeName=dyn_jacketTop_02_l_capsule, center=<2.6436 0 0>, rotation=<rot 85.6191 -71.2955 -175.457>}]
 head [{shapeType=Capsule, length=2.2709, radius=8.1, center=<2.4192 1.1811 0>, rotation=<rot 90 0 0>}]
 spine_04 [{shapeType=Capsule, length=16.3, radius=10.6, shapeName=spine_04_capsule, center=<4.0135 -0.656 0>, rotation=<rot 0 -90 0>}]
 upperarm_l [{shapeType=Capsule, length=11, radius=5.7, shapeName=upperarm_l_capsule, center=<10.1226 -0.5592 -0.8718>, rotation=<rot 90 0 0>}]
 upperarm_r [{shapeType=Capsule, length=17.7, radius=6, shapeName=upperarm_l_capsule, center=<-14.488 1.8835 0.7808>, rotation=<rot 90 0 180>}]
}
constraints [
 {
  bone1Name=base_hairF_L_a
  bone2Name=head
  swing1LimitDegrees=45
  swing1Motion=Locked
  swing2LimitDegrees=45
  swing2Motion=Locked
  twistLimitDegrees=45
  twistMotion=Locked
 }
 {
  bone1Name=head
  bone2Name=spine_04
  swing1LimitDegrees=45
  swing1Motion=Locked
  swing2LimitDegrees=45
  swing2Motion=Locked
  twistLimitDegrees=45
  twistMotion=Locked
 }
 {
  bone1Name=clavicle_r
  bone2Name=spine_04
  swing1LimitDegrees=45
  swing1Motion=Locked
  swing2LimitDegrees=45
  swing2Motion=Locked
  twistLimitDegrees=45
  twistMotion=Locked
 }
 {
  bone1Name=upperarm_r
  bone2Name=clavicle_r
  swing1LimitDegrees=45
  swing1Motion=Locked
  swing2LimitDegrees=45
  swing2Motion=Locked
  twistLimitDegrees=45
  twistMotion=Locked
 }
 {
  bone1Name=upperarm_l
  bone2Name=clavicle_l
  swing1LimitDegrees=45
  swing1Motion=Locked
  swing2LimitDegrees=45
  swing2Motion=Locked
  twistLimitDegrees=45
  twistMotion=Locked
 }
 {
  bone1Name=base_hairF_R_a
  bone2Name=head
  swing1LimitDegrees=10
  swing1Motion=Locked
  swing2LimitDegrees=10
  swing2Motion=Locked
  twistLimitDegrees=5
  twistMotion=Locked
 }
 {bone1Name=dyn_hairF_R_a01, bone2Name=base_hairF_R_a, swing1LimitDegrees=10, swing2LimitDegrees=15, twistLimitDegrees=5, twistMotion=Locked}
 {bone1Name=dyn_hairF_R_a02, bone2Name=dyn_hairF_R_a01, swing1LimitDegrees=20, swing2LimitDegrees=20, twistLimitDegrees=0, twistMotion=Locked}
 {bone1Name=dyn_hairF_R_a03, bone2Name=dyn_hairF_R_a02, swing1LimitDegrees=20, swing2LimitDegrees=20, twistLimitDegrees=0, twistMotion=Locked}
 {bone1Name=dyn_hairF_R_a04, bone2Name=dyn_hairF_R_a03, swing1LimitDegrees=10, swing2LimitDegrees=10, twistLimitDegrees=0, twistMotion=Locked}
 {
  bone1Name=dyn_jacketTop_01_r
  bone2Name=clavicle_r
  swing1LimitDegrees=10
  swing1Motion=Locked
  swing2LimitDegrees=30
  swing2Motion=Locked
  twistLimitDegrees=5
  twistMotion=Locked
 }
 {
  bone1Name=dyn_jacketTop_02_r
  bone2Name=clavicle_r
  swing1LimitDegrees=10
  swing1Motion=Locked
  swing2LimitDegrees=20
  swing2Motion=Locked
  twistLimitDegrees=5
  twistMotion=Locked
 }
 {bone1Name=dyn_hairF_L_a01, bone2Name=base_hairF_L_a, swing1LimitDegrees=15, swing2LimitDegrees=10, twistLimitDegrees=10}
 {bone1Name=dyn_hairF_L_a02, bone2Name=dyn_hairF_L_a01, swing1LimitDegrees=20, swing2LimitDegrees=15, twistLimitDegrees=5}
 {bone1Name=dyn_hairF_L_a03, bone2Name=dyn_hairF_L_a02, swing1LimitDegrees=15, swing2LimitDegrees=15, twistLimitDegrees=5, twistMotion=Locked}
 {
  bone1Name=dyn_hairF_L_a04
  bone2Name=dyn_hairF_L_a03
  swing1LimitDegrees=12
  swing2LimitDegrees=22.5
  twistLimitDegrees=0
  twistMotion=Locked
 }
 {
  bone1Name=dyn_jacketTop_01_l
  bone2Name=clavicle_l
  swing1LimitDegrees=10
  swing1Motion=Locked
  swing2LimitDegrees=30
  swing2Motion=Locked
  twistLimitDegrees=5
  twistMotion=Locked
 }
 {
  bone1Name=dyn_jacketTop_02_l
  bone2Name=clavicle_l
  swing1LimitDegrees=10
  swing1Motion=Locked
  swing2LimitDegrees=20
  swing2Motion=Locked
  twistLimitDegrees=5
  twistMotion=Locked
 }
 {bone1Name=dyn_hairF_L_b01, bone2Name=head, swing1LimitDegrees=15, swing2LimitDegrees=20, twistLimitDegrees=0, twistMotion=Locked}
 {bone1Name=dyn_hairF_L_b02, bone2Name=dyn_hairF_L_b01, swing1LimitDegrees=15, swing2LimitDegrees=20, twistLimitDegrees=0, twistMotion=Locked}
 {
  bone1Name=dyn_clavicle_l_help
  bone2Name=clavicle_l
  swing1LimitDegrees=45
  swing1Motion=Locked
  swing2LimitDegrees=45
  swing2Motion=Locked
  twistLimitDegrees=45
  twistMotion=Locked
 }
 {
  bone1Name=base_hairF_R_b
  bone2Name=head
  swing1LimitDegrees=45
  swing1Motion=Locked
  swing2LimitDegrees=45
  swing2Motion=Locked
  twistLimitDegrees=45
  twistMotion=Locked
 }
 {bone1Name=dyn_hairF_R_b, bone2Name=base_hairF_R_b, swing1LimitDegrees=20, swing2LimitDegrees=30, twistLimitDegrees=1, twistMotion=Locked}
 {bone1Name=dyn_hairF_R_b01, bone2Name=dyn_hairF_R_b, swing1LimitDegrees=30, swing2LimitDegrees=30, twistLimitDegrees=1}
 {bone1Name=dyn_hairF_R_b02, bone2Name=dyn_hairF_R_b01, swing1LimitDegrees=20, swing2LimitDegrees=20, twistLimitDegrees=1, twistMotion=Locked}
 {
  bone1Name=clavicle_l
  bone2Name=spine_04
  swing1LimitDegrees=45
  swing1Motion=Locked
  swing2LimitDegrees=45
  swing2Motion=Locked
  twistLimitDegrees=45
  twistMotion=Locked
 }
 {
  bone1Name=dyn_hairF_R_b01
  bone2Name=dyn_hairF_R_a02
  swing1LimitDegrees=45
  swing1Motion=Locked
  swing2LimitDegrees=45
  swing2Motion=Locked
  twistLimitDegrees=45
  twistMotion=Locked
 }
]
