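# Physics asset for the F_MED_PiperShelf_Pearl face accessory mesh, setting up capsule bodies for back-hair dynamics on a female character.
# Center, left and right hair chains hang from neck_02 through locked roots, with side branches and narrow, mostly single-axis swing limits.
# Spine, neck, clavicle and jacket capsules form a rigidly constrained upper body, apparently for hair collision.
# dyn_hairB_R_08 has a body but no constraint.
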PhysicsAsset /BRCosmetics/Characters/Player/Female/Medium/Bodies/F_MED_PiperShelf_Pearl/Meshes/Parts/F_MED_PiperShelf_Pearl_FaceAcc_Physics
bodies {
 clavicle_l [{shapeType=Capsule, length=10.6, radius=6.8, shapeName=clavicle_l_capsule, center=<8.9405 0.1462 -0.3963>, rotation=<rot 0 -90 90>}]
 clavicle_r [{shapeType=Capsule, length=5.2, radius=6.7, shapeName=clavicle_l_capsule, center=<-6.9912 0 0.3963>, rotation=<rot 0 -90 -90>}]
 dyn_clavicle_l_help [{shapeType=Capsule, length=13.2, radius=6.205, shapeName=dyn_clavicle_l_help_capsule, rotation=<rot 90 0 0>}]
 dyn_hairB_C_01 [{shapeType=Capsule, length=4.4, radius=0.6, shapeName=dyn_hairB_C_01_capsule, center=<4.0783 0 0>, rotation=<rot 90 0 0>}]
 dyn_hairB_C_02 [{shapeType=Capsule, length=2.7, radius=2, shapeName=dyn_hairB_C_02_capsule, center=<3.1895 0 0>}]
 dyn_hairB_C_03 [{shapeType=Capsule, length=7, radius=3.3, shapeName=dyn_hairB_L_03_capsule, center=<5.5296 0.1349 0>, rotation=<rot 90 180 176.5953>}]
 dyn_hairB_C_05 [{shapeType=Capsule, length=5.6, radius=3, shapeName=dyn_hairB_C_05_capsule, center=<7.6992 0.1233 0>, rotation=<rot 90 0 0>}]
 dyn_hairB_L_01 [{shapeType=Capsule, length=2.9, radius=0.5, shapeName=dyn_hairB_L_01_capsule, center=<5.339 0 0>, rotation=<rot 90 0 0>}]
 dyn_hairB_L_02 [{shapeType=Capsule, length=3.6, radius=2.9, shapeName=dyn_hairB_R_02_capsule, center=<3.1895 0 0>, rotation=<rot 0 0 26.513>}]
 dyn_hairB_L_03 [{shapeType=Capsule, length=6, radius=2.8, shapeName=dyn_hairB_L_03_capsule, center=<6.7472 0.2073 0>, rotation=<rot 85.6238 93.4047 90>}]
 dyn_hairB_L_04 [{shapeType=Capsule, length=4.7, radius=2.2, shapeName=dyn_hairB_L_04_capsule, center=<4.0421 0.0638 3.5076>, rotation=<rot 90 180 176.5952>}]
 dyn_hairB_L_05 [{shapeType=Capsule, length=2.6, radius=2.9, shapeName=dyn_hairB_L_05_capsule, center=<3.1716 -1 0>, rotation=<rot 90 3.4047 0>}]
 dyn_hairB_L_06 [{shapeType=Capsule, length=2.5, radius=2.1, shapeName=dyn_hairB_L_06_capsule, center=<3.2437 -0.35 -0.5>, rotation=<rot 90 180 -162.6665>}]
 dyn_hairB_L_07 [{shapeType=Capsule, length=2, radius=0.9, shapeName=dyn_hairB_L_07_capsule, center=<1.8772 1.1272 -0.6872>, rotation=<rot 67.2291 30.983 0>}]
 dyn_hairB_R_01 [{shapeType=Capsule, length=2.9, radius=0.5, shapeName=dyn_hairB_R_01_capsule, center=<5.1192 0 0>, rotation=<rot 90 0 0>}]
 dyn_hairB_R_02 [{shapeType=Capsule, length=3.9, radius=2.4, shapeName=dyn_hairB_R_02_capsule, center=<3.1895 0 0>, rotation=<rot 0 0 -26.513>}]
 dyn_hairB_R_03 [
  {
   shapeType=Capsule
   length=7
   radius=2.9
   shapeName=dyn_hairB_R_03_capsule
   center=<5.7072 0.3889 -0.6143>
   rotation=<rot -71.0727 100.2643 -98.2798>
  }
  {shapeType=Capsule, length=2, radius=2, shapeName=dyn_hairB_R_03_capsule, center=<6.7939 0.4216 -4.9918>, rotation=<rot -6.9713 0.0002 -10>}
 ]
 dyn_hairB_R_04 [{
   shapeType=Capsule
   length=3.3
   radius=2.4
   shapeName=dyn_hairB_L_04_capsule
   center=<3.4768 -0.5 -2.4572>
   rotation=<rot 73.4479 -176.5952 180>
  }]
 dyn_hairB_R_05 [{
   shapeType=Capsule
   length=6.6
   radius=2.3
   shapeName=dyn_hairB_R_05_capsule
   center=<4.5967 0.7984 1.0488>
   rotation=<rot -55.7146 4.6152 4.4482>
  }]
 dyn_hairB_R_06 [{
   shapeType=Capsule
   length=4.1
   radius=2.8
   shapeName=dyn_hairF_L_b01_capsule
   center=<4.5728 0.5355 -0.294>
   rotation=<rot 81.1818 172.9635 177.0916>
  }]
 dyn_hairB_R_07 [{shapeType=Capsule, length=4.3, radius=1.6, shapeName=dyn_hairB_R_07_capsule, center=<3 2.4706 -1.1>, rotation=<rot 90 90 75.8339>}]
 dyn_hairB_R_08 [{shapeType=Capsule, length=1.5, radius=2.5, shapeName=dyn_hairB_R_08_capsule, center=<3.2281 0 0>, rotation=<rot 90 0 0>}]
 dyn_jacketTop_02_l [{shapeType=Capsule, length=3.6, radius=1.7, shapeName=dyn_jacketTop_02_l_capsule, center=<2.6436 0 0>, rotation=<rot 0 0 96.3345>}]
 neck_01 [{shapeType=Capsule, length=0, radius=7.205, shapeName=neck_01_capsule, center=<1.2134 0.1918 0>, rotation=<rot 90 0 0>}]
 neck_02 [{shapeType=Capsule, length=28.8, radius=15, shapeName=spine_05_capsule, center=<-8.3476 13.4285 0>}]
 spine_04 [{shapeType=Capsule, length=27.1, radius=12, shapeName=spine_04_capsule, center=<1.4943 2.0706 0>}]
 spine_05 [{shapeType=Capsule, length=17.7, radius=7.7, shapeName=spine_05_capsule, center=<2.8857 -0.0282 0>, rotation=<rot 0 -90 0>}]
}
constraints [
 {
  bone1Name=dyn_hairB_C_01
  bone2Name=neck_02
  swing1LimitDegrees=20
  swing1Motion=Locked
  swing2LimitDegrees=0
  swing2Motion=Locked
  twistLimitDegrees=0
  twistMotion=Locked
 }
 {bone1Name=dyn_hairB_C_02, bone2Name=dyn_hairB_C_01, swing1LimitDegrees=30, swing2LimitDegrees=5, twistLimitDegrees=0, twistMotion=Locked}
 {
  bone1Name=dyn_hairB_C_03
  bone2Name=dyn_hairB_C_02
  swing1LimitDegrees=25
  swing2LimitDegrees=1
  swing2Motion=Locked
  twistLimitDegrees=0
  twistMotion=Locked
 }
 {bone1Name=dyn_hairB_C_05, bone2Name=dyn_hairB_C_03, swing1LimitDegrees=15, swing2LimitDegrees=8, twistLimitDegrees=0, twistMotion=Locked}
 {
  bone1Name=dyn_hairB_L_01
  bone2Name=neck_02
  swing1LimitDegrees=20
  swing1Motion=Locked
  swing2LimitDegrees=0
  swing2Motion=Locked
  twistLimitDegrees=0
  twistMotion=Locked
 }
 {bone1Name=dyn_hairB_L_02, bone2Name=dyn_hairB_L_01, swing1LimitDegrees=35, swing2LimitDegrees=3, swing2Motion=Locked, twistLimitDegrees=5}
 {bone1Name=dyn_hairB_L_03, bone2Name=dyn_hairB_L_02, swing1LimitDegrees=35, swing2LimitDegrees=5, twistLimitDegrees=10}
 {
  bone1Name=dyn_hairB_L_04
  bone2Name=dyn_hairB_L_03
  swing1LimitDegrees=20
  swing2LimitDegrees=1
  swing2Motion=Locked
  twistLimitDegrees=0
  twistMotion=Locked
 }
 {bone1Name=dyn_hairB_L_05, bone2Name=dyn_hairB_L_04, swing1LimitDegrees=15, swing2LimitDegrees=5, twistLimitDegrees=0, twistMotion=Locked}
 {
  bone1Name=dyn_hairB_L_06
  bone2Name=dyn_hairB_L_02
  swing1LimitDegrees=20
  swing2LimitDegrees=2
  swing2Motion=Locked
  twistLimitDegrees=0
  twistMotion=Locked
 }
 {
  bone1Name=dyn_hairB_L_07
  bone2Name=dyn_hairB_L_06
  swing1LimitDegrees=20
  swing2LimitDegrees=2
  swing2Motion=Locked
  twistLimitDegrees=0
  twistMotion=Locked
 }
 {
  bone1Name=dyn_hairB_R_01
  bone2Name=neck_02
  swing1LimitDegrees=25
  swing1Motion=Locked
  swing2LimitDegrees=5
  swing2Motion=Locked
  twistLimitDegrees=0
  twistMotion=Locked
 }
 {
  bone1Name=dyn_hairB_R_02
  bone2Name=dyn_hairB_R_01
  swing1LimitDegrees=60
  swing2LimitDegrees=0
  swing2Motion=Locked
  twistLimitDegrees=0
  twistMotion=Locked
 }
 {bone1Name=dyn_hairB_R_03, bone2Name=dyn_hairB_R_02, swing1LimitDegrees=40, swing2LimitDegrees=0, swing2Motion=Locked, twistLimitDegrees=10}
 {
  bone1Name=dyn_hairB_R_04
  bone2Name=dyn_hairB_R_03
  swing1LimitDegrees=25
  swing2LimitDegrees=0
  swing2Motion=Locked
  twistLimitDegrees=0
  twistMotion=Locked
 }
 {bone1Name=dyn_hairB_R_05, bone2Name=dyn_hairB_R_04, swing1LimitDegrees=15, swing2LimitDegrees=8, twistLimitDegrees=0, twistMotion=Locked}
 {
  bone1Name=spine_05
  bone2Name=spine_04
  swing1LimitDegrees=45
  swing1Motion=Locked
  swing2LimitDegrees=45
  swing2Motion=Locked
  twistLimitDegrees=45
  twistMotion=Locked
 }
 {
  bone1Name=clavicle_r
  bone2Name=spine_05
  swing1LimitDegrees=45
  swing1Motion=Locked
  swing2LimitDegrees=45
  swing2Motion=Locked
  twistLimitDegrees=45
  twistMotion=Locked
 }
 {
  bone1Name=neck_01
  bone2Name=spine_05
  swing1LimitDegrees=45
  swing1Motion=Locked
  swing2LimitDegrees=45
  swing2Motion=Locked
  twistLimitDegrees=45
  twistMotion=Locked
 }
 {
  bone1Name=clavicle_l
  bone2Name=spine_05
  swing1LimitDegrees=45
  swing1Motion=Locked
  swing2LimitDegrees=45
  swing2Motion=Locked
  twistLimitDegrees=45
  twistMotion=Locked
 }
 {
  bone1Name=dyn_hairB_R_07
  bone2Name=dyn_hairB_R_06
  swing1LimitDegrees=2
  swing1Motion=Locked
  swing2LimitDegrees=25
  twistLimitDegrees=0
  twistMotion=Locked
 }
 {
  bone1Name=dyn_jacketTop_02_l
  bone2Name=clavicle_l
  swing1LimitDegrees=10
  swing1Motion=Locked
  swing2LimitDegrees=20
  swing2Motion=Locked
  twistLimitDegrees=5
  twistMotion=Locked
 }
 {
  bone1Name=dyn_clavicle_l_help
  bone2Name=clavicle_l
  swing1LimitDegrees=45
  swing1Motion=Locked
  swing2LimitDegrees=45
  swing2Motion=Locked
  twistLimitDegrees=45
  twistMotion=Locked
 }
 {bone1Name=dyn_hairB_R_06, bone2Name=dyn_hairB_R_01, swing1LimitDegrees=10, swing2LimitDegrees=10, twistLimitDegrees=45, twistMotion=Locked}
]
