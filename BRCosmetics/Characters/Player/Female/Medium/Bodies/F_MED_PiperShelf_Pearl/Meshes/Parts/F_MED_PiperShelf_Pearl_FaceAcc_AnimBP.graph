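# Post-process style AnimBlueprint for the F_MED_PiperShelf_Pearl face accessory part, driving its hair physics on the character.
# Copied parent pose, with body-mesh jacket bones, passes through Control Rig into two RigidBody chains selected by tactical sprinting, with jump, skydive and backbling bone offsets.
# Cape or long backbling switches in weighted constraints pinning back hair to upper-body bones, before final corrective offsets.
# Event graph updates backbling flags and gravity override.
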AnimBlueprint /BRCosmetics/Characters/Player/Female/Medium/Bodies/F_MED_PiperShelf_Pearl/Meshes/Parts/F_MED_PiperShelf_Pearl_FaceAcc_AnimBP
target_skeleton=&/F_MED_PiperShelf_Pearl_FaceAcc_Skeleton
graphs [
 AnimGraph {nodes [
   n1 = "Constraint - dyn_hairB_L_03 (PPPPPPP)"!(
    ComponentPose=Local To Component(LocalPose=Use cached pose 'EndSimulation'())
    Alpha=0.5
    ConstraintWeights_0=0.428
    ConstraintWeights_1:float=0.1849
    ConstraintWeights_2:float=0.1368
    ConstraintWeights_3:float=0.1112
    ConstraintWeights_4:float=0.0607
    ConstraintWeights_5:float=0.0597
    ConstraintWeights_6:float=0.0187
   ) -> (Pose) {
    alphaBoolBlend {blendOption=Linear}
    boneToModify {boneName=dyn_hairB_L_03}
    constraintSetup [
     {targetBone {boneName=spine_05}}
     {targetBone {boneName=neck_02}}
     {targetBone {boneName=base_jacketTop_02_l}}
     {targetBone {boneName=neck_01}}
     {targetBone {boneName=head}}
     {targetBone {boneName=clavicle_l}}
     {targetBone {boneName=spine_03}}
    ]
    constraintWeights [1,1,1,1,1,1,1]
   }
   n2 = "Constraint - dyn_hairB_C_03 (PPPPP)"!(
    ComponentPose=@n1
    Alpha=0.5
    ConstraintWeights_0=0.5569
    ConstraintWeights_1:float=0.2111
    ConstraintWeights_2:float=0.1491
    ConstraintWeights_3:float=0.0465
    ConstraintWeights_4:float=0.0364
   ) -> (Pose) {
    alphaBoolBlend {blendOption=Linear}
    boneToModify {boneName=dyn_hairB_C_03}
    constraintSetup [
     {targetBone {boneName=spine_05}}
     {targetBone {boneName=neck_02}}
     {targetBone {boneName=neck_01}}
     {targetBone {boneName=head}}
     {targetBone {boneName=spine_03}}
    ]
    constraintWeights [1,1,1,1,1]
   }
   n3 = "Constraint - dyn_hairB_R_03 (PPPPPPP)"!(
    ComponentPose=@n2
    Alpha=0.5
    ConstraintWeights_0=0.428
    ConstraintWeights_1:float=0.1849
    ConstraintWeights_2:float=0.1368
    ConstraintWeights_3:float=0.1112
    ConstraintWeights_4:float=0.0607
    ConstraintWeights_5:float=0.0597
    ConstraintWeights_6:float=0.0187
   ) -> (Pose) {
    alphaBoolBlend {blendOption=Linear}
    boneToModify {boneName=dyn_hairB_R_03}
    constraintSetup [
     {targetBone {boneName=spine_05}}
     {targetBone {boneName=neck_02}}
     {targetBone {boneName=base_jacketTop_02_r}}
     {targetBone {boneName=neck_01}}
     {targetBone {boneName=head}}
     {targetBone {boneName=clavicle_r}}
     {targetBone {boneName=spine_03}}
    ]
    constraintWeights [1,1,1,1,1,1,1]
   }
   n4 = "Constraint - dyn_hairB_L_04 (PPPPPPP)"!(
    ComponentPose=@n3
    Alpha=1
    ConstraintWeights_0=0.5666
    ConstraintWeights_1:float=0.3934
    ConstraintWeights_2:float=0.0157
    ConstraintWeights_3:float=0.0077
    ConstraintWeights_4:float=0.0069
    ConstraintWeights_5:float=0.0051
    ConstraintWeights_6:float=0.0047
   ) -> (Pose) {
    alphaBoolBlend {blendOption=Linear}
    boneToModify {boneName=dyn_hairB_L_04}
    constraintSetup [
     {targetBone {boneName=clavicle_l}}
     {targetBone {boneName=spine_05}}
     {targetBone {boneName=clavicle_r}}
     {targetBone {boneName=upperarm_twist_01_l}}
     {targetBone {boneName=spine_04}}
     {targetBone {boneName=deform_upperarm_up_l}}
     {targetBone {boneName=deform_upperarm_bk_l}}
    ]
    constraintWeights [1,1,1,1,1,1,1]
   }
   n5 = "Constraint - dyn_hairB_C_04 (PPP)"!(ComponentPose=@n4, Alpha=1, ConstraintWeights_0=0.7732, ConstraintWeights_1:float=0.1134, ConstraintWeights_2:float=0.1134) -> (Pose) {
    alphaBoolBlend {blendOption=Linear}
    boneToModify {boneName=dyn_hairB_C_04}
    constraintSetup [{targetBone {boneName=spine_05}}, {targetBone {boneName=clavicle_l}}, {targetBone {boneName=clavicle_r}}]
    constraintWeights [1,1,1]
   }
   n6 = "Constraint - dyn_hairB_R_04 (PPPPPPP)"!(
    ComponentPose=@n5
    Alpha=1
    ConstraintWeights_0=0.5666
    ConstraintWeights_1:float=0.3934
    ConstraintWeights_2:float=0.0157
    ConstraintWeights_3:float=0.0077
    ConstraintWeights_4:float=0.0069
    ConstraintWeights_5:float=0.0051
    ConstraintWeights_6:float=0.0047
   ) -> (Pose) {
    alphaBoolBlend {blendOption=Linear}
    boneToModify {boneName=dyn_hairB_R_04}
    constraintSetup [
     {targetBone {boneName=clavicle_r}}
     {targetBone {boneName=spine_05}}
     {targetBone {boneName=clavicle_l}}
     {targetBone {boneName=upperarm_twist_01_r}}
     {targetBone {boneName=spine_04}}
     {targetBone {boneName=deform_upperarm_up_r}}
     {targetBone {boneName=deform_upperarm_bk_r}}
    ]
    constraintWeights [1,1,1,1,1,1,1]
   }
   n7 = "Constraint - dyn_hairB_L_05 (PPPPPPPP)"!(
    ComponentPose=@n6
    Alpha=1
    ConstraintWeights_0=0.4281
    ConstraintWeights_1:float=0.3494
    ConstraintWeights_2:float=0.1064
    ConstraintWeights_3:float=0.0518
    ConstraintWeights_4:float=0.026
    ConstraintWeights_5:float=0.0245
    ConstraintWeights_6:float=0.0071
    ConstraintWeights_7:float=0.0068
   ) -> (Pose) {
    alphaBoolBlend {blendOption=Linear}
    boneToModify {boneName=dyn_hairB_L_05}
    constraintSetup [
     {targetBone {boneName=clavicle_l}}
     {targetBone {boneName=spine_05}}
     {targetBone {boneName=spine_04}}
     {targetBone {boneName=upperarm_twist_01_l}}
     {targetBone {boneName=deform_upperarm_bk_l}}
     {targetBone {boneName=clavicle_r}}
     {targetBone {boneName=spine_02}}
     {targetBone {boneName=spine_03}}
    ]
    constraintWeights [1,1,1,1,1,1,1,1]
   }
   n8 = "Constraint - dyn_hairB_C_05 (PPPP)"!(
    ComponentPose=@n7
    Alpha=1
    ConstraintWeights_0=0.6325
    ConstraintWeights_1:float=0.1419
    ConstraintWeights_2:float=0.1419
    ConstraintWeights_3:float=0.0836
   ) -> (Pose) {
    alphaBoolBlend {blendOption=Linear}
    boneToModify {boneName=dyn_hairB_C_05}
    constraintSetup [{targetBone {boneName=spine_05}}, {targetBone {boneName=clavicle_l}}, {targetBone {boneName=clavicle_r}}, {targetBone {boneName=spine_04}}]
    constraintWeights [1,1,1,1]
   }
   n9 = "Constraint - dyn_hairB_R_05 (PPPPPPPP)"!(
    ComponentPose=@n8
    Alpha=1
    ConstraintWeights_0=0.4281
    ConstraintWeights_1:float=0.3494
    ConstraintWeights_2:float=0.1064
    ConstraintWeights_3:float=0.0518
    ConstraintWeights_4:float=0.026
    ConstraintWeights_5:float=0.0245
    ConstraintWeights_6:float=0.0071
    ConstraintWeights_7:float=0.0068
   ) -> (Pose) {
    alphaBoolBlend {blendOption=Linear}
    boneToModify {boneName=dyn_hairB_R_05}
    constraintSetup [
     {targetBone {boneName=clavicle_r}}
     {targetBone {boneName=spine_05}}
     {targetBone {boneName=spine_04}}
     {targetBone {boneName=upperarm_twist_01_r}}
     {targetBone {boneName=deform_upperarm_bk_r}}
     {targetBone {boneName=clavicle_l}}
     {targetBone {boneName=spine_02}}
     {targetBone {boneName=spine_03}}
    ]
    constraintWeights [1,1,1,1,1,1,1,1]
   }
   n10 = "Constraint - dyn_hairB_L_07 (PPPPPPP)"!(
    ComponentPose=@n9
    Alpha=1
    ConstraintWeights_0=0.7118
    ConstraintWeights_1:float=0.1057
    ConstraintWeights_2:float=0.0785
    ConstraintWeights_3:float=0.0688
    ConstraintWeights_4:float=0.0305
    ConstraintWeights_5:float=0.0035
    ConstraintWeights_6:float=0.0013
   ) -> (Pose) {
    alphaBoolBlend {blendOption=Linear}
    boneToModify {boneName=dyn_hairB_L_07}
    constraintSetup [
     {targetBone {boneName=clavicle_l}}
     {targetBone {boneName=upperarm_twist_01_l}}
     {targetBone {boneName=deform_upperarm_bk_l}}
     {targetBone {boneName=deform_upperarm_up_l}}
     {targetBone {boneName=spine_05}}
     {targetBone {boneName=spine_04}}
     {targetBone {boneName=deform_upperarm_fr_l}}
    ]
    constraintWeights [1,1,1,1,1,1,1]
   }
   n11 = "Constraint - dyn_hairB_R_08 (PPPPPPP)"!(
    ComponentPose=@n10
    Alpha=1
    ConstraintWeights_0=0.6791
    ConstraintWeights_1:float=0.1224
    ConstraintWeights_2:float=0.0967
    ConstraintWeights_3:float=0.075
    ConstraintWeights_4:float=0.0217
    ConstraintWeights_5:float=0.0041
    ConstraintWeights_6:float=0.001
   ) -> (Pose) {
    alphaBoolBlend {blendOption=Linear}
    boneToModify {boneName=dyn_hairB_R_08}
    constraintSetup [
     {targetBone {boneName=clavicle_r}}
     {targetBone {boneName=upperarm_twist_01_r}}
     {targetBone {boneName=deform_upperarm_up_r}}
     {targetBone {boneName=deform_upperarm_bk_r}}
     {targetBone {boneName=spine_05}}
     {targetBone {boneName=deform_upperarm_fr_r}}
     {targetBone {boneName=spine_04}}
    ]
    constraintWeights [1,1,1,1,1,1,1]
   }
   n12 = Component To Local(ComponentPose=@n11)
   n13 = Use cached pose 'EndSimulation'()
   n14 = Get Is Wearing Cape() -> (Is Wearing Cape:bool)
   n15 = "Transform (Modify) Bone - Bone: dyn_hairB_R_06"!(
    ComponentPose=Local To Component(LocalPose=Blend Poses by bool(BlendPose_0=@n12, BlendPose_1=@n13, bActiveValue=@n14) {blendType=Linear})
    bAlphaBoolEnabled:bool=True
    Translation="-0.281,-1.830,0.013"
    Rotation="0.036,-3.937,0.513"
   ) -> (Pose) {
    alphaInputType=Bool
    rotationMode=BMM_Additive
    rotationSpace=BCS_BoneSpace
    scaleSpace=BCS_BoneSpace
    translationMode=BMM_Additive
    translationSpace=BCS_BoneSpace
    alphaBoolBlend {blendOption=Linear}
    boneToModify {boneName=dyn_hairB_R_06}
    rotation=<rot 0.036 -3.937 0.513>
    scale=<0.9 0.9 0.9>
    translation=<-0.281 -1.83 0.013>
   }
   n16 = "Transform (Modify) Bone - Bone: dyn_hairF_L_b01"!(ComponentPose=@n15, bAlphaBoolEnabled:bool=True, Translation="0,-0.29,0", Rotation="2.241,-0.004,-0.002") -> (Pose) {
    alphaInputType=Bool
    rotationMode=BMM_Additive
    rotationSpace=BCS_BoneSpace
    scaleSpace=BCS_BoneSpace
    translationMode=BMM_Additive
    translationSpace=BCS_BoneSpace
    alphaBoolBlend {blendOption=Linear}
    boneToModify {boneName=dyn_hairF_L_b01}
    rotation=<rot 2.241 -0.004 -0.002>
    translation {y=-0.29}
   }
   Output Pose(Result=Component To Local(ComponentPose=@n16)) {layerGroup=None}
   n17 = Layered blend per bone(
    BasePose=Copy Pose From Mesh() {bCopyCurves=true, bUseAttachedParent=true}
    BlendPoses_0=Copy Pose From Mesh(SourceMeshComponent=Get Body_Mesh())
   ) {layerSetup [{branchFilters [{boneName=base_jacketTop_02_l}, {boneName=base_jacketTop_01_l}, {boneName=base_jacketTop_02_r}, {boneName=base_jacketTop_01_r}]}]}
   n18 = Control Rig(Source=@n17) {
    alphaBoolBlend {blendOption=Linear}
    controlRigAssetReference {blueprintRigClass=&/F_MED_PiperShelf_Pearl_FaceAcc_CtrlRig.F_MED_PiperShelf_Pearl_FaceAcc_CtrlRig_C}
   }
   n19 = "Transform (Modify) Bone - Bone: dyn_clavicle_l_help"!(ComponentPose=Local To Component(LocalPose=@n18), bAlphaBoolEnabled:bool=Get bIsSkyDiving(), Translation="0,4,8", Rotation="0,0,0") -> (Pose) {
    alphaInputType=Bool
    rotationMode=BMM_Additive
    rotationSpace=BCS_BoneSpace
    translationMode=BMM_Additive
    translationSpace=BCS_BoneSpace
    alphaBoolBlend {blendOption=Linear}
    boneToModify {boneName=dyn_clavicle_l_help}
    translation {y=4, z=8}
   }
   Save cached pose 'BaseSetup'(Pose=Component To Local(ComponentPose=@n19)) {cacheName=BaseSetup}
   n20 = Get bIsJumping()
   n21 = "Transform (Modify) Bone - Bone: base_hairB_C"!(ComponentPose=Local To Component(LocalPose=Use cached pose 'BaseSetup'()), bAlphaBoolEnabled:bool=@n20, Rotation="0,0,30") -> (Pose) {
    alphaInputType=Bool
    rotationMode=BMM_Additive
    translationSpace=BCS_BoneSpace
    alphaBoolBlend {blendOption=Linear}
    boneToModify {boneName=base_hairB_C}
    rotation {roll=30}
   }
   n22 = "Transform (Modify) Bone - Bone: base_hairB_R"!(ComponentPose=@n21, bAlphaBoolEnabled:bool=@n20, Rotation="0,0,30") -> (Pose) {
    alphaInputType=Bool
    rotationMode=BMM_Additive
    translationSpace=BCS_BoneSpace
    alphaBoolBlend {blendOption=Linear}
    boneToModify {boneName=base_hairB_R}
    rotation {roll=30}
   }
   n23 = "Transform (Modify) Bone - Bone: base_hairB_L"!(ComponentPose=@n22, bAlphaBoolEnabled:bool=@n20, Rotation="0,0,40") -> (Pose) {
    alphaInputType=Bool
    rotationMode=BMM_Additive
    translationSpace=BCS_BoneSpace
    alphaBoolBlend {blendOption=Linear}
    boneToModify {boneName=base_hairB_L}
    rotation {roll=40}
   }
   n24 = "Transform (Modify) Bone - Bone: dyn_backpack_help"!(ComponentPose=@n23, bAlphaBoolEnabled:bool=Get isWearingBackbling(), Translation="0,19,0") -> (Pose) {
    alphaInputType=Bool
    translationMode=BMM_Additive
    translationSpace=BCS_BoneSpace
    alphaBoolBlend {blendOption=Linear}
    boneToModify {boneName=dyn_backpack_help}
    rotation {roll=20}
    translation {y=19}
   }
   n25 = RigidBody!(
    ComponentPose=@n24
    bAlphaBoolEnabled:bool
    OverrideWorldGravity:Vector="0,0,-980"
    ComponentLinearAccScale:Vector="0,0,0"
    ComponentLinearVelScale:Vector="0,0,0"
   ) -> (Pose) {
    alphaInputType=Bool
    bClampLinearTranslationLimitToRefPose=true
    bOverrideWorldGravity=true
    bTransferBoneVelocities=true
    overridePhysicsAsset=&/F_MED_PiperShelf_Pearl_FaceAcc_Physics
    simulationSpace=BaseBoneSpace
    alphaBoolBlend {blendOption=Linear}
    baseBoneRef {boneName=neck_02}
    componentAppliedLinearAccClamp=<1000 1000 1000>
    componentLinearAccScale=<0.6 0.2 0.2>
    componentLinearVelScale=<0.6 0.2 0.2>
   }
   n26 = RigidBody(ComponentPose=@n25, OverrideWorldGravity="0,0,-980", ComponentLinearAccScale="0.05,0.05,0.1", ComponentLinearVelScale="0.05,0.05,0.1") {
    alphaInputType=Bool
    bClampLinearTranslationLimitToRefPose=true
    bOverrideWorldGravity=true
    overridePhysicsAsset=&F_MED_PiperShelf_Pearl_FaceAcc_Physics
    simulationSpace=BaseBoneSpace
    alphaBoolBlend {blendOption=Linear}
    baseBoneRef {boneName=head}
    componentAppliedLinearAccClamp=<1000 1000 1000>
    componentLinearAccScale=<0.6 0.2 0.2>
    componentLinearVelScale=<0.6 0.2 0.2>
   }
   n27 = RigidBody(ComponentPose=@n26, OverrideWorldGravity="0,0,-980", ComponentLinearAccScale="0.05,0.05,0.1", ComponentLinearVelScale="0.05,0.05,0.1") {
    alphaInputType=Bool
    bClampLinearTranslationLimitToRefPose=true
    bOverrideWorldGravity=true
    overridePhysicsAsset=&/F_MED_PiperShelf_Pearl_FaceAcc_FrontHair_Physics
    simulationSpace=BaseBoneSpace
    alphaBoolBlend {blendOption=Linear}
    baseBoneRef {boneName=neck_02}
    componentAppliedLinearAccClamp=<20000 20000 20000>
    componentLinearAccScale=<0.6 0.2 0.2>
    componentLinearVelScale=<0.6 0.2 0.2>
   }
   n28 = Get isWearingBackbling()
   n29 = "Transform (Modify) Bone - Bone: dyn_backpack_help"(ComponentPose=Local To Component(LocalPose=Use cached pose 'BaseSetup'()), bAlphaBoolEnabled=@n28, Translation="0,22,0") {
    alphaInputType=Bool
    translationMode=BMM_Additive
    translationSpace=BCS_BoneSpace
    alphaBoolBlend {blendOption=Linear}
    boneToModify {boneName=dyn_backpack_help}
    rotation {roll=20}
    translation {y=22}
   }
   n30 = RigidBody(
    ComponentPose=@n29
    bAlphaBoolEnabled=@n28
    OverrideWorldGravity="0,0,-980"
    ComponentLinearAccScale="0.01,0.01,0.01"
    ComponentLinearVelScale="0.01,0.01,0.01"
   ) {
    alphaInputType=Bool
    bClampLinearTranslationLimitToRefPose=true
    bOverrideWorldGravity=true
    bTransferBoneVelocities=true
    overridePhysicsAsset=&F_MED_PiperShelf_Pearl_FaceAcc_Physics
    simulationSpace=BaseBoneSpace
    alphaBoolBlend {blendOption=Linear}
    baseBoneRef {boneName=spine_05}
    componentAppliedLinearAccClamp=<1000 1000 1000>
    componentLinearAccScale=<0.6 0.2 0.2>
    componentLinearVelScale=<0.6 0.2 0.2>
   }
   n31 = RigidBody(
    ComponentPose=@n30
    bAlphaBoolEnabled=Get isWearingBackbling_not() -> (isWearingBackbling_not:bool)
    OverrideWorldGravity="0,0,0"
    ComponentLinearAccScale="0.3,0.3,0.5"
    ComponentLinearVelScale="0.3,0.3,0.5"
   ) {
    alphaInputType=Bool
    bClampLinearTranslationLimitToRefPose=true
    bOverrideWorldGravity=true
    overridePhysicsAsset=&F_MED_PiperShelf_Pearl_FaceAcc_Physics
    simulationSpace=BaseBoneSpace
    alphaBoolBlend {blendOption=Linear}
    baseBoneRef {boneName=spine_05}
    componentAppliedLinearAccClamp=<3000 3000 3000>
    componentLinearAccScale=<0.6 0.2 0.2>
    componentLinearVelScale=<0.6 0.2 0.2>
    simSpaceSettings {
     maxAngularAcceleration=1000
     maxAngularVelocity=3
     maxLinearAcceleration=700
     maxLinearVelocity=500
     velocityScaleZ=0.5
     worldAlpha=0.4
     externalLinearDragV=<0.2 0.2>
    }
   }
   n32 = RigidBody(ComponentPose=@n31, OverrideWorldGravity="0,0,0", ComponentLinearAccScale="0.12,0.12,0.2", ComponentLinearVelScale="0.12,0.12,0.2") {
    alphaInputType=Bool
    bClampLinearTranslationLimitToRefPose=true
    bOverrideWorldGravity=true
    overridePhysicsAsset=&F_MED_PiperShelf_Pearl_FaceAcc_FrontHair_Physics
    simulationSpace=BaseBoneSpace
    alphaBoolBlend {blendOption=Linear}
    baseBoneRef {boneName=spine_05}
    componentAppliedLinearAccClamp=<20000 20000 20000>
    componentLinearAccScale=<0.6 0.2 0.2>
    componentLinearVelScale=<0.6 0.2 0.2>
   }
   Save cached pose 'EndSimulation'(Pose=Blend Poses by bool(
     BlendPose_0=Component To Local(ComponentPose=@n27)
     BlendPose_1=Component To Local(ComponentPose=@n32)
     bActiveValue=Get bIsTacticalSprinting()
    ) {blendType=Linear}) {cacheName=EndSimulation}
  ]}
 EventGraph {nodes [
   Event BlueprintUpdateAnimation()
   > n1 = Sequence()
   > n2 = Gravity Override State Function DEPRECATED(CharacterPartAnimInstance=Self-Reference(), GameTimeInSeconds=GetGameTimeInSeconds(), GravityOverrideParams=Get Gravity Override Params())
   > Set GravityOverride(GravityOverride=@n2)
   n3 = CheckForBackbling()
   n4 = Set isWearingBackbling(execute=@n1.then_1, isWearingBackbling=@n3)
   > Set isWearingBackbling_not(isWearingBackbling_not:bool=NOT Boolean(A=@n4)) -> (Output_Get:bool)
   > n5 = Set Is Wearing Cape(Is Wearing Cape=OR Boolean(A=@n3.bIsBackblingACape, B=@n3.bIsBackblingLong))
   > Set LimitHair(LimitHair:bool=@n5) -> (Output_Get:bool)
   > Set isNotFrontEnd(isNotFrontEnd:bool=NOT Boolean(A=Get bIsInFrontEnd())) -> (Output_Get:bool)
   > n6 = Branch(Condition=@n4)
   > "Set Linear Scale "("Linear Scale ":Vector="0, 0, 0") -> (Output_Get:Vector)
   "Set Linear Scale "(execute=@n6.else, "Linear Scale "="0.12,0.12,0.2")
   K2Node_Event_1 = CustomEvent() {is=title K2Node_CustomEvent}
   > n7 = Cast To PlayerPawn_Athena(Object=TryGetPawnOwner())
   > Set Body_Mesh(Body_Mesh=GetSkeletalMeshForPartType(self=@n7))
   n8 = Event OnCharacterCustomizationLoaded()
   > Set Skel Mesh(Skel Mesh=@n8.CustomBaseMesh)
  ]}
]
variables [
 Body_Mesh {type=SkeletalMeshComponent}
 Gravity Override Params {
  type=GravityOverrideParamsStruct
  default {skydive_FloatingDown {gravityOverride=<0 -100 -100>, windFrequency=<1.5 3 5>}, skydive_GoingUp {gravityOverride=<0 -300 100>}}
 }
 GravityOverride {type=Vector}
 Head Down {type=number}
 Head_Turn_L {type=number}
 Head_Turn_R {type=number}
 HeadUp {type=number}
 Is Wearing Cape {type=boolean}
 isNotFrontEnd {type=boolean}
 isWearingBackbling {type=boolean}
 isWearingBackbling_not {type=boolean}
 LimitHair {type=boolean}
 "Linear Scale " {type=Vector}
 Linear Scale Backpack {type=Vector}
 Linear Scale Cape {type=Vector}
 Skel Mesh {type=SkeletalMeshComponent}
]
